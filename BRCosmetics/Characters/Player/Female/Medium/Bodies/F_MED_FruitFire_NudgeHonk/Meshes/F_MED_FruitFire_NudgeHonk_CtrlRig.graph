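# Control Rig for the F_MED_FruitFire_NudgeHonk female character body that procedurally corrects skirt and collar bones to follow body motion.
# Leg ROM values and spherical pose readers measure leg poses, driving thigh/knee constraint solvers, outward pushes, mirrored corrective transforms and curves.
# Sliding and DBNO states switch or blend skirt adjustments, with Kalman-smoothed weights for slides.
# Collar bones blend between spine and clavicles, and head/neck swing-twist yields pose alphas.
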ControlRigBlueprint /BRCosmetics/Characters/Player/Female/Medium/Bodies/F_MED_FruitFire_NudgeHonk/Meshes/F_MED_FruitFire_NudgeHonk_CtrlRig
graphs [
 Add_1_ContainedGraph {nodes [
   n1 = RigVMFunctionEntryNode()
   Add = &RigVMFunction_MathDoubleAdd(A=@n1.A, B=@n1.B)
   RigVMFunctionReturnNode(Result=&RigVMFunction_MathDoubleAdd(A=@Add.Result, B=@n1.C).Result)
  ]}
 Maximum_1_ContainedGraph {nodes [
   n1 = RigVMFunctionEntryNode()
   Maximum = &RigVMFunction_MathDoubleMax(A=@n1.A, B=@n1.B)
   RigVMFunctionReturnNode(Result=&RigVMFunction_MathDoubleMax(B=@n1.C, A=@Maximum.Result).Result)
  ]}
 Maximum_1_ContainedGraph {nodes [
   n1 = RigVMFunctionEntryNode()
   Maximum = &RigVMFunction_MathDoubleMax(A=@n1.A, B=@n1.B)
   RigVMFunctionReturnNode(Result=&RigVMFunction_MathDoubleMax(B=@n1.C, A=@Maximum.Result).Result)
  ]}
 Maximum_1_ContainedGraph {nodes [
   n1 = RigVMFunctionEntryNode()
   Maximum = &RigVMFunction_MathDoubleMax(A=@n1.A, B=@n1.B)
   RigVMFunctionReturnNode(Result=&RigVMFunction_MathDoubleMax(B=@n1.D, A=&RigVMFunction_MathDoubleMax(B=@n1.C, A=@Maximum.Result).Result).Result)
  ]}
 Maximum_2_ContainedGraph {nodes [
   n1 = RigVMFunctionEntryNode()
   RigVMFunctionReturnNode(Result=&RigVMFunction_MathDoubleMax(B=@n1.C, A=&RigVMFunction_MathDoubleMax(A=@n1.A, B=@n1.B).Result).Result)
  ]}
 Maximum_2_ContainedGraph {nodes [
   n1 = RigVMFunctionEntryNode()
   RigVMFunctionReturnNode(Result=&RigVMFunction_MathDoubleMax(B=@n1.C, A=&RigVMFunction_MathDoubleMax(A=@n1.A, B=@n1.B).Result).Result)
  ]}
 Neck Collar Constraints_ContainedGraph {nodes [
   RigVMFunctionEntryNode()
   > &RigUnit_ParentConstraint(
    AdvancedSettings=(InterpolationType=Average,RotationOrderForFilter=XZY)
    Child=(Type=Bone,Name="Collar_front_L")
    Filter=(TranslationFilter=(bX=true,bY=true,bZ=true),RotationFilter=(bX=True,bY=True,bZ=True),ScaleFilter=(bX=True,bY=True,bZ=True))
    Parents=((Item=(Type=Bone,Name="spine_05"),Weight=0.8),(Item=(Type=Bone,Name="clavicle_l"),Weight=0.1),(Item=(Type=Bone,Name="clavicle_r"),Weight=0.1))
    Weight=1
   )
   > &RigUnit_ParentConstraint(
    AdvancedSettings=(InterpolationType=Average,RotationOrderForFilter=XZY)
    Child=(Type=Bone,Name="Collar_side_L")
    Filter=(TranslationFilter=(bX=true,bY=true,bZ=true),RotationFilter=(bX=True,bY=True,bZ=True),ScaleFilter=(bX=True,bY=True,bZ=True))
    Parents=((Item=(Type=Bone,Name="spine_05"),Weight=0.7),(Item=(Type=Bone,Name="clavicle_l"),Weight=0.3))
    Weight=1
   )
   > &RigUnit_ParentConstraint(
    AdvancedSettings=(InterpolationType=Average,RotationOrderForFilter=XZY)
    Child=(Type=Bone,Name="Collar_back_L")
    Filter=(TranslationFilter=(bX=true,bY=true,bZ=true),RotationFilter=(bX=True,bY=True,bZ=True),ScaleFilter=(bX=True,bY=True,bZ=True))
    Parents=((Item=(Type=Bone,Name="spine_05"),Weight=0.7),(Item=(Type=Bone,Name="clavicle_l"),Weight=0.3))
    Weight=1
   )
   > &RigUnit_ParentConstraint(
    AdvancedSettings=(InterpolationType=Average,RotationOrderForFilter=XZY)
    Child=(Type=Bone,Name="Collar_back_M")
    Filter=(TranslationFilter=(bX=true,bY=true,bZ=true),RotationFilter=(bX=True,bY=True,bZ=True),ScaleFilter=(bX=True,bY=True,bZ=True))
    Parents=((Item=(Type=Bone,Name="spine_05"),Weight=0.75),(Item=(Type=Bone,Name="clavicle_l"),Weight=0.15))
    Weight=1
   )
   > &RigUnit_ParentConstraint(
    AdvancedSettings=(InterpolationType=Average,RotationOrderForFilter=XZY)
    Child=(Type=Bone,Name="Collar_back_R")
    Filter=(TranslationFilter=(bX=true,bY=true,bZ=true),RotationFilter=(bX=True,bY=True,bZ=True),ScaleFilter=(bX=True,bY=True,bZ=True))
    Parents=((Item=(Type=Bone,Name="spine_05"),Weight=0.7),(Item=(Type=Bone,Name="clavicle_r"),Weight=0.3))
    Weight=1
   )
   > &RigUnit_ParentConstraint(
    AdvancedSettings=(InterpolationType=Average,RotationOrderForFilter=XZY)
    Child=(Type=Bone,Name="Collar_side_R")
    Filter=(TranslationFilter=(bX=true,bY=true,bZ=true),RotationFilter=(bX=True,bY=True,bZ=True),ScaleFilter=(bX=True,bY=True,bZ=True))
    Parents=((Item=(Type=Bone,Name="spine_05"),Weight=0.7),(Item=(Type=Bone,Name="clavicle_r"),Weight=0.3))
    Weight=1
   )
   > &RigUnit_ParentConstraint(
    AdvancedSettings=(InterpolationType=Average,RotationOrderForFilter=XZY)
    Child=(Type=Bone,Name="Collar_side_R")
    Filter=(TranslationFilter=(bX=true,bY=true,bZ=true),RotationFilter=(bX=True,bY=True,bZ=True),ScaleFilter=(bX=True,bY=True,bZ=True))
    Parents=((Item=(Type=Bone,Name="spine_05"),Weight=0.75),(Item=(Type=Bone,Name="clavicle_r"),Weight=0.15))
    Weight=1
   )
   > &RigUnit_ParentConstraint(
    AdvancedSettings=(InterpolationType=Average,RotationOrderForFilter=XZY)
    Child=(Type=Bone,Name="Collar_front_R")
    Filter=(TranslationFilter=(bX=true,bY=true,bZ=true),RotationFilter=(bX=True,bY=True,bZ=True),ScaleFilter=(bX=True,bY=True,bZ=True))
    Parents=((Item=(Type=Bone,Name="spine_05"),Weight=0.8),(Item=(Type=Bone,Name="clavicle_l"),Weight=0.1),(Item=(Type=Bone,Name="clavicle_r"),Weight=0.1))
    Weight=1
   )
   > RigVMFunctionReturnNode()
  ]}
 New Function_1_ContainedGraph {nodes [
   n1 = RigVMFunctionEntryNode()
   > RigVMFunction_Sequence = RigVMAggregateNode()
   RigVMFunctionReturnNode(ExecuteContext=@RigVMFunction_Sequence.C)
   n2 = &RigVMFunction_MathDoubleRemap(SourceMaximum=0.3, SourceMinimum=0, TargetMaximum=1, TargetMinimum=0, Value=@n1.Thigh Side Spr)
   Clamp = &RigVMFunction_MathDoubleClamp(Maximum=0.5, Value=@n2.Result)
   n3 = &RigVMFunction_MathDoubleRemap(SourceMaximum=0.3, SourceMinimum=0, TargetMaximum=1, TargetMinimum=0, Value=@n1.Thigh Fwd Side Spr)
   Remap = &RigVMFunction_MathDoubleRemap(SourceMaximum=0.3, SourceMinimum=0, TargetMaximum=1, TargetMinimum=0, Value=@n1.Thigh Fwd Spr)
   Maximum = RigVMAggregateNode(C=@Clamp.Result, B=@n3.Result, A=@Remap.Result)
   n4 = &RigVMFunction_MathDoubleRemap(SourceMaximum=0.3, SourceMinimum=0, TargetMaximum=1, TargetMinimum=0, Value=@n1.Thigh Rear Spr)
   Subtract = &RigVMFunction_MathDoubleSub(A=1, B=@n4.Result)
   Multiply = &RigVMFunction_MathDoubleMul(A=@Maximum.Result, B=@Subtract.Result)
   &RigUnit_ParentConstraint(
    AdvancedSettings=(InterpolationType=Average,RotationOrderForFilter=XZY)
    Filter=(TranslationFilter=(bX=True,bY=True,bZ=True),RotationFilter=(bX=true,bY=true,bZ=true),ScaleFilter=(bX=True,bY=True,bZ=True))
    Parents=((Item=(Type=Bone,Name="thigh_l"),Weight=1))
    bMaintainOffset=true
    Weight=&RigVMFunction_MathDoubleRemap(SourceMaximum=1, SourceMinimum=0, TargetMaximum=1, TargetMinimum=0.3, Value=@Multiply.Result).Result
    ExecutePin=@RigVMFunction_Sequence.A
    Child=@n1.Bind Front Skirt Joint
    Parents.0.Item=@n1.Thigh Joint
   )
   n5 = &RigVMFunction_MathDoubleRemap(SourceMaximum=0.3, SourceMinimum=0, TargetMaximum=1, TargetMinimum=0, Value=@n1.Thigh Side Rear Spr)
   Maximum_1 = RigVMAggregateNode(
    A=&RigVMFunction_MathDoubleClamp(Maximum=0.3, Value=@n3.Result).Result
    C=&RigVMFunction_MathDoubleClamp(Maximum=0.3, Value=@n5.Result).Result
    B=&RigVMFunction_MathDoubleRemap(SourceMaximum=1, SourceMinimum=0, TargetMaximum=1, TargetMinimum=0, Value=@n1.Thigh Side Spr).Result
   )
   n6 = &RigVMFunction_MathDoubleSub(A=1, B=@Maximum_1.Result)
   > &RigUnit_ParentConstraint(
    AdvancedSettings=(InterpolationType=Average,RotationOrderForFilter=XZY)
    Filter=(TranslationFilter=(bX=True,bY=True,bZ=True),RotationFilter=(bX=true,bY=true,bZ=true),ScaleFilter=(bX=True,bY=True,bZ=True))
    Parents=((Item=(Type=Bone,Name="thigh_l"),Weight=1),(Item=(Type=Bone,Name="None"),Weight=1),(Item=(Type=Bone,Name="None"),Weight=1))
    Weight=1
    bMaintainOffset=true
    Parents.2.Weight=@Maximum_1.Result
    Parents.0.Weight=@n6.Result
    Parents.1.Weight=@n6.Result
    Parents.1.Item=@n1.Bind Front Skirt Joint
    Child=@n1.Bind Side Skirt Joint
    Parents.0.Item=@n1.Bind Rear Skirt Joint
    Parents.2.Item=@n1.Thigh Joint
   )
   Maximum_2 = RigVMAggregateNode(
    A=&RigVMFunction_MathDoubleClamp(Maximum=0.5, Value=@n2.Result).Result
    D=&RigVMFunction_MathDoubleClamp(Maximum=0.5, Value=@Remap.Result).Result
    C=@n4.Result
    B=@n5.Result
   )
   n7 = &RigVMFunction_MathDoubleSub(A=1, B=@Remap.Result)
   > &RigUnit_ParentConstraint(
    AdvancedSettings=(InterpolationType=Average,RotationOrderForFilter=XZY)
    Parents=((Item=(Type=Bone,Name="thigh_l"),Weight=1),(Item=(Type=Bone,Name="None"),Weight=0))
    bMaintainOffset=true
    Parents.1.Item=/Script/ControlRig.RigUnit_HierarchyGetParent(bDefaultParent=True, Child=@n1.Bind Rear Skirt Joint) {is=struct RigVMUnitNode}.Parent
    Weight=&RigVMFunction_MathDoubleRemap(
     SourceMaximum=1
     SourceMinimum=0
     TargetMaximum=1
     TargetMinimum=0.3
     Value=&RigVMFunction_MathDoubleMul(A=@Maximum_2.Result, B=@n7.Result).Result
    ).Result
    Child=@n1.Bind Rear Skirt Joint
    Parents.0.Item=@n1.Thigh Joint
   )
   RigVMFunction_MathVectorDistance = &RigVMFunction_MathVectorDistance(
    A=&RigUnit_GetTransform(Space=LocalSpace, bInitial=False, Item=@n1.Bind Front Skirt Joint).Transform.Translation
    B=&RigUnit_GetTransform(Space=LocalSpace, bInitial=true, Item=@n1.Bind Front Skirt Joint).Transform.Translation
   )
   > &RigUnit_OffsetTransformForItem(
    OffsetTransform=(Rotation=(X=0,Y=0,Z=0,W=1),Translation=(X=0,Y=0,Z=0),Scale3D=(X=1,Y=1,Z=1))
    Weight=1
    bPropagateToChildren=True
    OffsetTransform.Translation=/Script/RigVM.RigVMFunction_MathVectorMul(A=(X=1,Y=0,Z=0), A.X=@RigVMFunction_MathVectorDistance.Result, B=@n1.Push Vector) {is=struct RigVMUnitNode, template="Multiply::Execute(in A,in B,out Result)"}.Result
    Item=@n1.Bind Front Skirt Joint
   )
   n8 = &RigUnit_GetTransform(Space=LocalSpace, bInitial=False, Item=@n1.Bind Side Skirt Joint)
   n9 = &RigUnit_GetTransform(Space=LocalSpace, bInitial=true, Item=@n1.Bind Side Skirt Joint)
   > &RigUnit_OffsetTransformForItem(
    OffsetTransform=(Rotation=(X=0,Y=0,Z=0,W=1),Translation=(X=0,Y=0,Z=0),Scale3D=(X=1,Y=1,Z=1))
    Weight=1
    bPropagateToChildren=True
    OffsetTransform.Translation=&RigVMFunction_MathVectorMul(
     A=(X=1,Y=0,Z=0)
     A.X=&RigVMFunction_MathVectorDistance(A=@n8.Transform.Translation, B=@n9.Transform.Translation).Result
     B=@n1.Push Vector
    ) {template="Multiply::Execute(in A,in B,out Result)"}.Result
    Item=@n1.Bind Side Skirt Joint
   )
   n10 = &RigUnit_GetTransform(Space=LocalSpace, bInitial=False, Item=@n1.Bind Rear Skirt Joint)
   n11 = &RigUnit_GetTransform(Space=LocalSpace, bInitial=true, Item=@n1.Bind Rear Skirt Joint)
   > &RigUnit_OffsetTransformForItem(
    OffsetTransform=(Rotation=(X=0,Y=0,Z=0,W=1),Translation=(X=0,Y=0,Z=0),Scale3D=(X=1,Y=1,Z=1))
    Weight=1
    bPropagateToChildren=True
    OffsetTransform.Translation=&RigVMFunction_MathVectorMul(
     A=(X=1,Y=0,Z=0)
     A.X=&RigVMFunction_MathVectorDistance(A=@n10.Transform.Translation, B=@n11.Transform.Translation).Result
     B=@n1.Push Vector
    ) {template="Multiply::Execute(in A,in B,out Result)"}.Result
    Item=@n1.Bind Rear Skirt Joint
   )
  ]}
 New Function_1_ContainedGraph {nodes [
   RigVMFunctionEntryNode()
   Remap = &RigVMFunction_MathDoubleRemap(SourceMaximum=0.3, SourceMinimum=0, TargetMaximum=1, TargetMinimum=0, Value=RigVMVariableNode() {variable=Rear Knee Bend SPR}.Value)
   n1 = RigVMVariableNode() {variable=Bind Rear Knee}
   > &RigUnit_AimConstraintLocalSpaceOffset(
    Filter=(bX=True,bY=True,bZ=True)
    Parents=((Item=(Type=Bone,Name="ball_l"),Weight=1))
    WorldUp=(Target=(X=0,Y=0,Z=1),Kind=Location,Space=(Type=Bone,Name="ball_l"))
    Weight=@Remap.Result
    Parents.0.Item=RigVMVariableNode() {variable=Ball}.Value
    Child=@n1.Value
    WorldUp.Space=@n1.Value
    AimAxis=RigVMVariableNode() {variable=AimAxis}.Value
   )
   > &RigUnit_ParentConstraint(
    AdvancedSettings=(InterpolationType=Average,RotationOrderForFilter=XZY)
    Parents=((Item=(Type=Bone,Name="None"),Weight=1))
    Parents.0.Item=RigVMVariableNode() {variable=Knee}.Value
    Child=RigVMVariableNode() {variable=Bind Front Knee}.Value
    Weight=RigVMVariableNode() {variable=Front Knee Bend SPR}.Value
   )
   > RigVMFunctionReturnNode()
   n2 = RigVMVariableNode() {variable=Bind Rear Thigh}
   &RigUnit_AimConstraintLocalSpaceOffset(
    ExecutePin=()
    Filter=(bX=True,bY=True,bZ=True)
    Parents=((Item=(Type=Bone,Name="None"),Weight=0.1))
    Weight=1
    WorldUp=(Target=(X=0,Y=0,Z=1),Kind=Location,Space=(Type=None,Name="None"))
    Child=@n2.Value
    WorldUp.Space=@n2.Value
    Parents.0.Item=RigVMVariableNode() {variable=Ball}.Value
   )
   RigVMVariableNode() {variable=Foot Close to Pelvis SPR}
  ]}
 New Function_ContainedGraph {nodes [
   RigVMFunctionEntryNode()
   n1 = &RigVMFunction_MathQuaternionSwingTwist(
    TwistAxis=(X=1,Y=0,Z=0)
    Input=&RigUnit_GetRelativeTransformForItem(Child=(Type=Bone,Name="head"), Parent=(Type=Control,Name="ref_head_ctrl"), bChildInitial=False, bParentInitial=False).RelativeTransform.Rotation
   )
   RigVMFunction_MathQuaternionToEuler = &RigVMFunction_MathQuaternionToEuler(RotationOrder=ZYX, Value=@n1.Swing)
   n2 = &RigVMFunction_MathQuaternionToEuler(
    RotationOrder=ZYX
    Value=&RigVMFunction_MathQuaternionSwingTwist(
     TwistAxis=(X=1,Y=0,Z=0)
     Input=&RigUnit_GetRelativeTransformForItem(Child=(Type=Bone,Name="neck_02"), Parent=(Type=Control,Name="ref_neck_ctrl"), bChildInitial=False, bParentInitial=False).RelativeTransform.Rotation
    ).Swing
   )
   Maximum_2 = RigVMAggregateNode(
    B=&RigVMFunction_MathDoubleRemap(SourceMaximum=60, SourceMinimum=0, TargetMaximum=1, TargetMinimum=0, Value=@RigVMFunction_MathQuaternionToEuler.Result.Y).Result
    C=&RigVMFunction_MathDoubleRemap(SourceMaximum=-20, SourceMinimum=0, TargetMaximum=1, TargetMinimum=0, Value=@n2.Result.Y).Result
    A=&RigVMFunction_MathDoubleRemap(SourceMaximum=-60, SourceMinimum=0, TargetMaximum=1, TargetMinimum=0, Value=RigVMVariableNode() {variable=Clavicle Rot_L}.Value.Y).Result
   )
   > RigVMVariableNode(Value=@Maximum_2.Result) {variable="Head Side L "}
   n3 = &RigVMFunction_MathQuaternionToEuler(RotationOrder=ZYX, Value=@n1.Twist)
   > RigVMVariableNode(Value=&RigVMFunction_MathDoubleRemap(SourceMaximum=-35, SourceMinimum=0, TargetMaximum=1, TargetMinimum=0, Value=@n3.Result.X).Result) {variable=Head Twist L}
   n4 = &RigVMFunction_MathDoubleRemap(SourceMaximum=-70, SourceMinimum=0, TargetMaximum=1, TargetMinimum=0, Value=RigVMVariableNode() {variable=Clavicle Rot_R}.Value.Y)
   Add = RigVMAggregateNode(
    B=&RigVMFunction_MathDoubleRemap(SourceMaximum=-60, SourceMinimum=0, TargetMaximum=1, TargetMinimum=0, Value=@RigVMFunction_MathQuaternionToEuler.Result.Y).Result
    C=&RigVMFunction_MathDoubleRemap(SourceMaximum=-20, SourceMinimum=0, TargetMaximum=1, TargetMinimum=0, Value=@n2.Result.Y).Result
    A=@n4.Result
   )
   Maximum_1 = RigVMAggregateNode(B=@Add.Result, A=@n4.Result)
   > RigVMVariableNode(Value=@Maximum_1.Result) {variable=Head Side R}
   > RigVMVariableNode(Value=&RigVMFunction_MathDoubleRemap(SourceMaximum=30, SourceMinimum=0, TargetMaximum=1, TargetMinimum=0, Value=@n3.Result.X).Result) {variable=Head Twist R}
   RigVMFunction_AnimEasing = /Script/RigVM.RigVMFunction_AnimEasing(
    SourceMaximum=1
    SourceMinimum=0
    TargetMaximum=1
    TargetMinimum=0
    Type=CubicEaseIn
    Value=&RigVMFunction_MathDoubleRemap(SourceMaximum=-35, SourceMinimum=0, TargetMaximum=1, TargetMinimum=0, Value=@RigVMFunction_MathQuaternionToEuler.Result.Z).Result
   ) {is=struct RigVMUnitNode}
   > RigVMVariableNode(Value=&RigVMFunction_MathDoubleAdd(
     B=&RigVMFunction_MathDoubleRemap(SourceMaximum=-35, SourceMinimum=0, TargetMaximum=1, TargetMinimum=0, Value=@n2.Result.Z).Result
     A=@RigVMFunction_AnimEasing.Result
    ).Result) {variable=Head Back C}
   > RigVMFunctionReturnNode()
  ]}
 New Function_ContainedGraph {nodes [
   RigVMFunctionEntryNode()
   > RigVMFunction_Sequence = RigVMAggregateNode()
   Additional Skirt Posing = RigVMFunctionReferenceNode(ExecuteContext=@RigVMFunction_Sequence.F)
   n1 = &RigVMFunction_ControlFlowBranch(ExecuteContext=@RigVMFunction_Sequence.G, Condition=RigVMVariableNode() {variable=Is DBNO Carried}.Value)
   &RigUnit_ModifyTransforms(
    ItemToModify="((Item=(Type=Bone,Name=\"base_C_skirt_b\"),Transform=(Rotation=(X=0.2607,Y=0.2664,Z=-0.0027,W=0.9279),Translation=(X=1.053,Y=-1.0951,Z=-8.3715),Scale3D=(X=1,Y=1,Z=1))),(Item=(Type=Bone,Name=\"base_C_skirt_a\"),Transform=(Rotation=(X=0.0956,Y=-0.5533,Z=-0.1447, ...(truncated)"
    Weight=1
    WeightMinimum=0
    ExecutePin=@n1.True
   )
   > &RigUnit_ModifyTransforms(
    ItemToModify="((Item=(Type=Bone,Name=\"dyn_L_skirt_b_4\"),Transform=(Rotation=(X=0.0035,Y=0.0295,Z=-0.1185,W=0.9925),Translation=(X=0,Y=0,Z=0),Scale3D=(X=1,Y=1,Z=1))),(Item=(Type=Bone,Name=\"dyn_C_skirt_a_4\"),Transform=(Rotation=(X=0,Y=0.4595,Z=0,W=0.8882),Translation=(X=0 ...(truncated)"
    Weight=0.5
    WeightMinimum=0
   )
   RigVMFunctionReturnNode(ExecuteContext=@RigVMFunction_Sequence.H)
   n2 = &RigVMFunction_ControlFlowBranch(Condition=RigVMVariableNode() {variable=Pub DBNO}.Value, ExecuteContext=@RigVMFunction_Sequence.B)
   n3 = &RigVMFunction_Sequence(ExecuteContext=@n2.False)
   Solve Thigh_2 = RigVMFunctionReferenceNode(
    Bind Front Skirt Joint=(Type=Bone,Name="base_R_skirt_a")
    Bind Rear Skirt Joint=(Type=Bone,Name="base_R_skirt_c")
    Bind Side Skirt Joint=(Type=Bone,Name="base_R_skirt_b")
    Push Vector=(X=1.1,Y=0,Z=0)
    Thigh Joint=(Type=Bone,Name="thigh_r")
    ExecuteContext=@n3.B
    Thigh Fwd Spr=RigVMVariableNode() {variable=Leg Rom R Thigh Raised Fwd}.Value
    Thigh Fwd Side Spr=RigVMVariableNode() {variable=Leg ROM_R Thigh Raised Fwd Side Inbetween}.Value
    Thigh Side Spr=RigVMVariableNode() {variable=Leg ROM_R Thigh Raised Side}.Value
    Thigh Side Rear Spr=RigVMVariableNode() {variable=Leg ROM_R Thigh Raised Side Rear Inbetween}.Value
    Thigh Rear Spr=RigVMVariableNode() {variable=Leg ROM_R Thigh Raised Rear}.Value
   )
   > Solve Knee_2 = RigVMFunctionReferenceNode(
    AimAxis=(X=1,Y=0,Z=0)
    Ball=(Type=Bone,Name="foot_r")
    Bind Front Knee=(Type=Bone,Name="dyn_R_skirt_a_3")
    Bind Rear Knee=(Type=Bone,Name="base_R_skirt_c")
    Bind Rear Thigh=(Type=Bone,Name="base_R_skirt_c")
    Knee=(Type=Bone,Name="calf_r")
    Rear Knee Bend SPR=RigVMVariableNode() {variable=Leg ROM_R Thigh Rear Knee Bend}.Value
    Foot Close to Pelvis SPR=RigVMVariableNode() {variable=Leg ROM_R Foot Near Pelvis}.Value
    Front Knee Bend SPR=RigVMVariableNode() {variable=Leg ROM_R Thigh Fwd Knee Bend}.Value
   )
   Solve Thigh_1 = RigVMFunctionReferenceNode(
    Bind Front Skirt Joint=(Type=Bone,Name="base_L_skirt_a")
    Bind Rear Skirt Joint=(Type=Bone,Name="base_L_skirt_c")
    Bind Side Skirt Joint=(Type=Bone,Name="base_L_skirt_b")
    Push Vector=(X=1.1,Y=0,Z=0)
    Thigh Joint=(Type=Bone,Name="thigh_l")
    Thigh Fwd Spr=RigVMVariableNode() {variable=Leg Rom L Thigh Raised Fwd}.Value
    Thigh Fwd Side Spr=RigVMVariableNode() {variable=Leg ROM_L Thigh Raised Fwd Side Inbetween}.Value
    Thigh Side Spr=RigVMVariableNode() {variable=Leg ROM_L Thigh Raised Side}.Value
    Thigh Side Rear Spr=RigVMVariableNode() {variable=Leg ROM_L Thigh Raised Side Rear Inbetween}.Value
    Thigh Rear Spr=RigVMVariableNode() {variable=Leg ROM_L Thigh Raised Rear}.Value
    ExecuteContext=@n3.A
   )
   > Solve Knee_1 = RigVMFunctionReferenceNode(
    AimAxis=(X=1,Y=0,Z=0)
    Ball=(Type=Bone,Name="foot_l")
    Bind Front Knee=(Type=Bone,Name="dyn_L_skirt_a_3")
    Bind Rear Knee=(Type=Bone,Name="base_L_skirt_c")
    Bind Rear Thigh=(Type=Bone,Name="base_L_skirt_c")
    Knee=(Type=Bone,Name="calf_l")
    Front Knee Bend SPR=RigVMVariableNode() {variable=Leg ROM_L Thigh Fwd Knee Bend}.Value
    Rear Knee Bend SPR=RigVMVariableNode() {variable=Leg ROM_L Thigh Rear Knee Bend}.Value
    Foot Close to Pelvis SPR=RigVMVariableNode() {variable=Leg ROM_L Foot Near Pelvis}.Value
   )
   n4 = &RigVMFunction_Sequence(ExecuteContext=@n2.True)
   Solve Thigh_4 = RigVMFunctionReferenceNode(
    Bind Front Skirt Joint=(Type=Bone,Name="base_R_skirt_a")
    Bind Rear Skirt Joint=(Type=Bone,Name="base_R_skirt_c")
    Bind Side Skirt Joint=(Type=Bone,Name="base_R_skirt_b")
    Push Vector=(X=1.1,Y=0,Z=0)
    Thigh Joint=(Type=Bone,Name="thigh_r")
    ExecuteContext=@n4.B
    Thigh Fwd Spr=RigVMVariableNode() {variable=Leg Rom R Thigh Raised Fwd}.Value
    Thigh Fwd Side Spr=RigVMVariableNode() {variable=Leg ROM_R Thigh Raised Fwd Side Inbetween}.Value
    Thigh Side Spr=RigVMVariableNode() {variable=Leg ROM_R Thigh Raised Side}.Value
    Thigh Side Rear Spr=RigVMVariableNode() {variable=Leg ROM_R Thigh Raised Side Rear Inbetween}.Value
    Thigh Rear Spr=RigVMVariableNode() {variable=Leg ROM_R Thigh Raised Rear}.Value
   )
   > Solve Knee_4 = RigVMFunctionReferenceNode(
    AimAxis=(X=1,Y=0,Z=0)
    Ball=(Type=Bone,Name="foot_r")
    Bind Front Knee=(Type=Bone,Name="None")
    Bind Rear Knee=(Type=Bone,Name="base_R_skirt_c")
    Bind Rear Thigh=(Type=Bone,Name="base_R_skirt_c")
    Knee=(Type=Bone,Name="calf_r")
    Rear Knee Bend SPR=RigVMVariableNode() {variable=Leg ROM_R Thigh Rear Knee Bend}.Value
    Foot Close to Pelvis SPR=RigVMVariableNode() {variable=Leg ROM_R Foot Near Pelvis}.Value
    Front Knee Bend SPR=RigVMVariableNode() {variable=Leg ROM_R Thigh Fwd Knee Bend}.Value
   )
   Solve Thigh_3 = RigVMFunctionReferenceNode(
    Bind Front Skirt Joint=(Type=Bone,Name="base_L_skirt_a")
    Bind Rear Skirt Joint=(Type=Bone,Name="base_L_skirt_c")
    Bind Side Skirt Joint=(Type=Bone,Name="base_L_skirt_b")
    Push Vector=(X=1.1,Y=0,Z=0)
    Thigh Joint=(Type=Bone,Name="thigh_l")
    Thigh Fwd Spr=RigVMVariableNode() {variable=Leg Rom L Thigh Raised Fwd}.Value
    Thigh Fwd Side Spr=RigVMVariableNode() {variable=Leg ROM_L Thigh Raised Fwd Side Inbetween}.Value
    Thigh Side Spr=RigVMVariableNode() {variable=Leg ROM_L Thigh Raised Side}.Value
    Thigh Side Rear Spr=RigVMVariableNode() {variable=Leg ROM_L Thigh Raised Side Rear Inbetween}.Value
    Thigh Rear Spr=RigVMVariableNode() {variable=Leg ROM_L Thigh Raised Rear}.Value
    ExecuteContext=@n4.A
   )
   > Solve Knee_3 = RigVMFunctionReferenceNode(
    AimAxis=(X=1,Y=0,Z=0)
    Ball=(Type=Bone,Name="foot_l")
    Bind Front Knee=(Type=Bone,Name="None")
    Bind Rear Knee=(Type=Bone,Name="base_L_skirt_c")
    Bind Rear Thigh=(Type=Bone,Name="base_L_skirt_c")
    Knee=(Type=Bone,Name="calf_l")
    Front Knee Bend SPR=RigVMVariableNode() {variable=Leg ROM_L Thigh Fwd Knee Bend}.Value
    Rear Knee Bend SPR=RigVMVariableNode() {variable=Leg ROM_L Thigh Rear Knee Bend}.Value
    Foot Close to Pelvis SPR=RigVMVariableNode() {variable=Leg ROM_L Foot Near Pelvis}.Value
   )
   Leg ROM_1 = RigVMFunctionReferenceNode(
    Include Ankles=false
    Include Distance Checks=true
    Include Knees=true
    Interp Output Speed=5
    Interp Results=false
    SPR Optional Parent Item=(Type=Bone,Name="pelvis")
    Result Alpha=&RigVMFunction_KalmanFloat(BufferSize=4, Value=RigVMVariableNode() {variable=Leg ROM_Result Alpha}.Value) {template="KalmanFilter::Execute(in Value,in BufferSize,out Result)"}.Result
    ExecuteContext=@RigVMFunction_Sequence.A
   )
   > RigVMVariableNode(Value=@"Leg ROM_1".L Thigh Raised Fwd) {variable=Leg Rom L Thigh Raised Fwd}
   > RigVMVariableNode(Value=@"Leg ROM_1".R Thigh Raised Fwd) {variable=Leg Rom R Thigh Raised Fwd}
   > RigVMVariableNode(Value=@"Leg ROM_1".L Thigh Raised Fwd Side Inbetween) {variable=Leg ROM_L Thigh Raised Fwd Side Inbetween}
   > RigVMVariableNode(Value=@"Leg ROM_1".R Thigh Raised Fwd Side Inbetween) {variable=Leg ROM_R Thigh Raised Fwd Side Inbetween}
   > RigVMVariableNode(Value=@"Leg ROM_1".L Thigh Raised Side) {variable=Leg ROM_L Thigh Raised Side}
   > RigVMVariableNode(Value=@"Leg ROM_1".R Thigh Raised Side) {variable=Leg ROM_R Thigh Raised Side}
   > RigVMVariableNode(Value=@"Leg ROM_1".L Thigh Raised Side Rear Inbetween) {variable=Leg ROM_L Thigh Raised Side Rear Inbetween}
   > RigVMVariableNode(Value=@"Leg ROM_1".R Thigh Raised Side Rear Inbetween) {variable=Leg ROM_R Thigh Raised Side Rear Inbetween}
   > RigVMVariableNode(Value=@"Leg ROM_1".L Thigh Raised Rear) {variable=Leg ROM_L Thigh Raised Rear}
   > RigVMVariableNode(Value=@"Leg ROM_1".R Thigh Raised Rear) {variable=Leg ROM_R Thigh Raised Rear}
   > RigVMVariableNode(Value=@"Leg ROM_1".L Thigh Across R) {variable=Leg ROM_L Thigh Across R}
   > RigVMVariableNode(Value=@"Leg ROM_1".R Thigh Across L) {variable=Leg ROM_R Thigh Across L}
   > RigVMVariableNode(Value=@"Leg ROM_1".L Leg Across R Leg Side) {variable=Leg ROM_L Leg Across R Leg Side}
   > RigVMVariableNode(Value=@"Leg ROM_1".R Leg Across L Leg Side) {variable=Leg ROM_R Leg Across L Leg Side}
   > RigVMVariableNode(Value=@"Leg ROM_1".L Thigh Fwd Knee Bend) {variable=Leg ROM_L Thigh Fwd Knee Bend}
   > RigVMVariableNode(Value=@"Leg ROM_1".R Thigh Fwd Knee Bend) {variable=Leg ROM_R Thigh Fwd Knee Bend}
   > RigVMVariableNode(Value=@"Leg ROM_1".L Thigh Rear Knee Bend) {variable=Leg ROM_L Thigh Rear Knee Bend}
   > RigVMVariableNode(Value=@"Leg ROM_1".R Thigh Rear Knee Bend) {variable=Leg ROM_R Thigh Rear Knee Bend}
   > RigVMVariableNode(Value=@"Leg ROM_1".L Foot Near Pelvis) {variable=Leg ROM_L Foot Near Pelvis}
   > RigVMVariableNode(Value=@"Leg ROM_1".R Foot Near Pelvis) {variable=Leg ROM_R Foot Near Pelvis}
   > RigVMVariableNode(Value=@"Leg ROM_1".L Foot on Thigh R) {variable=Leg ROM_L Foot on Thigh R}
   Modify Transforms with Mirror_10 = RigVMFunctionReferenceNode(
    Base Modify Data="((Item=(Type=Bone,Name=\"base_L_skirt_a\"),Transform=(Rotation=(X=-0.0057,Y=-0.0432,Z=-0.1304,W=0.9905),Translation=(X=0,Y=-10,Z=5),Scale3D=(X=1,Y=1,Z=1))),(Item=(Type=Bone,Name=\"base_R_skirt_c\"),Transform=(Rotation=(X=0.0076,Y=-0.0868,Z=0.0868,W=0.9924),Tra ...(truncated)"
    Mirror Modify Data="((Item=(Type=Bone,Name=\"base_R_skirt_a\"),Transform=(Rotation=(X=0,Y=0,Z=0.0436,W=0.999),Translation=(X=0,Y=-5,Z=8),Scale3D=(X=1,Y=1,Z=1))),(Item=(Type=Bone,Name=\"base_L_skirt_c\"),Transform=(Rotation=(X=-0.0076,Y=-0.0868,Z=-0.0868,W=0.9924),Translation=(X=0 ...(truncated)"
    Weight=RigVMVariableNode() {variable=Leg Rom L Thigh Raised Fwd}.Value
    "Mirror Weight "=RigVMVariableNode() {variable=Leg Rom R Thigh Raised Fwd}.Value
    ExecuteContext=@RigVMFunction_Sequence.C
   )
   > Modify Transforms with Mirror_7 = RigVMFunctionReferenceNode(
    Base Modify Data="((Item=(Type=Bone,Name=\"base_L_skirt_a\"),Transform=(Rotation=(X=0,Y=0,Z=0.0872,W=0.9962),Translation=(X=0,Y=10,Z=-4),Scale3D=(X=1,Y=1,Z=1))),(Item=(Type=Bone,Name=\"base_L_skirt_b\"),Transform=(Rotation=(X=-0.0151,Y=-0.173,Z=-0.0858,W=0.9811),Translation=(X= ...(truncated)"
    Mirror Modify Data="((Item=(Type=Bone,Name=\"base_R_skirt_a\"),Transform=(Rotation=(X=0.0038,Y=0.0435,Z=-0.0871,W=0.9952),Translation=(X=0,Y=0,Z=8),Scale3D=(X=1,Y=1,Z=1))),(Item=(Type=Bone,Name=\"base_R_skirt_b\"),Transform=(Rotation=(X=0,Y=0,Z=0,W=1),Translation=(X=0,Y=6,Z=0),Sc ...(truncated)"
    Weight=RigVMVariableNode() {variable=Leg ROM_L Thigh Raised Fwd Side Inbetween}.Value
    "Mirror Weight "=RigVMVariableNode() {variable=Leg ROM_R Thigh Raised Fwd Side Inbetween}.Value
   )
   > Modify Transforms with Mirror_3 = RigVMFunctionReferenceNode(
    Base Modify Data="((Item=(Type=Bone,Name=\"base_L_skirt_a\"),Transform=(Rotation=(X=0.3981,Y=-0.0264,Z=-0.017,W=0.9168),Translation=(X=0,Y=0,Z=0),Scale3D=(X=1,Y=1,Z=1))),(Item=(Type=Bone,Name=\"base_R_skirt_a\"),Transform=(Rotation=(X=0.0926,Y=0.0432,Z=0.4204,W=0.9016),Translat ...(truncated)"
    Mirror Modify Data="((Item=(Type=Bone,Name=\"base_R_skirt_a\"),Transform=(Rotation=(X=-0.3996,Y=0.0111,Z=0.0619,W=0.9145),Translation=(X=0,Y=0,Z=0),Scale3D=(X=1,Y=1,Z=1))),(Item=(Type=Bone,Name=\"base_L_skirt_a\"),Transform=(Rotation=(X=0.0657,Y=-0.0783,Z=-0.762,W=0.6394),Transla ...(truncated)"
    Weight=RigVMVariableNode() {variable=Leg ROM_L Thigh Across R}.Value
    "Mirror Weight "=RigVMVariableNode() {variable=Leg ROM_R Thigh Across L}.Value
   )
   > &RigUnit_ModifyTransforms(
    ItemToModify="((Item=(Type=Bone,Name=\"dyn_R_skirt_a_1\"),Transform=(Rotation=(X=0,Y=0,Z=-0.5736,W=0.8192),Translation=(X=0,Y=0,Z=0),Scale3D=(X=1,Y=1,Z=1))),(Item=(Type=Bone,Name=\"dyn_R_skirt_b_1\"),Transform=(Rotation=(X=-0.0665,Y=-0.1604,Z=-0.3769,W=0.9098),Translation=( ...(truncated)"
    WeightMinimum=0
    Weight=RigVMVariableNode() {variable=Leg ROM_L Foot on Thigh R}.Value
   )
   RigVMFunction_MathBoolOr = &RigVMFunction_MathBoolOr(A=RigVMVariableNode() {variable="Is Knee Sliding "}.Value, B=RigVMVariableNode() {variable="Is Baseball Sliding "}.Value)
   RigVMFunction_ControlFlowBranch = &RigVMFunction_ControlFlowBranch(Condition=@RigVMFunction_MathBoolOr.Result, ExecuteContext=@RigVMFunction_Sequence.E)
   RigVMVariableNode(Value=1, ExecuteContext=@RigVMFunction_ControlFlowBranch.False) {variable=Leg ROM_Result Alpha}
   n5 = RigVMVariableNode(ExecuteContext=@RigVMFunction_ControlFlowBranch.True) {variable=Leg ROM_Result Alpha}
   n6 = &RigVMFunction_Sequence(ExecuteContext=@n5.ExecuteContext)
   n7 = &RigVMFunction_ControlFlowBranch(ExecuteContext=@n6.A, Condition=RigVMVariableNode() {variable="Is Knee Sliding "}.Value)
   Subtract = &RigVMFunction_MathDoubleSub(A=1, B=RigVMVariableNode() {variable=Leg ROM_Result Alpha}.Value)
   n8 = &RigVMFunction_KalmanFloat(BufferSize=4, Value=@Subtract.Result) {template="KalmanFilter::Execute(in Value,in BufferSize,out Result)"}
   Modify Transforms with Mirror_6 = RigVMFunctionReferenceNode(
    Base Modify Data="((Item=(Type=Bone,Name=\"base_L_skirt_b\"),Transform=(Rotation=(X=0,Y=-0.1736,Z=0,W=0.9848),Translation=(X=0,Y=0,Z=0),Scale3D=(X=1,Y=1,Z=1))),(Item=(Type=Bone,Name=\"base_L_skirt_c\"),Transform=(Rotation=(X=0,Y=-0.3827,Z=0,W=0.9239),Translation=(X=-2,Y=-5,Z=0) ...(truncated)"
    Mirror Modify Data="((Item=(Type=Bone,Name=\"base_R_skirt_b\"),Transform=(Rotation=(X=0,Y=0,Z=0,W=1),Translation=(X=0,Y=0,Z=0),Scale3D=(X=1,Y=1,Z=1))),(Item=(Type=Bone,Name=\"base_R_skirt_c\"),Transform=(Rotation=(X=-0.0552,Y=-0.1183,Z=-0.419,W=0.8986),Translation=(X=-7,Y=0,Z=0), ...(truncated)"
    "Mirror Weight "=@n8.Result
    Weight=@n8.Result
    ExecuteContext=@n7.True
   )
   > &RigUnit_ModifyTransforms(
    ItemToModify=((Item=(Type=Bone,Name="spine_01"),Transform=(Rotation=(X=0,Y=0,Z=0,W=1),Translation=(X=3,Y=0,Z=0),Scale3D=(X=1,Y=1,Z=1))))
    Weight=1
    WeightMinimum=0
   )
   n9 = &RigVMFunction_ControlFlowBranch(ExecuteContext=@n6.B, Condition=RigVMVariableNode() {variable="Is Baseball Sliding "}.Value)
   Modify Transforms with Mirror_5 = RigVMFunctionReferenceNode(
    Base Modify Data="((Item=(Type=Bone,Name=\"base_L_skirt_b\"),Transform=(Rotation=(X=0,Y=-0.4226,Z=0,W=0.9063),Translation=(X=0,Y=0,Z=0),Scale3D=(X=1,Y=1,Z=1))),(Item=(Type=Bone,Name=\"base_L_skirt_c\"),Transform=(Rotation=(X=-0.3237,Y=-0.6943,Z=-0.2717,W=0.5826),Translation=(X= ...(truncated)"
    Mirror Modify Data="((Item=(Type=Bone,Name=\"base_R_skirt_b\"),Transform=(Rotation=(X=-0.0868,Y=-0.4924,Z=-0.1504,W=0.8529),Translation=(X=0,Y=0,Z=-5),Scale3D=(X=1,Y=1,Z=1))),(Item=(Type=Bone,Name=\"base_R_skirt_c\"),Transform=(Rotation=(X=0,Y=0,Z=-0.5,W=0.866),Translation=(X=0,Y ...(truncated)"
    "Mirror Weight "=@n8.Result
    Weight=@n8.Result
    ExecuteContext=@n9.True
   )
   > &RigUnit_ModifyTransforms(
    ItemToModify=((Item=(Type=Bone,Name="Skirt_Master"),Transform=(Rotation=(X=0,Y=0,Z=0,W=1),Translation=(X=4,Y=0,Z=0),Scale3D=(X=1,Y=1,Z=1))))
    Weight=1
    WeightMinimum=0
   )
   /Script/ControlRig.RigUnit_SetCurveValue(Curve=CR Thigh Raised Front L, Value=RigVMVariableNode() {variable=Leg Rom L Thigh Raised Fwd}.Value, ExecutePin=@RigVMFunction_Sequence.D) {is=struct RigVMUnitNode}
   > &RigUnit_SetCurveValue(Curve="CR ThighRaised  Front R", Value=RigVMVariableNode() {variable=Leg Rom R Thigh Raised Fwd}.Value)
   > &RigUnit_SetCurveValue(Curve=CR Rear Knee Bend L, Value=RigVMVariableNode() {variable=Leg ROM_R Thigh Fwd Knee Bend}.Value)
   > &RigUnit_SetCurveValue(Curve="CR  Rear Knee Bend R", Value=RigVMVariableNode() {variable=Leg ROM_R Thigh Rear Knee Bend}.Value)
   > &RigUnit_SetCurveValue(Curve="CR Foot Near Pelvis L ", Value=RigVMVariableNode() {variable=Leg ROM_L Foot Near Pelvis}.Value)
   > &RigUnit_SetCurveValue(Curve=CR Foot Near Pelvis R, Value=RigVMVariableNode() {variable=Leg ROM_R Foot Near Pelvis}.Value)
   > &RigUnit_SetCurveValue(Curve=CR Thigh L Across R, Value=RigVMVariableNode() {variable=Leg ROM_L Thigh Across R}.Value)
   > &RigUnit_SetCurveValue(Curve=CR Thigh L Across R, Value=RigVMVariableNode() {variable=Leg ROM_R Thigh Across L}.Value)
  ]}
 New Function_ContainedGraph {nodes [
   n1 = RigVMFunctionEntryNode()
   RigVMFunction_Sequence = &RigVMFunction_Sequence(ExecuteContext=@n1.ExecuteContext)
   n2 = &RigUnit_SphericalPoseReader(
    ActiveRegionScaleFactors=(PositiveWidth=2,NegativeWidth=2,PositiveHeight=2,NegativeHeight=1)
    ActiveRegionSize=0.1
    Debug=(bDrawDebug=false,bDraw2D=False,bDrawLocalAxes=False,DebugScale=25,DebugSegments=20,DebugThickness=0.25)
    DriverAxis=(X=-1,Y=0,Z=0)
    DriverItem=(Type=Bone,Name="thigh_l")
    FalloffRegionScaleFactors=(PositiveWidth=2,NegativeWidth=2,PositiveHeight=4,NegativeHeight=1)
    FalloffSize=0.3
    FlipHeightScaling=false
    FlipWidthScaling=false
    OptionalParentItem=(Type=Bone,Name="None")
    RotationOffset=(X=0,Y=90,Z=80)
    ExecutePin=@RigVMFunction_Sequence.A
   )
   > n3 = &RigUnit_SphericalPoseReader(
    ActiveRegionScaleFactors=(PositiveWidth=2,NegativeWidth=2,PositiveHeight=2,NegativeHeight=1)
    ActiveRegionSize=0.1
    Debug=(bDrawDebug=false,bDraw2D=False,bDrawLocalAxes=False,DebugScale=25,DebugSegments=20,DebugThickness=0.25)
    DriverAxis=(X=1,Y=0,Z=0)
    DriverItem=(Type=Bone,Name="thigh_r")
    FalloffRegionScaleFactors=(PositiveWidth=2,NegativeWidth=2,PositiveHeight=4,NegativeHeight=1)
    FalloffSize=0.3
    FlipHeightScaling=false
    FlipWidthScaling=false
    OptionalParentItem=(Type=Bone,Name="None")
    RotationOffset=(X=180,Y=90,Z=80)
   )
   > n4 = &RigUnit_SphericalPoseReader(
    ActiveRegionScaleFactors=(PositiveWidth=4,NegativeWidth=3,PositiveHeight=2.5,NegativeHeight=2)
    ActiveRegionSize=0.1
    Debug=(bDrawDebug=false,bDraw2D=False,bDrawLocalAxes=False,DebugScale=25,DebugSegments=20,DebugThickness=0.25)
    DriverAxis=(X=-1,Y=0,Z=0)
    DriverItem=(Type=Bone,Name="thigh_l")
    FalloffRegionScaleFactors=(PositiveWidth=1,NegativeWidth=4,PositiveHeight=1,NegativeHeight=0.5)
    FalloffSize=0.4
    FlipHeightScaling=False
    FlipWidthScaling=False
    OptionalParentItem=(Type=Bone,Name="None")
    RotationOffset=(X=100,Y=0,Z=-20)
   )
   > n5 = &RigUnit_SphericalPoseReader(
    ActiveRegionScaleFactors=(PositiveWidth=4,NegativeWidth=3,PositiveHeight=2.5,NegativeHeight=2)
    ActiveRegionSize=0.1
    Debug=(bDrawDebug=false,bDraw2D=False,bDrawLocalAxes=False,DebugScale=25,DebugSegments=20,DebugThickness=0.25)
    DriverAxis=(X=1,Y=0,Z=0)
    DriverItem=(Type=Bone,Name="thigh_r")
    FalloffRegionScaleFactors=(PositiveWidth=1,NegativeWidth=4,PositiveHeight=1,NegativeHeight=0.5)
    FalloffSize=0.4
    FlipHeightScaling=False
    FlipWidthScaling=true
    OptionalParentItem=(Type=Bone,Name="None")
    RotationOffset=(X=-80,Y=0,Z=-20)
   )
   > n6 = &RigUnit_SphericalPoseReader(
    ActiveRegionScaleFactors=(PositiveWidth=3,NegativeWidth=2,PositiveHeight=2,NegativeHeight=2)
    ActiveRegionSize=0.1
    Debug=(bDrawDebug=false,bDraw2D=False,bDrawLocalAxes=False,DebugScale=25,DebugSegments=20,DebugThickness=0.25)
    DriverAxis=(X=-1,Y=0,Z=0)
    DriverItem=(Type=Bone,Name="thigh_l")
    FalloffRegionScaleFactors=(PositiveWidth=2,NegativeWidth=2,PositiveHeight=2,NegativeHeight=2)
    FalloffSize=0.25
    FlipHeightScaling=false
    FlipWidthScaling=false
    OptionalParentItem=(Type=Bone,Name="None")
    RotationOffset=(X=180,Y=-40,Z=0)
   )
   > n7 = &RigUnit_SphericalPoseReader(
    ActiveRegionScaleFactors=(PositiveWidth=3,NegativeWidth=2,PositiveHeight=2,NegativeHeight=2)
    ActiveRegionSize=0.1
    Debug=(bDrawDebug=false,bDraw2D=False,bDrawLocalAxes=False,DebugScale=25,DebugSegments=20,DebugThickness=0.25)
    DriverAxis=(X=1,Y=0,Z=0)
    DriverItem=(Type=Bone,Name="thigh_r")
    FalloffRegionScaleFactors=(PositiveWidth=2,NegativeWidth=2,PositiveHeight=2,NegativeHeight=2)
    FalloffSize=0.25
    FlipHeightScaling=False
    FlipWidthScaling=true
    OptionalParentItem=(Type=Bone,Name="None")
    RotationOffset=(X=0,Y=-40,Z=0)
   )
   RigVMFunction_MathQuaternionToEuler = &RigVMFunction_MathQuaternionToEuler(
    RotationOrder=ZYX
    Value=&RigUnit_GetRelativeTransformForItem(Child=(Type=Bone,Name="calf_l"), Parent=(Type=Bone,Name="thigh_l"), bChildInitial=False, bParentInitial=False).RelativeTransform.Rotation
   )
   Remap = &RigVMFunction_MathDoubleRemap(SourceMaximum=110, SourceMinimum=25, TargetMaximum=1, TargetMinimum=0, Value=@RigVMFunction_MathQuaternionToEuler.Result.Z)
   > RigVMVariableNode(Value=@Remap.Result) {variable=L calfBend Alpha}
   n8 = &RigUnit_GetRelativeTransformForItem(Child=(Type=Bone,Name="calf_r"), Parent=(Type=Bone,Name="thigh_r"), bChildInitial=False, bParentInitial=False)
   > RigVMVariableNode(Value=&RigVMFunction_MathDoubleRemap(
     SourceMaximum=110
     SourceMinimum=25
     TargetMaximum=1
     TargetMinimum=0
     Value=&RigVMFunction_MathQuaternionToEuler(RotationOrder=ZYX, Value=@n8.RelativeTransform.Rotation).Result.Z
    ).Result) {variable=R calfBend Alpha}
   n9 = &RigUnit_GetRelativeTransformForItem(Child=(Type=Bone,Name="thigh_l"), Parent=(Type=Bone,Name="pelvis"), bChildInitial=False, bParentInitial=False)
   > RigVMVariableNode(Value=&RigVMFunction_MathDoubleRemap(
     SourceMaximum=-90
     SourceMinimum=-50
     TargetMaximum=1
     TargetMinimum=0
     Value=&RigVMFunction_MathQuaternionToEuler(RotationOrder=ZYX, Value=@n9.RelativeTransform.Rotation).Result.Z
    ).Result) {variable=L thighVeryForward}
   Absolute = &RigVMFunction_MathDoubleAbs(Value=&RigVMFunction_MathQuaternionToEuler(
     RotationOrder=ZYX
     Value=&RigUnit_GetRelativeTransformForItem(Child=(Type=Bone,Name="thigh_r"), Parent=(Type=Bone,Name="pelvis"), bChildInitial=False, bParentInitial=False).RelativeTransform.Rotation
    ).Result.Z)
   > RigVMVariableNode(Value=&RigVMFunction_MathDoubleRemap(SourceMaximum=90, SourceMinimum=130, TargetMaximum=1, TargetMinimum=0, Value=@Absolute.Result).Result) {variable=R thighVeryForward}
   n10 = &RigUnit_GetRelativeTransformForItem(Child=(Type=Bone,Name="spine_01"), Parent=(Type=Bone,Name="root"), bChildInitial=False, bParentInitial=false)
   n11 = &RigVMFunction_MathDoubleAbs(Value=&RigVMFunction_MathQuaternionToEuler(RotationOrder=XZY, Value=@n10.RelativeTransform.Rotation).Result.Z)
   n12 = &RigVMFunction_MathDoubleRemap(SourceMaximum=30, SourceMinimum=60, TargetMaximum=1, TargetMinimum=0, Value=@n10.RelativeTransform.Translation.Z)
   > RigVMVariableNode(Value=&RigVMFunction_MathDoubleMul(
     A=&RigVMFunction_MathDoubleRemap(SourceMaximum=60, SourceMinimum=35, TargetMaximum=0, TargetMinimum=1, Value=@n11.Result).Result
     B=@n12.Result
    ).Result) {variable=pelvisLowToGround_translateRotate}
   > RigVMVariableNode(Value=&RigVMFunction_MathDoubleSub(A=1, B=@n12.Result).Result) {variable=pelvisLowToGround_INVERSE}
   > RigVMVariableNode(Value=@n2.OutputParam) {variable=L thighBack Alpha}
   > RigVMVariableNode(Value=@n3.OutputParam) {variable=R thighBack Alpha}
   > RigVMVariableNode(Value=@n4.OutputParam) {variable=L thighForward Alpha}
   > RigVMVariableNode(Value=@n5.OutputParam) {variable=R thighForward Alpha}
   > RigVMVariableNode(Value=@n6.OutputParam) {variable=L thighSide Alpha}
   > RigVMVariableNode(Value=@n7.OutputParam) {variable=R thighSide Alpha}
   > &RigUnit_ParentConstraint(
    AdvancedSettings=(InterpolationType=Average,RotationOrderForFilter=XZY)
    Child=(Type=Bone,Name="dyn_L_skirt_a_1")
    Filter=(TranslationFilter=(bX=true,bY=true,bZ=true),RotationFilter=(bX=True,bY=True,bZ=True),ScaleFilter=(bX=True,bY=True,bZ=True))
    Parents=((Item=(Type=Bone,Name="thigh_l"),Weight=1),(Item=(Type=Bone,Name="pelvis"),Weight=0))
    Weight=0.5
   )
   > &RigUnit_ParentConstraint(
    AdvancedSettings=(InterpolationType=Average,RotationOrderForFilter=XZY)
    Child=(Type=Bone,Name="dyn_L_skirt_a_3")
    Filter=(TranslationFilter=(bX=true,bY=true,bZ=true),RotationFilter=(bX=True,bY=True,bZ=True),ScaleFilter=(bX=True,bY=True,bZ=True))
    Parents=((Item=(Type=Bone,Name="calf_l"),Weight=0.5),(Item=(Type=Bone,Name="thigh_l"),Weight=0.5))
    Weight=0.5
   )
   > &RigUnit_ParentConstraint(
    AdvancedSettings=(InterpolationType=Average,RotationOrderForFilter=XZY)
    Child=(Type=Bone,Name="dyn_R_skirt_a_1")
    Filter=(TranslationFilter=(bX=true,bY=true,bZ=true),RotationFilter=(bX=True,bY=True,bZ=True),ScaleFilter=(bX=True,bY=True,bZ=True))
    Parents=((Item=(Type=Bone,Name="thigh_r"),Weight=1),(Item=(Type=Bone,Name="pelvis"),Weight=0))
    Weight=0.5
   )
   > &RigUnit_ParentConstraint(
    AdvancedSettings=(InterpolationType=Average,RotationOrderForFilter=XZY)
    Child=(Type=Bone,Name="dyn_R_skirt_a_3")
    Filter=(TranslationFilter=(bX=true,bY=true,bZ=true),RotationFilter=(bX=True,bY=True,bZ=True),ScaleFilter=(bX=True,bY=True,bZ=True))
    Parents=((Item=(Type=Bone,Name="calf_r"),Weight=0.5),(Item=(Type=Bone,Name="thigh_r"),Weight=0.5))
    Weight=0.5
   )
   > &RigUnit_ParentConstraint(
    AdvancedSettings=(InterpolationType=Average,RotationOrderForFilter=XZY)
    Child=(Type=Bone,Name="base_L_skirt_b")
    Filter=(TranslationFilter=(bX=true,bY=true,bZ=true),RotationFilter=(bX=true,bY=true,bZ=true),ScaleFilter=(bX=True,bY=True,bZ=True))
    Parents=((Item=(Type=Bone,Name="pelvis"),Weight=1))
    Weight=1
   )
   > &RigUnit_ParentConstraint(
    AdvancedSettings=(InterpolationType=Average,RotationOrderForFilter=XZY)
    Child=(Type=Bone,Name="base_R_skirt_b")
    Filter=(TranslationFilter=(bX=true,bY=true,bZ=true),RotationFilter=(bX=True,bY=True,bZ=True),ScaleFilter=(bX=True,bY=True,bZ=True))
    Parents=((Item=(Type=Bone,Name="pelvis"),Weight=1))
    Weight=1
   )
   > &RigUnit_ParentConstraint(
    AdvancedSettings=(InterpolationType=Average,RotationOrderForFilter=XZY)
    Child=(Type=Bone,Name="base_L_skirt_c")
    Filter=(TranslationFilter=(bX=true,bY=true,bZ=true),RotationFilter=(bX=true,bY=true,bZ=true),ScaleFilter=(bX=True,bY=True,bZ=True))
    Parents=((Item=(Type=Bone,Name="pelvis"),Weight=1))
    Weight=1
   )
   > &RigUnit_ParentConstraint(
    AdvancedSettings=(InterpolationType=Average,RotationOrderForFilter=XZY)
    Child=(Type=Bone,Name="base_R_skirt_c")
    Filter=(TranslationFilter=(bX=true,bY=true,bZ=true),RotationFilter=(bX=True,bY=True,bZ=True),ScaleFilter=(bX=True,bY=True,bZ=True))
    Parents=((Item=(Type=Bone,Name="pelvis"),Weight=1))
    Weight=1
   )
   > &RigUnit_ParentConstraint(
    AdvancedSettings=(InterpolationType=Average,RotationOrderForFilter=XZY)
    Child=(Type=Bone,Name="dyn_L_skirt_c_1")
    Filter=(TranslationFilter=(bX=true,bY=true,bZ=true),RotationFilter=(bX=true,bY=true,bZ=true),ScaleFilter=(bX=True,bY=True,bZ=True))
    Parents=((Item=(Type=Bone,Name="pelvis"),Weight=1))
    Weight=1
   )
   > &RigUnit_ParentConstraint(
    AdvancedSettings=(InterpolationType=Average,RotationOrderForFilter=XZY)
    Child=(Type=Bone,Name="dyn_R_skirt_c_1")
    Filter=(TranslationFilter=(bX=true,bY=true,bZ=true),RotationFilter=(bX=true,bY=true,bZ=true),ScaleFilter=(bX=True,bY=True,bZ=True))
    Parents=((Item=(Type=Bone,Name="pelvis"),Weight=1))
    Weight=1
   )
   > &RigUnit_ParentConstraint(
    AdvancedSettings=(InterpolationType=Average,RotationOrderForFilter=XZY)
    Child=(Type=Bone,Name="dyn_L_skirt_b_1")
    Filter=(TranslationFilter=(bX=true,bY=true,bZ=true),RotationFilter=(bX=true,bY=true,bZ=true),ScaleFilter=(bX=True,bY=True,bZ=True))
    Parents=((Item=(Type=Bone,Name="base_L_skirt_b"),Weight=1))
    Weight=1
   )
   > &RigUnit_ParentConstraint(
    AdvancedSettings=(InterpolationType=Average,RotationOrderForFilter=XZY)
    Child=(Type=Bone,Name="dyn_R_skirt_b_1")
    Filter=(TranslationFilter=(bX=true,bY=true,bZ=true),RotationFilter=(bX=True,bY=True,bZ=True),ScaleFilter=(bX=True,bY=True,bZ=True))
    Parents=((Item=(Type=Bone,Name="base_R_skirt_b"),Weight=1))
    Weight=1
   )
   > &RigUnit_ModifyTransforms(
    ItemToModify=((Item=(Type=Bone,Name="dyn_L_skirt_a_1"),Transform=(Rotation=(X=0,Y=-0.0436,Z=0,W=0.999),Translation=(X=0,Y=0,Z=0),Scale3D=(X=1,Y=1,Z=1))))
    Weight=1
    WeightMinimum=0
   )
   > &RigUnit_ModifyTransforms(
    ItemToModify=((Item=(Type=Bone,Name="dyn_R_skirt_a_1"),Transform=(Rotation=(X=0,Y=-0.0436,Z=0,W=0.999),Translation=(X=0,Y=0,Z=0),Scale3D=(X=1,Y=1,Z=1))))
    Weight=1
    WeightMinimum=0
   )
   > &RigUnit_ParentConstraint(
    AdvancedSettings=(InterpolationType=Average,RotationOrderForFilter=XZY)
    Child=(Type=Bone,Name="dyn_L_skirt_c_1")
    Filter=(TranslationFilter=(bX=false,bY=false,bZ=false),RotationFilter=(bX=true,bY=true,bZ=true),ScaleFilter=(bX=True,bY=True,bZ=True))
    Parents=((Item=(Type=Bone,Name="dyn_L_skirt_b_1"),Weight=0.2),(Item=(Type=Bone,Name="pelvis"),Weight=0.8))
    Weight=1
   )
   > &RigUnit_ParentConstraint(
    AdvancedSettings=(InterpolationType=Average,RotationOrderForFilter=XZY)
    Child=(Type=Bone,Name="dyn_R_skirt_c_1")
    Filter=(TranslationFilter=(bX=false,bY=false,bZ=false),RotationFilter=(bX=True,bY=True,bZ=True),ScaleFilter=(bX=True,bY=True,bZ=True))
    Parents=((Item=(Type=Bone,Name="dyn_R_skirt_b_1"),Weight=0.2),(Item=(Type=Bone,Name="pelvis"),Weight=0.8))
    Weight=1
   )
   > &RigUnit_ParentConstraint(
    AdvancedSettings=(InterpolationType=Average,RotationOrderForFilter=XZY)
    Child=(Type=Bone,Name="base_L_skirt_b")
    Filter=(TranslationFilter=(bX=false,bY=false,bZ=false),RotationFilter=(bX=false,bY=true,bZ=true),ScaleFilter=(bX=True,bY=True,bZ=True))
    Parents=((Item=(Type=Bone,Name="thigh_l"),Weight=0.5),(Item=(Type=Bone,Name="pelvis"),Weight=0.5))
    Weight=1
    bMaintainOffset=true
   )
   > &RigUnit_ParentConstraint(
    AdvancedSettings=(InterpolationType=Average,RotationOrderForFilter=XZY)
    Child=(Type=Bone,Name="base_R_skirt_b")
    Filter=(TranslationFilter=(bX=false,bY=false,bZ=false),RotationFilter=(bX=false,bY=true,bZ=true),ScaleFilter=(bX=True,bY=True,bZ=True))
    Parents=((Item=(Type=Bone,Name="thigh_r"),Weight=0.5),(Item=(Type=Bone,Name="pelvis"),Weight=0.5))
    Weight=1
   )
   > &RigUnit_ModifyTransforms(
    ItemToModify=((Item=(Type=Bone,Name="dyn_R_skirt_c_1"),Transform=(Rotation=(X=0,Y=-0.1305,Z=0,W=0.9914),Translation=(X=0,Y=0,Z=0),Scale3D=(X=1,Y=1,Z=1))))
    Weight=1
    WeightMinimum=0
   )
   > &RigUnit_ModifyTransforms(
    ItemToModify=((Item=(Type=Bone,Name="dyn_L_skirt_c_1"),Transform=(Rotation=(X=0,Y=-0.1305,Z=0,W=0.9914),Translation=(X=0,Y=0,Z=0),Scale3D=(X=1,Y=1,Z=1))))
    Weight=1
    WeightMinimum=0
   )
   > &RigUnit_ModifyTransforms(
    ItemToModify=((Item=(Type=Bone,Name="dyn_R_skirt_b_1"),Transform=(Rotation=(X=0,Y=-0.0436,Z=0,W=0.999),Translation=(X=0,Y=0,Z=0),Scale3D=(X=1,Y=1,Z=1))))
    Weight=1
    WeightMinimum=0
   )
   > &RigUnit_ModifyTransforms(
    ItemToModify=((Item=(Type=Bone,Name="dyn_L_skirt_b_1"),Transform=(Rotation=(X=0,Y=-0.0436,Z=0,W=0.999),Translation=(X=0,Y=0,Z=0),Scale3D=(X=1,Y=1,Z=1))))
    Weight=1
    WeightMinimum=0
   )
   n13 = RigVMVariableNode() {variable=R thighSide Alpha}
   n14 = RigVMVariableNode() {variable=R thighForward Alpha}
   n15 = &RigVMFunction_MathDoubleMul(
    A=RigVMVariableNode() {variable=pelvisLowToGround_INVERSE}.Value
    B=&RigVMFunction_MathDoubleMul(B=0.75, A=&RigVMFunction_MathDoubleAdd(B=@n13.Value, A=@n14.Value).Result).Result
   )
   > &RigUnit_ParentConstraint(
    AdvancedSettings=(InterpolationType=Average,RotationOrderForFilter=XZY)
    Child=(Type=Bone,Name="dyn_R_skirt_b_1")
    Filter=(TranslationFilter=(bX=true,bY=true,bZ=true),RotationFilter=(bX=false,bY=true,bZ=true),ScaleFilter=(bX=True,bY=True,bZ=True))
    Parents=((Item=(Type=Bone,Name="thigh_r"),Weight=1))
    Weight=@n15.Result
   )
   > &RigUnit_ModifyTransforms(
    ItemToModify=((Item=(Type=Bone,Name="dyn_R_skirt_b_1"),Transform=(Rotation=(X=0,Y=-0.0872,Z=0,W=0.9962),Translation=(X=0,Y=2,Z=2),Scale3D=(X=1,Y=1,Z=1))))
    WeightMinimum=0
    Weight=@n15.Result
   )
   n16 = &RigVMFunction_MathDoubleMul(
    A=RigVMVariableNode() {variable=pelvisLowToGround_INVERSE}.Value
    B=&RigVMFunction_MathDoubleSub(A=RigVMVariableNode() {variable=R thighSide Alpha}.Value, B=RigVMVariableNode() {variable=R thighForward Alpha}.Value).Result
   )
   > &RigUnit_ParentConstraint(
    AdvancedSettings=(InterpolationType=Average,RotationOrderForFilter=XZY)
    Child=(Type=Bone,Name="dyn_R_skirt_b_3")
    Filter=(TranslationFilter=(bX=false,bY=false,bZ=false),RotationFilter=(bX=false,bY=true,bZ=true),ScaleFilter=(bX=True,bY=True,bZ=True))
    Parents=((Item=(Type=Bone,Name="thigh_r"),Weight=0.5),(Item=(Type=Bone,Name="calf_r"),Weight=0.5))
    Weight=@n16.Result
   )
   > &RigUnit_ParentConstraint(
    AdvancedSettings=(InterpolationType=Average,RotationOrderForFilter=XZY)
    Child=(Type=Bone,Name="dyn_R_skirt_c_1")
    Filter=(TranslationFilter=(bX=false,bY=false,bZ=false),RotationFilter=(bX=false,bY=true,bZ=true),ScaleFilter=(bX=True,bY=True,bZ=True))
    Parents=((Item=(Type=Bone,Name="thigh_r"),Weight=0.5),(Item=(Type=Bone,Name="calf_r"),Weight=0.5))
    Weight=@n16.Result
   )
   > &RigUnit_ParentConstraint(
    AdvancedSettings=(InterpolationType=Average,RotationOrderForFilter=XZY)
    Child=(Type=Bone,Name="dyn_R_skirt_c_3")
    Filter=(TranslationFilter=(bX=false,bY=false,bZ=false),RotationFilter=(bX=false,bY=true,bZ=true),ScaleFilter=(bX=True,bY=True,bZ=True))
    Parents=((Item=(Type=Bone,Name="thigh_r"),Weight=0.5),(Item=(Type=Bone,Name="calf_r"),Weight=0.5))
    Weight=@n16.Result
   )
   > &RigUnit_ParentConstraint(
    AdvancedSettings=(InterpolationType=Average,RotationOrderForFilter=XZY)
    Child=(Type=Bone,Name="dyn_R_skirt_c_1")
    Filter=(TranslationFilter=(bX=false,bY=false,bZ=false),RotationFilter=(bX=false,bY=true,bZ=true),ScaleFilter=(bX=True,bY=True,bZ=True))
    Parents=((Item=(Type=Bone,Name="thigh_r"),Weight=0.8),(Item=(Type=Bone,Name="calf_r"),Weight=0.2))
    Weight=&RigVMFunction_MathDoubleSub(A=0.3, B=RigVMVariableNode() {variable=R thighForward Alpha}.Value).Result
   )
   n17 = &RigVMFunction_MathDoubleAdd(A=RigVMVariableNode() {variable=R calfBend Alpha}.Value, B=RigVMVariableNode() {variable=R thighBack Alpha}.Value)
   n18 = &RigVMFunction_MathDoubleSub(A=1, B=RigVMVariableNode() {variable=R thighForward Alpha}.Value)
   > &RigUnit_ParentConstraint(
    AdvancedSettings=(InterpolationType=Average,RotationOrderForFilter=XZY)
    Child=(Type=Bone,Name="dyn_R_skirt_c_3")
    Filter=(TranslationFilter=(bX=false,bY=false,bZ=false),RotationFilter=(bX=true,bY=true,bZ=true),ScaleFilter=(bX=True,bY=True,bZ=True))
    Parents=((Item=(Type=Bone,Name="calf_r"),Weight=0.65),(Item=(Type=Bone,Name="thigh_r"),Weight=0.35))
    bMaintainOffset=true
    Weight=&RigVMFunction_MathDoubleMul(
     A=RigVMVariableNode() {variable=pelvisLowToGround_INVERSE}.Value
     B=&RigVMFunction_MathDoubleMul(A=@n17.Result, B=@n18.Result).Result
    ).Result
   )
   n19 = &RigVMFunction_MathDoubleAdd(A=RigVMVariableNode() {variable=R calfBend Alpha}.Value, B=RigVMVariableNode() {variable=R thighBack Alpha}.Value)
   n20 = &RigVMFunction_MathDoubleMul(B=0.2, A=RigVMVariableNode() {variable=R thighForward Alpha}.Value)
   > &RigUnit_ModifyTransforms(
    ItemToModify=((Item=(Type=Bone,Name="dyn_R_skirt_c_1"),Transform=(Rotation=(X=-0.0037,Y=-0.0609,Z=-0.0609,W=0.9963),Translation=(X=0,Y=0,Z=0),Scale3D=(X=1,Y=1,Z=1))))
    WeightMinimum=0
    Weight=&RigVMFunction_MathDoubleMul(
     B=&RigVMFunction_MathDoubleSub(A=@n19.Result, B=@n20.Result).Result
     A=RigVMVariableNode() {variable=pelvisLowToGround_INVERSE}.Value
    ).Result
   )
   > &RigUnit_ModifyTransforms(
    ItemToModify=((Item=(Type=Bone,Name="dyn_R_skirt_c_1"),Transform=(Rotation=(X=-0.0027,Y=-0.0523,Z=-0.0523,W=0.9973),Translation=(X=0,Y=0,Z=0),Scale3D=(X=1,Y=1,Z=1))))
    WeightMinimum=0
    Weight=@n17.Result
   )
   > &RigUnit_ModifyTransforms(
    ItemToModify=((Item=(Type=Bone,Name="dyn_R_skirt_b_1"),Transform=(Rotation=(X=-0.0005,Y=-0.0262,Z=-0.0174,W=0.9995),Translation=(X=2,Y=-3,Z=0),Scale3D=(X=1,Y=1,Z=1))))
    WeightMinimum=0
    Weight=RigVMVariableNode() {variable=R thighVeryForward}.Value
   )
   n21 = RigVMVariableNode() {variable=L thighSide Alpha}
   n22 = RigVMVariableNode() {variable=L thighForward Alpha}
   n23 = &RigVMFunction_MathDoubleMul(
    A=RigVMVariableNode() {variable=pelvisLowToGround_INVERSE}.Value
    B=&RigVMFunction_MathDoubleMul(B=0.75, A=&RigVMFunction_MathDoubleAdd(B=@n21.Value, A=@n22.Value).Result).Result
   )
   > &RigUnit_ParentConstraint(
    AdvancedSettings=(InterpolationType=Average,RotationOrderForFilter=XZY)
    Child=(Type=Bone,Name="dyn_L_skirt_b_1")
    Filter=(TranslationFilter=(bX=true,bY=true,bZ=true),RotationFilter=(bX=false,bY=true,bZ=true),ScaleFilter=(bX=True,bY=True,bZ=True))
    Parents=((Item=(Type=Bone,Name="thigh_l"),Weight=1))
    Weight=@n23.Result
   )
   > &RigUnit_ModifyTransforms(
    ItemToModify=((Item=(Type=Bone,Name="dyn_L_skirt_b_1"),Transform=(Rotation=(X=0,Y=-0.0872,Z=0,W=0.9962),Translation=(X=0,Y=2,Z=2),Scale3D=(X=1,Y=1,Z=1))))
    WeightMinimum=0
    Weight=@n23.Result
   )
   n24 = &RigVMFunction_MathDoubleMul(
    A=RigVMVariableNode() {variable=pelvisLowToGround_INVERSE}.Value
    B=&RigVMFunction_MathDoubleSub(A=RigVMVariableNode() {variable=L thighSide Alpha}.Value, B=RigVMVariableNode() {variable=L thighForward Alpha}.Value).Result
   )
   > &RigUnit_ParentConstraint(
    AdvancedSettings=(InterpolationType=Average,RotationOrderForFilter=XZY)
    Child=(Type=Bone,Name="dyn_L_skirt_b_3")
    Filter=(TranslationFilter=(bX=false,bY=false,bZ=false),RotationFilter=(bX=false,bY=true,bZ=true),ScaleFilter=(bX=True,bY=True,bZ=True))
    Parents=((Item=(Type=Bone,Name="thigh_l"),Weight=0.5),(Item=(Type=Bone,Name="calf_l"),Weight=0.5))
    bMaintainOffset=true
    Weight=@n24.Result
   )
   > &RigUnit_ParentConstraint(
    AdvancedSettings=(InterpolationType=Average,RotationOrderForFilter=XZY)
    Child=(Type=Bone,Name="dyn_L_skirt_c_1")
    Filter=(TranslationFilter=(bX=false,bY=false,bZ=false),RotationFilter=(bX=false,bY=true,bZ=true),ScaleFilter=(bX=True,bY=True,bZ=True))
    Parents=((Item=(Type=Bone,Name="thigh_l"),Weight=0.5),(Item=(Type=Bone,Name="calf_l"),Weight=0.5))
    Weight=@n24.Result
   )
   > &RigUnit_ParentConstraint(
    AdvancedSettings=(InterpolationType=Average,RotationOrderForFilter=XZY)
    Child=(Type=Bone,Name="dyn_L_skirt_c_3")
    Filter=(TranslationFilter=(bX=false,bY=false,bZ=false),RotationFilter=(bX=false,bY=true,bZ=true),ScaleFilter=(bX=True,bY=True,bZ=True))
    Parents=((Item=(Type=Bone,Name="thigh_l"),Weight=0.5),(Item=(Type=Bone,Name="calf_l"),Weight=0.5))
    Weight=@n24.Result
   )
   > &RigUnit_ParentConstraint(
    AdvancedSettings=(InterpolationType=Average,RotationOrderForFilter=XZY)
    Child=(Type=Bone,Name="dyn_L_skirt_c_1")
    Filter=(TranslationFilter=(bX=false,bY=false,bZ=false),RotationFilter=(bX=false,bY=true,bZ=true),ScaleFilter=(bX=True,bY=True,bZ=True))
    Parents=((Item=(Type=Bone,Name="thigh_l"),Weight=0.8),(Item=(Type=Bone,Name="calf_l"),Weight=0.2))
    Weight=&RigVMFunction_MathDoubleSub(A=0.3, B=RigVMVariableNode() {variable=L thighForward Alpha}.Value).Result
   )
   n25 = &RigVMFunction_MathDoubleAdd(A=RigVMVariableNode() {variable=L calfBend Alpha}.Value, B=RigVMVariableNode() {variable=L thighBack Alpha}.Value)
   n26 = &RigVMFunction_MathDoubleSub(A=1, B=RigVMVariableNode() {variable=L thighForward Alpha}.Value)
   > &RigUnit_ParentConstraint(
    AdvancedSettings=(InterpolationType=Average,RotationOrderForFilter=XZY)
    Child=(Type=Bone,Name="dyn_L_skirt_c_3")
    Filter=(TranslationFilter=(bX=false,bY=false,bZ=false),RotationFilter=(bX=true,bY=true,bZ=true),ScaleFilter=(bX=True,bY=True,bZ=True))
    Parents=((Item=(Type=Bone,Name="calf_l"),Weight=0.65),(Item=(Type=Bone,Name="thigh_l"),Weight=0.35))
    Weight=&RigVMFunction_MathDoubleMul(
     A=RigVMVariableNode() {variable=pelvisLowToGround_INVERSE}.Value
     B=&RigVMFunction_MathDoubleMul(A=@n25.Result, B=@n26.Result).Result
    ).Result
   )
   n27 = &RigVMFunction_MathDoubleAdd(A=RigVMVariableNode() {variable=L calfBend Alpha}.Value, B=RigVMVariableNode() {variable=L thighBack Alpha}.Value)
   n28 = &RigVMFunction_MathDoubleMul(B=0.2, A=RigVMVariableNode() {variable=L thighForward Alpha}.Value)
   > &RigUnit_ModifyTransforms(
    ItemToModify=((Item=(Type=Bone,Name="dyn_L_skirt_c_1"),Transform=(Rotation=(X=0.0037,Y=-0.0609,Z=0.0609,W=0.9963),Translation=(X=0,Y=0,Z=0),Scale3D=(X=1,Y=1,Z=1))))
    WeightMinimum=0
    Weight=&RigVMFunction_MathDoubleMul(
     B=&RigVMFunction_MathDoubleSub(A=@n27.Result, B=@n28.Result).Result
     A=RigVMVariableNode() {variable=pelvisLowToGround_INVERSE}.Value
    ).Result
   )
   > &RigUnit_ModifyTransforms(
    ItemToModify=((Item=(Type=Bone,Name="dyn_L_skirt_c_1"),Transform=(Rotation=(X=0.0027,Y=-0.0523,Z=0.0523,W=0.9973),Translation=(X=0,Y=0,Z=0),Scale3D=(X=1,Y=1,Z=1))))
    WeightMinimum=0
    Weight=@n25.Result
   )
   > &RigUnit_ModifyTransforms(
    ItemToModify=((Item=(Type=Bone,Name="dyn_L_skirt_b_1"),Transform=(Rotation=(X=0.0005,Y=-0.0262,Z=0.0174,W=0.9995),Translation=(X=2,Y=-3,Z=0),Scale3D=(X=1,Y=1,Z=1))))
    WeightMinimum=0
    Weight=RigVMVariableNode() {variable=L thighVeryForward}.Value
   )
   n29 = &RigVMFunction_MathDoubleSub(A=1, B=RigVMVariableNode() {variable=L thighSide Alpha}.Value)
   n30 = RigVMVariableNode() {variable=L calfBend Alpha}
   n31 = RigVMVariableNode() {variable=L thighForward Alpha}
   n32 = &RigVMFunction_MathDoubleMul(
    B=&RigVMFunction_MathDoubleSub(
     A=&RigVMFunction_MathDoubleMul(A=@n29.Result, B=@n30.Value).Result
     B=&RigVMFunction_MathDoubleMul(B=0.1, A=@n31.Value).Result
    ).Result
    A=RigVMVariableNode() {variable=pelvisLowToGround_INVERSE}.Value
   )
   > &RigUnit_ParentConstraint(
    AdvancedSettings=(InterpolationType=Average,RotationOrderForFilter=XZY)
    Child=(Type=Bone,Name="dyn_R_skirt_a_1")
    Filter=(TranslationFilter=(bX=false,bY=false,bZ=false),RotationFilter=(bX=True,bY=true,bZ=true),ScaleFilter=(bX=True,bY=True,bZ=True))
    Parents=((Item=(Type=Bone,Name="thigh_r"),Weight=1))
    Weight=@n32.Result
   )
   > &RigUnit_ModifyTransforms(
    ItemToModify=((Item=(Type=Bone,Name="dyn_R_skirt_a_1"),Transform=(Rotation=(X=0,Y=-0.1736,Z=0,W=0.9848),Translation=(X=0,Y=0,Z=0),Scale3D=(X=1,Y=1,Z=1))))
    WeightMinimum=0
    Weight=@n32.Result
   )
   > &RigUnit_ModifyTransforms(
    ItemToModify=((Item=(Type=Bone,Name="dyn_C_skirt_a_1"),Transform=(Rotation=(X=0,Y=-0.1305,Z=0,W=0.9914),Translation=(X=0,Y=0,Z=0),Scale3D=(X=1,Y=1,Z=1))))
    WeightMinimum=0
    Weight=@n32.Result
   )
   Subtract = &RigVMFunction_MathDoubleSub(A=1, B=RigVMVariableNode() {variable=R thighSide Alpha}.Value)
   Multiply = &RigVMFunction_MathDoubleMul(A=@Subtract.Result, B=RigVMVariableNode() {variable=R calfBend Alpha}.Value)
   n33 = RigVMVariableNode() {variable=R thighForward Alpha}
   n34 = &RigVMFunction_MathDoubleMul(
    B=&RigVMFunction_MathDoubleSub(A=@Multiply.Result, B=&RigVMFunction_MathDoubleMul(B=0.1, A=@n33.Value).Result).Result
    A=RigVMVariableNode() {variable=pelvisLowToGround_INVERSE}.Value
   )
   > &RigUnit_ParentConstraint(
    AdvancedSettings=(InterpolationType=Average,RotationOrderForFilter=XZY)
    Child=(Type=Bone,Name="dyn_L_skirt_a_1")
    Filter=(TranslationFilter=(bX=false,bY=false,bZ=false),RotationFilter=(bX=True,bY=true,bZ=true),ScaleFilter=(bX=True,bY=True,bZ=True))
    Parents=((Item=(Type=Bone,Name="thigh_l"),Weight=1))
    bMaintainOffset=true
    Weight=@n34.Result
   )
   > &RigUnit_ModifyTransforms(
    ItemToModify=((Item=(Type=Bone,Name="dyn_L_skirt_a_1"),Transform=(Rotation=(X=0,Y=-0.1736,Z=0,W=0.9848),Translation=(X=0,Y=0,Z=0),Scale3D=(X=1,Y=1,Z=1))))
    WeightMinimum=0
    Weight=@n34.Result
   )
   > &RigUnit_ModifyTransforms(
    ItemToModify=((Item=(Type=Bone,Name="dyn_C_skirt_a_1"),Transform=(Rotation=(X=0,Y=-0.1305,Z=0,W=0.9914),Translation=(X=0,Y=0,Z=0),Scale3D=(X=1,Y=1,Z=1))))
    WeightMinimum=0
    Weight=@n34.Result
   )
   > n35 = &RigUnit_SphericalPoseReader(
    ActiveRegionScaleFactors=(PositiveWidth=2,NegativeWidth=2,PositiveHeight=3,NegativeHeight=3)
    ActiveRegionSize=0.1
    Debug=(bDrawDebug=true,bDraw2D=False,bDrawLocalAxes=False,DebugScale=25,DebugSegments=20,DebugThickness=0.25)
    DriverAxis=(X=-1,Y=0,Z=0)
    DriverItem=(Type=Bone,Name="thigh_l")
    FalloffRegionScaleFactors=(PositiveWidth=1,NegativeWidth=1,PositiveHeight=1,NegativeHeight=1)
    FalloffSize=0.45
    FlipHeightScaling=false
    FlipWidthScaling=false
    OptionalParentItem=(Type=Bone,Name="None")
    RotationOffset=(X=180,Y=-20,Z=-20)
   )
   > &RigUnit_ModifyTransforms(
    ItemToModify=((Item=(Type=Bone,Name="dyn_L_skirt_b_1"),Transform=(Rotation=(X=0,Y=-0.0436,Z=0,W=0.999),Translation=(X=0,Y=0,Z=0),Scale3D=(X=1,Y=1,Z=1))))
    WeightMinimum=0
    Weight=@n35.OutputParam
   )
   > n36 = &RigUnit_SphericalPoseReader(
    ActiveRegionScaleFactors=(PositiveWidth=2,NegativeWidth=1.5,PositiveHeight=2,NegativeHeight=2)
    ActiveRegionSize=0.1
    Debug=(bDrawDebug=false,bDraw2D=false,DebugScale=25,DebugSegments=20,DebugThickness=0.25)
    DriverAxis=(X=-1,Y=0,Z=0)
    DriverItem=(Type=Bone,Name="thigh_l")
    FalloffRegionScaleFactors=(PositiveWidth=1,NegativeWidth=1.25,PositiveHeight=1,NegativeHeight=1)
    FalloffSize=0.45
    OptionalParentItem=(Type=Bone,Name="pelvis")
    RotationOffset=(X=150,Y=180,Z=10)
   )
   > &RigUnit_ModifyTransforms(
    ItemToModify=((Item=(Type=Bone,Name="dyn_L_skirt_b_1"),Transform=(Rotation=(X=0,Y=-0.2588,Z=0,W=0.9659),Translation=(X=0,Y=0,Z=0),Scale3D=(X=1,Y=1,Z=1))))
    WeightMinimum=0
    Weight=@n36.OutputParam
   )
   > n37 = &RigUnit_SphericalPoseReader(
    ActiveRegionScaleFactors=(PositiveWidth=2,NegativeWidth=2,PositiveHeight=3,NegativeHeight=3)
    ActiveRegionSize=0.1
    Debug=(bDrawDebug=false,bDraw2D=False,bDrawLocalAxes=False,DebugScale=25,DebugSegments=20,DebugThickness=0.25)
    DriverAxis=(X=1,Y=0,Z=0)
    DriverItem=(Type=Bone,Name="thigh_r")
    FalloffRegionScaleFactors=(PositiveWidth=1,NegativeWidth=1,PositiveHeight=1,NegativeHeight=1)
    FalloffSize=0.45
    FlipHeightScaling=False
    FlipWidthScaling=true
    OptionalParentItem=(Type=Bone,Name="None")
    RotationOffset=(X=0,Y=-20,Z=-20)
   )
   > &RigUnit_ModifyTransforms(
    ItemToModify=((Item=(Type=Bone,Name="dyn_R_skirt_b_1"),Transform=(Rotation=(X=0,Y=-0.0436,Z=0,W=0.999),Translation=(X=0,Y=0,Z=0),Scale3D=(X=1,Y=1,Z=1))))
    WeightMinimum=0
    Weight=@n37.OutputParam
   )
   > n38 = &RigUnit_SphericalPoseReader(
    ActiveRegionScaleFactors=(PositiveWidth=2,NegativeWidth=1.5,PositiveHeight=2,NegativeHeight=2)
    ActiveRegionSize=0.1
    Debug=(bDrawDebug=false,bDraw2D=false,DebugScale=25,DebugSegments=20,DebugThickness=0.25)
    DriverAxis=(X=1,Y=0,Z=0)
    DriverItem=(Type=Bone,Name="thigh_r")
    FalloffRegionScaleFactors=(PositiveWidth=1,NegativeWidth=1.25,PositiveHeight=1,NegativeHeight=1)
    FalloffSize=0.45
    OptionalParentItem=(Type=Bone,Name="pelvis")
    RotationOffset=(X=-150,Y=0,Z=10)
   )
   > &RigUnit_ModifyTransforms(
    ItemToModify=((Item=(Type=Bone,Name="dyn_R_skirt_b_1"),Transform=(Rotation=(X=0,Y=-0.2588,Z=0,W=0.9659),Translation=(X=0,Y=0,Z=0),Scale3D=(X=1,Y=1,Z=1))))
    WeightMinimum=0
    Weight=@n38.OutputParam
   )
   n39 = RigVMVariableNode() {variable=L calfBend Alpha}
   n40 = RigVMVariableNode() {variable=L thighBack Alpha}
   n41 = RigVMVariableNode() {variable=L thighForward Alpha}
   n42 = &RigVMFunction_MathDoubleMul(
    B=&RigVMFunction_MathDoubleSub(
     A=&RigVMFunction_MathDoubleAdd(A=@n39.Value, B=@n40.Value).Result
     B=&RigVMFunction_MathDoubleMul(B=0.2, A=@n41.Value).Result
    ).Result
    A=RigVMVariableNode() {variable=pelvisLowToGround_INVERSE}.Value
   )
   > &RigUnit_ParentConstraint(
    AdvancedSettings=(InterpolationType=Average,RotationOrderForFilter=XZY)
    Child=(Type=Bone,Name="dyn_L_skirt_b_3")
    Filter=(TranslationFilter=(bX=false,bY=false,bZ=false),RotationFilter=(bX=false,bY=true,bZ=true),ScaleFilter=(bX=True,bY=True,bZ=True))
    Parents=((Item=(Type=Bone,Name="calf_l"),Weight=0.5),(Item=(Type=Bone,Name="thigh_l"),Weight=0.5))
    bMaintainOffset=true
    Weight=@n42.Result
   )
   > &RigUnit_ModifyTransforms(
    ItemToModify=((Item=(Type=Bone,Name="dyn_L_skirt_b_1"),Transform=(Rotation=(X=0,Y=-0.1305,Z=0,W=0.9914),Translation=(X=0,Y=0,Z=0),Scale3D=(X=1,Y=1,Z=1))))
    WeightMinimum=0
    Weight=@n42.Result
   )
   n43 = RigVMVariableNode() {variable=R calfBend Alpha}
   n44 = RigVMVariableNode() {variable=R thighBack Alpha}
   n45 = RigVMVariableNode() {variable=R thighForward Alpha}
   n46 = &RigVMFunction_MathDoubleMul(
    B=&RigVMFunction_MathDoubleSub(
     A=&RigVMFunction_MathDoubleAdd(A=@n43.Value, B=@n44.Value).Result
     B=&RigVMFunction_MathDoubleMul(B=0.2, A=@n45.Value).Result
    ).Result
    A=RigVMVariableNode() {variable=pelvisLowToGround_INVERSE}.Value
   )
   > &RigUnit_ParentConstraint(
    AdvancedSettings=(InterpolationType=Average,RotationOrderForFilter=XZY)
    Child=(Type=Bone,Name="dyn_R_skirt_b_3")
    Filter=(TranslationFilter=(bX=false,bY=false,bZ=false),RotationFilter=(bX=false,bY=true,bZ=true),ScaleFilter=(bX=True,bY=True,bZ=True))
    Parents=((Item=(Type=Bone,Name="calf_r"),Weight=0.5),(Item=(Type=Bone,Name="thigh_r"),Weight=1))
    bMaintainOffset=true
    Weight=@n46.Result
   )
   > &RigUnit_ModifyTransforms(
    ItemToModify=((Item=(Type=Bone,Name="dyn_R_skirt_b_1"),Transform=(Rotation=(X=0,Y=-0.1305,Z=0,W=0.9914),Translation=(X=0,Y=0,Z=0),Scale3D=(X=1,Y=1,Z=1))))
    WeightMinimum=0
    Weight=@n46.Result
   )
   Cast = &RigVMFunction_MathBoolToFloat(Value=RigVMVariableNode() {variable=Is DBNO Carried}.Value)
   n47 = &RigVMFunction_MathDoubleMul(
    A=&RigVMFunction_MathDoubleSub(A=1, B=@Cast.Result).Result
    B=RigVMVariableNode() {variable=pelvisLowToGround_translateRotate}.Value
   )
   > &RigUnit_ModifyTransforms(
    ItemToModify=((Item=(Type=Bone,Name="dyn_L_skirt_b_1"),Transform=(Rotation=(X=-0.067,Y=-0.25,Z=-0.25,W=0.933),Translation=(X=0,Y=0,Z=0),Scale3D=(X=1,Y=1,Z=1))))
    WeightMinimum=0
    Weight=@n47.Result
   )
   > &RigUnit_ModifyTransforms(
    ItemToModify=((Item=(Type=Bone,Name="dyn_L_skirt_b_2"),Transform=(Rotation=(X=0,Y=0,Z=-0.0872,W=0.9962),Translation=(X=0,Y=0,Z=0),Scale3D=(X=1,Y=1,Z=1))))
    WeightMinimum=0
    Weight=@n47.Result
   )
   > &RigUnit_ModifyTransforms(
    ItemToModify=((Item=(Type=Bone,Name="dyn_L_skirt_b_3"),Transform=(Rotation=(X=0,Y=0,Z=0,W=1),Translation=(X=0,Y=0,Z=0),Scale3D=(X=1,Y=1,Z=1))))
    WeightMinimum=0
    Weight=@n47.Result
   )
   > &RigUnit_ModifyTransforms(
    ItemToModify=((Item=(Type=Bone,Name="dyn_R_skirt_b_1"),Transform=(Rotation=(X=0.067,Y=-0.25,Z=0.25,W=0.933),Translation=(X=0,Y=0,Z=0),Scale3D=(X=1,Y=1,Z=1))))
    WeightMinimum=0
    Weight=@n47.Result
   )
   > &RigUnit_ModifyTransforms(
    ItemToModify=((Item=(Type=Bone,Name="dyn_R_skirt_b_2"),Transform=(Rotation=(X=0,Y=0,Z=0.0872,W=0.9962),Translation=(X=0,Y=0,Z=0),Scale3D=(X=1,Y=1,Z=1))))
    WeightMinimum=0
    Weight=@n47.Result
   )
   > &RigUnit_ModifyTransforms(
    ItemToModify=((Item=(Type=Bone,Name="dyn_R_skirt_b_3"),Transform=(Rotation=(X=0,Y=0,Z=0,W=1),Translation=(X=0,Y=0,Z=0),Scale3D=(X=1,Y=1,Z=1))))
    WeightMinimum=0
    Weight=@n47.Result
   )
   > &RigUnit_ModifyTransforms(
    ItemToModify=((Item=(Type=Bone,Name="dyn_L_skirt_c_1"),Transform=(Rotation=(X=0.1786,Y=-0.383,Z=0.383,W=0.8214),Translation=(X=0,Y=0,Z=0),Scale3D=(X=1,Y=1,Z=1))))
    WeightMinimum=0
    Weight=@n47.Result
   )
   > &RigUnit_ModifyTransforms(
    ItemToModify=((Item=(Type=Bone,Name="dyn_L_skirt_c_2"),Transform=(Rotation=(X=0.0076,Y=-0.0868,Z=0.0868,W=0.9924),Translation=(X=0,Y=0,Z=0),Scale3D=(X=1,Y=1,Z=1))))
    WeightMinimum=0
    Weight=@n47.Result
   )
   > &RigUnit_ModifyTransforms(
    ItemToModify=((Item=(Type=Bone,Name="dyn_L_skirt_c_3"),Transform=(Rotation=(X=0,Y=0,Z=0,W=1),Translation=(X=0,Y=0,Z=0),Scale3D=(X=1,Y=1,Z=1))))
    WeightMinimum=0
    Weight=@n47.Result
   )
   > &RigUnit_ModifyTransforms(
    ItemToModify=((Item=(Type=Bone,Name="dyn_R_skirt_c_1"),Transform=(Rotation=(X=-0.1786,Y=-0.383,Z=-0.383,W=0.8214),Translation=(X=0,Y=0,Z=0),Scale3D=(X=1,Y=1,Z=1))))
    WeightMinimum=0
    Weight=@n47.Result
   )
   > &RigUnit_ModifyTransforms(
    ItemToModify=((Item=(Type=Bone,Name="dyn_R_skirt_c_2"),Transform=(Rotation=(X=-0.0076,Y=-0.0868,Z=-0.0868,W=0.9924),Translation=(X=0,Y=0,Z=0),Scale3D=(X=1,Y=1,Z=1))))
    WeightMinimum=0
    Weight=@n47.Result
   )
   > &RigUnit_ModifyTransforms(
    ItemToModify=((Item=(Type=Bone,Name="dyn_R_skirt_c_3"),Transform=(Rotation=(X=0,Y=0,Z=0,W=1),Translation=(X=0,Y=0,Z=0),Scale3D=(X=1,Y=1,Z=1))))
    WeightMinimum=0
    Weight=@n47.Result
   )
   > &RigUnit_ParentConstraint(
    AdvancedSettings=(InterpolationType=Average,RotationOrderForFilter=XZY)
    Child=(Type=Bone,Name="dyn_C_skirt_b_1")
    Filter=(TranslationFilter=(bX=true,bY=true,bZ=true),RotationFilter=(bX=True,bY=True,bZ=True),ScaleFilter=(bX=True,bY=True,bZ=True))
    Parents=((Item=(Type=Bone,Name="dyn_L_skirt_c_1"),Weight=0.5),(Item=(Type=Bone,Name="dyn_R_skirt_c_1"),Weight=0.5))
    Weight=1
   )
   > &RigUnit_ParentConstraint(
    AdvancedSettings=(InterpolationType=Average,RotationOrderForFilter=XZY)
    Child=(Type=Bone,Name="dyn_C_skirt_b_2")
    Filter=(TranslationFilter=(bX=true,bY=true,bZ=true),RotationFilter=(bX=True,bY=True,bZ=True),ScaleFilter=(bX=True,bY=True,bZ=True))
    Parents=((Item=(Type=Bone,Name="dyn_L_skirt_c_2"),Weight=0.5),(Item=(Type=Bone,Name="dyn_R_skirt_c_2"),Weight=0.5))
    Weight=1
    bMaintainOffset=true
   )
   > &RigUnit_ParentConstraint(
    AdvancedSettings=(InterpolationType=Average,RotationOrderForFilter=XZY)
    Child=(Type=Bone,Name="dyn_C_skirt_b_3")
    Parents=((Item=(Type=Bone,Name="dyn_L_skirt_c_3"),Weight=0.5),(Item=(Type=Bone,Name="dyn_R_skirt_c_3"),Weight=0.5))
    Weight=1
   )
   > &RigUnit_ParentConstraint(
    AdvancedSettings=(InterpolationType=Average,RotationOrderForFilter=XZY)
    Child=(Type=Bone,Name="dyn_C_skirt_b_4")
    Parents=((Item=(Type=Bone,Name="dyn_L_skirt_c_4"),Weight=0.5),(Item=(Type=Bone,Name="dyn_R_skirt_c_4"),Weight=0.5))
    Weight=1
   )
   > &RigUnit_ParentConstraint(
    AdvancedSettings=(InterpolationType=Average,RotationOrderForFilter=XZY)
    Child=(Type=Bone,Name="dyn_C_skirt_b_5")
    Parents=((Item=(Type=Bone,Name="dyn_L_skirt_c_5"),Weight=0.5),(Item=(Type=Bone,Name="dyn_R_skirt_c_5"),Weight=0.5))
    Weight=1
   )
   n48 = &RigVMFunction_MathBoolToFloat(Value=RigVMVariableNode() {variable=Is DBNO Carried}.Value)
   > &RigUnit_ModifyTransforms(
    ItemToModify=((Item=(Type=Bone,Name="dyn_C_skirt_b_1"),Transform=(Rotation=(X=0,Y=0,Z=0,W=1),Translation=(X=-5,Y=0,Z=0),Scale3D=(X=1,Y=1,Z=1))))
    WeightMinimum=0
    Weight=&RigVMFunction_MathDoubleMul(
     A=&RigVMFunction_MathDoubleSub(A=1, B=@n48.Result).Result
     B=RigVMVariableNode() {variable=pelvisLowToGround_translateRotate}.Value
    ).Result
   )
   > &RigUnit_ParentConstraint(
    AdvancedSettings=(InterpolationType=Average,RotationOrderForFilter=XZY)
    Child=(Type=Bone,Name="base_C_skirt_a")
    Filter=(TranslationFilter=(bX=true,bY=true,bZ=true),RotationFilter=(bX=True,bY=True,bZ=True),ScaleFilter=(bX=True,bY=True,bZ=True))
    Parents=((Item=(Type=Bone,Name="base_R_skirt_a"),Weight=0.5),(Item=(Type=Bone,Name="base_L_skirt_a"),Weight=0.5))
    Weight=1
   )
   > &RigUnit_ParentConstraint(
    AdvancedSettings=(InterpolationType=Average,RotationOrderForFilter=XZY)
    Child=(Type=Bone,Name="dyn_C_skirt_a_1")
    Filter=(TranslationFilter=(bX=true,bY=true,bZ=true),RotationFilter=(bX=True,bY=True,bZ=True),ScaleFilter=(bX=True,bY=True,bZ=True))
    Parents=((Item=(Type=Bone,Name="dyn_R_skirt_a_1"),Weight=0.5),(Item=(Type=Bone,Name="dyn_L_skirt_a_1"),Weight=0.5))
    Weight=1
   )
   > &RigUnit_ParentConstraint(
    AdvancedSettings=(InterpolationType=Average,RotationOrderForFilter=XZY)
    Child=(Type=Bone,Name="dyn_C_skirt_a_2")
    Filter=(TranslationFilter=(bX=true,bY=true,bZ=true),RotationFilter=(bX=True,bY=True,bZ=True),ScaleFilter=(bX=True,bY=True,bZ=True))
    Parents=((Item=(Type=Bone,Name="dyn_R_skirt_a_2"),Weight=0.5),(Item=(Type=Bone,Name="dyn_L_skirt_a_2"),Weight=0.5))
    Weight=1
    bMaintainOffset=true
   )
   > &RigUnit_ParentConstraint(
    AdvancedSettings=(InterpolationType=Average,RotationOrderForFilter=XZY)
    Child=(Type=Bone,Name="dyn_C_skirt_a_3")
    Parents=((Item=(Type=Bone,Name="dyn_R_skirt_a_3"),Weight=0.5),(Item=(Type=Bone,Name="dyn_L_skirt_a_3"),Weight=0.5))
    Weight=1
   )
   > &RigUnit_ParentConstraint(
    AdvancedSettings=(InterpolationType=Average,RotationOrderForFilter=XZY)
    Child=(Type=Bone,Name="dyn_C_skirt_a_4")
    Parents=((Item=(Type=Bone,Name="dyn_R_skirt_a_4"),Weight=0.5),(Item=(Type=Bone,Name="dyn_L_skirt_a_4"),Weight=0.5))
    Weight=1
   )
   > &RigUnit_ParentConstraint(
    AdvancedSettings=(InterpolationType=Average,RotationOrderForFilter=XZY)
    Child=(Type=Bone,Name="dyn_C_skirt_a_5")
    Parents=((Item=(Type=Bone,Name="dyn_R_skirt_a_5"),Weight=0.5),(Item=(Type=Bone,Name="dyn_L_skirt_a_5"),Weight=0.5))
    Weight=1
   )
   n49 = RigVMVariableNode() {variable=L calfBend Alpha}
   n50 = RigVMVariableNode() {variable=L thighBack Alpha}
   n51 = RigVMVariableNode() {variable=R thighBack Alpha}
   n52 = RigVMVariableNode() {variable=R calfBend Alpha}
   Maximum = &RigVMFunction_MathDoubleMax(
    B=&RigVMFunction_MathDoubleSub(A=&RigVMFunction_MathDoubleAdd(A=@n49.Value, B=@n50.Value).Result, B=RigVMVariableNode() {variable=L thighForward Alpha}.Value).Result
    A=&RigVMFunction_MathDoubleSub(A=&RigVMFunction_MathDoubleAdd(B=@n51.Value, A=@n52.Value).Result, B=RigVMVariableNode() {variable=R thighForward Alpha}.Value).Result
   )
   > &RigUnit_ModifyTransforms(
    ItemToModify=((Item=(Type=Bone,Name="dyn_C_skirt_b_1"),Transform=(Rotation=(X=0,Y=-0.0523,Z=0,W=0.9986),Translation=(X=0,Y=0,Z=0),Scale3D=(X=1,Y=1,Z=1))))
    WeightMinimum=0
    Weight=@Maximum.Result
   )
   > &RigUnit_ModifyTransforms(
    ItemToModify=((Item=(Type=Bone,Name="dyn_C_skirt_b_2"),Transform=(Rotation=(X=0,Y=-0.0523,Z=0,W=0.9986),Translation=(X=0,Y=0,Z=0),Scale3D=(X=1,Y=1,Z=1))))
    WeightMinimum=0
    Weight=@Maximum.Result
   )
   n53 = RigVMVariableNode() {variable=L thighBack Alpha}
   n54 = RigVMVariableNode() {variable=L calfBend Alpha}
   n55 = RigVMVariableNode() {variable=R calfBend Alpha}
   n56 = RigVMVariableNode() {variable=R thighBack Alpha}
   n57 = &RigVMFunction_MathDoubleMax(
    B=&RigVMFunction_MathDoubleSub(B=0, A=&RigVMFunction_MathDoubleAdd(B=@n53.Value, A=@n54.Value).Result).Result
    A=&RigVMFunction_MathDoubleSub(B=0, A=&RigVMFunction_MathDoubleAdd(A=@n55.Value, B=@n56.Value).Result).Result
   )
   > &RigUnit_ModifyTransforms(
    ItemToModify=((Item=(Type=Bone,Name="dyn_C_skirt_a_1"),Transform=(Rotation=(X=0,Y=-0.0262,Z=0,W=0.9997),Translation=(X=0,Y=0,Z=0),Scale3D=(X=1,Y=1,Z=1))))
    WeightMinimum=0
    Weight=@n57.Result
   )
   > &RigUnit_ModifyTransforms(
    ItemToModify=((Item=(Type=Bone,Name="dyn_C_skirt_a_2"),Transform=(Rotation=(X=0,Y=-0.0262,Z=0,W=0.9997),Translation=(X=0,Y=0,Z=0),Scale3D=(X=1,Y=1,Z=1))))
    WeightMinimum=0
    Weight=@n57.Result
   )
   n58 = RigVMVariableNode() {variable=Movement Alpha}
   > &RigUnit_ModifyTransforms(
    ItemToModify=((Item=(Type=Bone,Name="dyn_L_skirt_b_1"),Transform=(Rotation=(X=0,Y=-0.0436,Z=0,W=0.999),Translation=(X=0,Y=0,Z=0),Scale3D=(X=1,Y=1,Z=1))))
    WeightMinimum=0
    Weight=@n58.Value
   )
   > &RigUnit_ModifyTransforms(
    ItemToModify=((Item=(Type=Bone,Name="dyn_L_skirt_c_1"),Transform=(Rotation=(X=0.0019,Y=-0.0436,Z=0.0436,W=0.9981),Translation=(X=0,Y=0,Z=0),Scale3D=(X=1,Y=1,Z=1))))
    WeightMinimum=0
    Weight=@n58.Value
   )
   > &RigUnit_ModifyTransforms(
    ItemToModify=((Item=(Type=Bone,Name="base_C_skirt_b"),Transform=(Rotation=(X=0,Y=-0.0436,Z=0,W=0.999),Translation=(X=0,Y=0,Z=0),Scale3D=(X=1,Y=1,Z=1))))
    WeightMinimum=0
    Weight=@n58.Value
   )
   > &RigUnit_ModifyTransforms(
    ItemToModify=((Item=(Type=Bone,Name="dyn_R_skirt_b_1"),Transform=(Rotation=(X=0,Y=-0.0436,Z=0,W=0.999),Translation=(X=0,Y=0,Z=0),Scale3D=(X=1,Y=1,Z=1))))
    WeightMinimum=0
    Weight=@n58.Value
   )
   > &RigUnit_ModifyTransforms(
    ItemToModify=((Item=(Type=Bone,Name="dyn_R_skirt_c_1"),Transform=(Rotation=(X=-0.0019,Y=-0.0436,Z=-0.0436,W=0.9981),Translation=(X=0,Y=0,Z=0),Scale3D=(X=1,Y=1,Z=1))))
    WeightMinimum=0
    Weight=@n58.Value
   )
   RigVMFunctionReturnNode(ExecuteContext=@RigVMFunction_Sequence.B)
   n59 = RigVMVariableNode() {variable=R thighForward Alpha}
   n60 = &RigVMFunction_MathDoubleSub(A=1, B=RigVMVariableNode() {variable=R thighSide Alpha}.Value)
   n61 = RigVMVariableNode() {variable=R calfBend Alpha}
   n62 = &RigVMFunction_MathDoubleMul(
    B=&RigVMFunction_MathDoubleSub(
     B=&RigVMFunction_MathDoubleMul(B=0.7, A=@n59.Value).Result
     A=&RigVMFunction_MathDoubleMul(A=@n60.Result, B=@n61.Value).Result
    ).Result
    A=RigVMVariableNode() {variable=pelvisLowToGround_INVERSE}.Value
   )
   &RigUnit_ParentConstraint(
    AdvancedSettings=(InterpolationType=Average,RotationOrderForFilter=XZY)
    Child=(Type=Bone,Name="dyn_L_skirt_b_1")
    ExecutePin=()
    Filter=(TranslationFilter=(bX=false,bY=false,bZ=false),RotationFilter=(bX=true,bY=true,bZ=True),ScaleFilter=(bX=True,bY=True,bZ=True))
    Parents=((Item=(Type=Bone,Name="calf_r"),Weight=1))
    Weight=@n62.Result
   )
   > &RigUnit_ParentConstraint(
    AdvancedSettings=(InterpolationType=Average,RotationOrderForFilter=XZY)
    Child=(Type=Bone,Name="dyn_L_skirt_c_1")
    Filter=(TranslationFilter=(bX=false,bY=false,bZ=false),RotationFilter=(bX=True,bY=True,bZ=True),ScaleFilter=(bX=True,bY=True,bZ=True))
    Parents=((Item=(Type=Bone,Name="calf_r"),Weight=1))
    Weight=@n62.Result
   )
   n63 = RigVMVariableNode() {variable=L thighForward Alpha}
   n64 = &RigVMFunction_MathDoubleSub(A=1, B=RigVMVariableNode() {variable=L thighSide Alpha}.Value)
   n65 = RigVMVariableNode() {variable=L calfBend Alpha}
   n66 = &RigVMFunction_MathDoubleMul(
    B=&RigVMFunction_MathDoubleSub(
     B=&RigVMFunction_MathDoubleMul(B=0.7, A=@n63.Value).Result
     A=&RigVMFunction_MathDoubleMul(A=@n64.Result, B=@n65.Value).Result
    ).Result
    A=RigVMVariableNode() {variable=pelvisLowToGround_INVERSE}.Value
   )
   &RigUnit_ParentConstraint(
    AdvancedSettings=(InterpolationType=Average,RotationOrderForFilter=XZY)
    Child=(Type=Bone,Name="dyn_R_skirt_b_1")
    ExecutePin=()
    Filter=(TranslationFilter=(bX=false,bY=false,bZ=false),RotationFilter=(bX=True,bY=True,bZ=True),ScaleFilter=(bX=True,bY=True,bZ=True))
    Parents=((Item=(Type=Bone,Name="calf_l"),Weight=1))
    Weight=@n66.Result
   )
   > &RigUnit_ParentConstraint(
    AdvancedSettings=(InterpolationType=Average,RotationOrderForFilter=XZY)
    Child=(Type=Bone,Name="dyn_R_skirt_c_1")
    Filter=(TranslationFilter=(bX=false,bY=false,bZ=false),RotationFilter=(bX=True,bY=True,bZ=True),ScaleFilter=(bX=True,bY=True,bZ=True))
    Parents=((Item=(Type=Bone,Name="calf_l"),Weight=1))
    Weight=@n66.Result
   )
   &RigUnit_ParentConstraint(
    AdvancedSettings=(InterpolationType=Average,RotationOrderForFilter=XZY)
    Child=(Type=Bone,Name="dyn_R_skirt_b_1")
    ExecutePin=()
    Filter=(TranslationFilter=(bX=false,bY=false,bZ=false),RotationFilter=(bX=false,bY=true,bZ=true),ScaleFilter=(bX=True,bY=True,bZ=True))
    Parents=((Item=(Type=Bone,Name="dyn_R_skirt_c_1"),Weight=1))
    Weight=0.2
   )
   &RigUnit_ParentConstraint(
    AdvancedSettings=(InterpolationType=Average,RotationOrderForFilter=XZY)
    Child=(Type=Bone,Name="dyn_L_skirt_b_1")
    ExecutePin=()
    Filter=(TranslationFilter=(bX=false,bY=false,bZ=false),RotationFilter=(bX=false,bY=True,bZ=True),ScaleFilter=(bX=True,bY=True,bZ=True))
    Parents=((Item=(Type=Bone,Name="dyn_L_skirt_c_1"),Weight=1))
    Weight=0.2
   )
  ]}
 RigVMFunctionLibrary {nodes [
   Additional Skirt Posing = RigVMCollapseNode(ExecuteContext=())
   Neck Collar = RigVMCollapseNode(ExecuteContext=())
   Neck Collar Constraints = RigVMCollapseNode(ExecuteContext=())
   Skirt Correctives = RigVMCollapseNode(ExecuteContext=())
   Solve Knee = RigVMCollapseNode(
    AimAxis=(X=1,Y=0,Z=0)
    Ball=(Type=None,Name="None")
    Bind Front Knee=(Type=None,Name="None")
    Bind Rear Knee=(Type=None,Name="None")
    Bind Rear Thigh=(Type=None,Name="None")
    ExecuteContext=()
    Knee=(Type=None,Name="None")
   )
   Solve Thigh = RigVMCollapseNode(
    Bind Front Skirt Joint=(Type=None,Name="None")
    Bind Rear Skirt Joint=(Type=None,Name="None")
    Bind Side Skirt Joint=(Type=None,Name="None")
    ExecuteContext=()
    Push Vector=(X=0,Y=0,Z=0)
    Thigh Joint=(Type=None,Name="None")
   )
  ]}
 RigVMFunction_Sequence_1_ContainedGraph {nodes [
   n1 = RigVMFunctionEntryNode()
   RigVMFunction_Sequence = &RigVMFunction_Sequence(ExecuteContext=@n1.ExecuteContext)
   n2 = &RigVMFunction_Sequence(ExecuteContext=@RigVMFunction_Sequence.B)
   n3 = &RigVMFunction_Sequence(ExecuteContext=@n2.B)
   n4 = &RigVMFunction_Sequence(ExecuteContext=@n3.B)
   n5 = &RigVMFunction_Sequence(ExecuteContext=@n4.B)
   n6 = &RigVMFunction_Sequence(ExecuteContext=@n5.B)
   n7 = &RigVMFunction_Sequence(ExecuteContext=@n6.B)
   RigVMFunctionReturnNode(A=@RigVMFunction_Sequence.A, B=@n2.A, C=@n3.A, D=@n4.A, E=@n5.A, F=@n6.A, G=@n7.A, H=@n7.B)
  ]}
 RigVMFunction_Sequence_1_ContainedGraph {nodes [
   n1 = RigVMFunctionEntryNode()
   RigVMFunction_Sequence = &RigVMFunction_Sequence(ExecuteContext=@n1.ExecuteContext)
   n2 = &RigVMFunction_Sequence(ExecuteContext=@RigVMFunction_Sequence.B)
   RigVMFunctionReturnNode(A=@RigVMFunction_Sequence.A, B=@n2.A, C=@n2.B)
  ]}
 RigVMModel {nodes [
   RigUnit_BeginExecution = &RigUnit_BeginExecution()
   FN_DeformRig_FNC = RigVMFunctionReferenceNode(
    DeformRig_ Curve="(deformRig_l_upperarm_up_alpha_72_#1=False,deformRig_r_upperarm_up_alpha_73_#2=False,deformRig_l_upperarm_fr_alpha_74_#3=False,deformRig_r_upperarm_fr_alpha_75_CB51B8 ...(truncated)"
    DeformRig_Bool="(upperarm_1_#4=True,elbow_3_#5=True,wrist_7_#6=True,groin_8_#7=True,glute_17_#8=True,knee_9_40C408324CA261 ...(truncated)"
    DeformRig_Debug=(upperarm_debug_18_#4=False,wrist_debug_20_#6=False,lat_debug_26_#9=False,groinGlute_debug_27_#7=False)
    DeformRig_Value="(upperarm_up_18_#10=(Rotation=(X=0,Y=0,Z=0,W=1),Translation=(X=0,Y=0,Z=8),Scale3D=(X=0.7,Y=1,Z=1)),upperarm_fwd_fr_19_#11=(Rotation=(X=0,Y=0,Z=0,W=1),Translation=(X=0,Y=10,Z=0),Scale3D=(X=0.5,Y=1,Z= ...(truncated)"
    ExecuteContext=@RigUnit_BeginExecution.ExecutePin
   )
   > Skirt Correctives_1 = RigVMFunctionReferenceNode()
   > Neck Collar Constraints_1 = RigVMFunctionReferenceNode()
   > Neck Collar_1 = RigVMFunctionReferenceNode()
   > &RigUnit_ModifyTransforms(
    ItemToModify="((Item=(Type=Bone,Name=\"Collar_back_M\"),Transform=(Rotation=(X=0.0418,Y=-0.0313,Z=-0.7567,W=-0.6517),Translation=(X=0,Y=0,Z=0),Scale3D=(X=1,Y=1,Z=1))),(Item=(Type=Bone,Name=\"Collar_back_R\"),Transform=(Rotation=(X=0,Y=0,Z=-0.8181,W=-0.575),Translation=(X=0, ...(truncated)"
    WeightMinimum=0
    Weight=RigVMVariableNode() {variable=Head Back C}.Value
   )
   > &RigUnit_ModifyTransforms(
    ItemToModify="((Item=(Type=Bone,Name=\"Collar_side_L\"),Transform=(Rotation=(X=0,Y=0,Z=0.2381,W=0.9712),Translation=(X=0,Y=0,Z=0),Scale3D=(X=1,Y=1,Z=1))),(Item=(Type=Bone,Name=\"Collar_front_L\"),Transform=(Rotation=(X=0,Y=0,Z=0,W=1),Translation=(X=0,Y=0,Z=0),Scale3D=(X=1,Y ...(truncated)"
    WeightMinimum=0
    Weight=RigVMVariableNode() {variable="Head Side L "}.Value
   )
   > &RigUnit_ModifyTransforms(
    ItemToModify="((Item=(Type=Bone,Name=\"Collar_side_R\"),Transform=(Rotation=(X=0.6655,Y=-0.3191,Z=0.3446,W=0.5801),Translation=(X=0,Y=0,Z=0),Scale3D=(X=1,Y=1,Z=1))),(Item=(Type=Bone,Name=\"Collar_front_R\"),Transform=(Rotation=(X=0,Y=0,Z=0,W=1),Translation=(X=0,Y=0,Z=0),Sca ...(truncated)"
    WeightMinimum=0
    Weight=RigVMVariableNode() {variable=Head Side R}.Value
   )
  ]}
 "RigVMModel Construction Solve " {nodes [
   n1 = &RigUnit_PrepareForExecution()
   Control Stack at Item = RigVMFunctionReferenceNode(
    Add Bottom Null=False
    Add Null=true
    Add Secondary Control=False
    Control Suffix=_ctrl
    Item=(Name="head",Type=Bone)
    Name=ref_head
    Null Suffix=_null
    Parent=(Name="neck_02",Type=Bone)
    Secondary Settings="(Shape=(bVisible=True,Name=\"Default\",Color=(R=1,G=0,B=0,A=1),Transform=(Rotation=(X=0,Y=0,Z=0,W=1),Translation=(X=0,Y=0,Z=0),Scale3D=(X=1,Y=1,Z=1))),Proxy=(bIsProxy=False,ShapeVisibility=BasedOnSelection),DisplayName=\"None\",InitialSpace=LocalSpace,bUsePref ...(truncated)"
    Settings="(Shape=(bVisible=True,Name=\"Pyramid_Thin\",Color=(R=1,G=0,B=0,A=1),Transform=(Rotation=(X=0,Y=0,Z=0,W=1),Translation=(X=0,Y=0,Z=0),Scale3D=(X=1,Y=1,Z=1))),Proxy=(bIsProxy=False,ShapeVisibility=BasedOnSelection),DisplayName=\"None\",InitialSpace=LocalSpace,bUs ...(truncated)"
    ExecuteContext=@n1.ExecutePin
   )
   > Control Stack at Item_1 = RigVMFunctionReferenceNode(
    Add Bottom Null=False
    Add Null=true
    Add Secondary Control=False
    Control Suffix=_ctrl
    Item=(Name="neck_02",Type=Bone)
    Name=ref_neck
    Null Suffix=_null
    Parent=(Name="spine_05",Type=Bone)
    Secondary Settings="(Shape=(bVisible=True,Name=\"Default\",Color=(R=1,G=0,B=0,A=1),Transform=(Rotation=(X=0,Y=0,Z=0,W=1),Translation=(X=0,Y=0,Z=0),Scale3D=(X=1,Y=1,Z=1))),Proxy=(bIsProxy=False,ShapeVisibility=BasedOnSelection),DisplayName=\"None\",InitialSpace=LocalSpace,bUsePref ...(truncated)"
    Settings="(Shape=(bVisible=True,Name=\"Pyramid_Thin\",Color=(R=1,G=0,B=0,A=1),Transform=(Rotation=(X=0,Y=0,Z=0,W=1),Translation=(X=0,Y=0,Z=0),Scale3D=(X=1,Y=1,Z=1))),Proxy=(bIsProxy=False,ShapeVisibility=BasedOnSelection),DisplayName=\"None\",InitialSpace=LocalSpace,bUs ...(truncated)"
   )
   > Control Stack at Item_2 = RigVMFunctionReferenceNode(
    Add Bottom Null=False
    Add Null=true
    Add Secondary Control=False
    Control Suffix=_ctrl
    Item=(Name="clavicle_l",Type=Bone)
    Name=ref_clavicle_l
    Null Suffix=_null
    Parent=(Name="spine_05",Type=Bone)
    Secondary Settings="(Shape=(bVisible=True,Name=\"Default\",Color=(R=1,G=0,B=0,A=1),Transform=(Rotation=(X=0,Y=0,Z=0,W=1),Translation=(X=0,Y=0,Z=0),Scale3D=(X=1,Y=1,Z=1))),Proxy=(bIsProxy=False,ShapeVisibility=BasedOnSelection),DisplayName=\"None\",InitialSpace=LocalSpace,bUsePref ...(truncated)"
    Settings="(Shape=(bVisible=True,Name=\"Pyramid_Thin\",Color=(R=1,G=0,B=0,A=1),Transform=(Rotation=(X=0,Y=0,Z=0,W=1),Translation=(X=0,Y=0,Z=0),Scale3D=(X=1,Y=1,Z=1))),Proxy=(bIsProxy=False,ShapeVisibility=BasedOnSelection),DisplayName=\"None\",InitialSpace=LocalSpace,bUs ...(truncated)"
   )
   > Control Stack at Item_3 = RigVMFunctionReferenceNode(
    Add Bottom Null=False
    Add Null=true
    Add Secondary Control=False
    Control Suffix=_ctrl
    Item=(Name="clavicle_r",Type=Bone)
    Name=ref_clavicle_r
    Null Suffix=_null
    Parent=(Name="spine_05",Type=Bone)
    Secondary Settings="(Shape=(bVisible=True,Name=\"Default\",Color=(R=1,G=0,B=0,A=1),Transform=(Rotation=(X=0,Y=0,Z=0,W=1),Translation=(X=0,Y=0,Z=0),Scale3D=(X=1,Y=1,Z=1))),Proxy=(bIsProxy=False,ShapeVisibility=BasedOnSelection),DisplayName=\"None\",InitialSpace=LocalSpace,bUsePref ...(truncated)"
    Settings="(Shape=(bVisible=True,Name=\"Pyramid_Thin\",Color=(R=1,G=0,B=0,A=1),Transform=(Rotation=(X=0,Y=0,Z=0,W=1),Translation=(X=0,Y=0,Z=0),Scale3D=(X=1,Y=1,Z=1))),Proxy=(bIsProxy=False,ShapeVisibility=BasedOnSelection),DisplayName=\"None\",InitialSpace=LocalSpace,bUs ...(truncated)"
   )
   > &RigUnit_HierarchyAddNull(
    Name=Ref_Thigh_L
    Parent=(Type=Bone,Name="pelvis")
    Space=GlobalSpace
    Transform=&RigUnit_GetTransform(Item=(Type=Bone,Name="thigh_twist_01_l"), bInitial=False).Transform
   )
   > &RigUnit_HierarchyAddNull(
    Name=Ref_Thigh_R
    Parent=(Type=Bone,Name="pelvis")
    Space=GlobalSpace
    Transform=&RigUnit_GetTransform(Item=(Type=Bone,Name="thigh_twist_01_r"), bInitial=False).Transform
   )
   &RigUnit_HierarchyImportFromSkeleton(ExecutePin=(), NameSpace=None, bIncludeCurves=False, bIncludeMeshSockets=False, bIncludeVirtualBones=True)
  ]}
]
hierarchy [
 CR Thigh Raised Front L {type=Curve, loc=<0 0 0>, rot=<0 0 0 1>}
 "CR ThighRaised  Front R" {type=Curve, loc=<0 0 0>, rot=<0 0 0 1>}
 CR Rear Knee Bend L {type=Curve, loc=<0 0 0>, rot=<0 0 0 1>}
 "CR  Rear Knee Bend R" {type=Curve, loc=<0 0 0>, rot=<0 0 0 1>}
 "CR Foot Near Pelvis L " {type=Curve, loc=<0 0 0>, rot=<0 0 0 1>}
 CR Foot Near Pelvis R {type=Curve, loc=<0 0 0>, rot=<0 0 0 1>}
 CR Thigh L Across R {type=Curve, loc=<0 0 0>, rot=<0 0 0 1>}
 root {loc=<0 0 0>, rot=<0 0 0 1>}
 pelvis {parent=root, loc=<0 -0.5301 98.0571>, rot=<0.0127 0.707 -0.0127 -0.707>}
 spine_01 {parent=pelvis, loc=<6.1208 -0.9328 0>, rot=<0 0 0.034 -0.9994>}
 spine_02 {parent=spine_01, loc=<6.5997 -0.2095 0>, rot=<0 0 0.0238 -0.9997>}
 spine_03 {parent=spine_02, loc=<6.4914 -0.1237 0>, rot=<0 0 0.0149 -0.9999>}
 spine_04 {parent=spine_03, loc=<6.9598 0.0794 0>, rot=<0 0 -0.0132 -0.9999>}
 spine_05 {parent=spine_04, loc=<6.9987 0 0>, rot=<0 0 -0.0015 -1>}
 neck_01 {parent=spine_05, loc=<11.3138 0.1529 0>, rot=<0 0 -0.2003 -0.9797>}
 neck_02 {parent=neck_01, loc=<4.5627 0 0>, rot=<0 0 0.0504 -0.9987>}
 head {parent=neck_02, loc=<5.3914 0 0>, rot=<0 0 0.115 -0.9934>}
 clavicle_l {parent=spine_05, loc=<5.4831 -0.1053 -3.2946>, rot=<0.0465 -0.7289 -0.009 -0.6829>}
 upperarm_l {parent=clavicle_l, loc=<9.8658 -1.2136 -0.4105>, rot=<0.0164 -0.3245 -0.0269 -0.9454>}
 lowerarm_l {parent=upperarm_l, loc=<26.465 -0.8345 0.259>, rot=<0.0036 0.0161 -0.2298 -0.9731>}
 hand_l {parent=lowerarm_l, loc=<22.7355 -0.3203 -0.6105>, rot=<-0.6784 -0.0223 0.0735 -0.7307>}
 thumb_01_l {parent=hand_l, loc=<1.3118 -0.3191 -2.459>, rot=<0.6546 -0.4042 0.0664 -0.6354>}
 thumb_02_l {parent=thumb_01_l, loc=<5.9336 -0.0448 0.0401>, rot=<-0.0037 0.0213 0.298 -0.9543>}
 thumb_03_l {parent=thumb_02_l, loc=<2.2223 0.3682 -0.0907>, rot=<0.0118 -0.013 0.0066 -0.9998>}
 index_metacarpal_l {parent=hand_l, loc=<1.6535 0.0183 -2.96>, rot=<0.0158 -0.0124 -0.0173 -0.9996>}
 index_01_l {parent=index_metacarpal_l, loc=<8.3008 0.0357 -0.0524>, rot=<-0.0044 0.0485 0.0924 -0.9945>}
 index_02_l {parent=index_01_l, loc=<3.3587 -0.3004 0.0452>, rot=<-0.002 0.0027 0.3225 -0.9466>}
 index_03_l {parent=index_02_l, loc=<2.5578 -0.123 -0.0252>, rot=<-0.002 0.0067 0.2558 -0.9667>}
 middle_metacarpal_l {parent=hand_l, loc=<2.1856 -0.1725 -0.5984>, rot=<0.006 0.0168 -0.0329 -0.9993>}
 middle_01_l {parent=middle_metacarpal_l, loc=<7.3275 -0.2447 0.0129>, rot=<-0.0208 0.0042 0.0858 -0.9961>}
 middle_02_l {parent=middle_01_l, loc=<4.0123 0.1401 0.1175>, rot=<-0.0218 0.0258 0.3512 -0.9357>}
 middle_03_l {parent=middle_02_l, loc=<2.6077 0.0534 0.1373>, rot=<0.0082 0.0188 0.1708 -0.9851>}
 ring_metacarpal_l {parent=hand_l, loc=<2.2074 -0.3625 1.3638>, rot=<-0.0284 0.0722 -0.0458 -0.9959>}
 ring_01_l {parent=ring_metacarpal_l, loc=<7.0295 -0.6197 -0.1446>, rot=<-0.0489 -0.0252 0.1892 -0.9804>}
 ring_02_l {parent=ring_01_l, loc=<3.6174 0.4647 -0.1027>, rot=<0 0 0.2806 -0.9598>}
 ring_03_l {parent=ring_02_l, loc=<2.4152 0.114 0.0199>, rot=<0 0 0.1189 -0.9929>}
 pinky_metacarpal_l {parent=hand_l, loc=<1.7616 -0.9753 3.3814>, rot=<-0.0397 0.0814 -0.069 -0.9935>}
 pinky_01_l {parent=pinky_metacarpal_l, loc=<7.1709 -0.4831 -0.1094>, rot=<-0.0654 -0.0035 0.2242 -0.9723>}
 pinky_02_l {parent=pinky_01_l, loc=<2.8274 0.0893 -0.1043>, rot=<0.0037 0.0011 0.2854 -0.9584>}
 pinky_03_l {parent=pinky_02_l, loc=<2.2046 0.0307 -0.0078>, rot=<-0.0002 0.0098 0.1445 -0.9895>}
 deform_wrist_dn_base_l {parent=hand_l, loc=<0 -1.5326 0>, rot=<0 0 0 -1>}
 deform_wrist_dn_l {parent=deform_wrist_dn_base_l, loc=<0 0 0>, rot=<0 0 0 1>}
 deform_wrist_up_base_l {parent=hand_l, loc=<0 2.3009 0>, rot=<0 0 0 -1>}
 deform_wrist_up_l {parent=deform_wrist_up_base_l, loc=<0 0 0>, rot=<0 0 0 1>}
 lowerarm_twist_02_l {parent=lowerarm_l, loc=<10.6707 0 0>, rot=<0 0 0 1>}
 lowerarm_twist_01_l {parent=lowerarm_l, loc=<18.316 0 0>, rot=<0 0 0 1>}
 deform_elbow_in_base_l {parent=lowerarm_l, loc=<0.1834 3.6062 -0.2572>, rot=<0 0 0 -1>}
 deform_elbow_in_l {parent=deform_elbow_in_base_l, loc=<0 0 0>, rot=<0 0 0 1>}
 deform_elbow_out_base_l {parent=lowerarm_l, loc=<0 -4 0>, rot=<0 0 0 1>}
 deform_elbow_out_l {parent=deform_elbow_out_base_l, loc=<0 0 0>, rot=<0 0 0 1>}
 upperarm_twist_01_l {parent=upperarm_l, loc=<7.2135 -0.1134 0>, rot=<0 0 0 1>}
 deform_upperarm_bk_base_l {parent=upperarm_twist_01_l, loc=<-6.4053 -4.3746 0.314>, rot=<0 0 0 -1>}
 deform_upperarm_bk_l {parent=deform_upperarm_bk_base_l, loc=<0 0 0>, rot=<0 0 0 1>}
 deform_upperarm_fr_base_l {parent=upperarm_twist_01_l, loc=<-5.6312 5.092 0.3354>, rot=<0 0 0 -1>}
 deform_upperarm_fr_l {parent=deform_upperarm_fr_base_l, loc=<0 0 0>, rot=<0 0 0 1>}
 deform_upperarm_up_base_l {parent=upperarm_twist_01_l, loc=<-8.891 0.4825 3.9517>, rot=<0 0 0 -1>}
 deform_upperarm_up_l {parent=deform_upperarm_up_base_l, loc=<0 0 0>, rot=<0 0 0 1>}
 upperarm_twist_02_l {parent=upperarm_l, loc=<19.7998 -0.1134 0>, rot=<0 0 0 1>}
 Collar_front_L {parent=clavicle_l, loc=<0.451 7.2089 2.3253>, rot=<0.541 0.6134 0.0236 -0.5749>}
 Collar_side_L {parent=clavicle_l, loc=<3.5947 2.2344 4.7524>, rot=<0.6 0.5668 0.2103 -0.5239>}
 Collar_back_L {parent=clavicle_l, loc=<2.1907 -3.2033 6.4481>, rot=<0.7297 0.1526 0.5669 -0.3505>}
 clavicle_r {parent=spine_05, loc=<5.4831 -0.1053 3.2946>, rot=<0.7289 0.0465 -0.6829 0.009>}
 upperarm_r {parent=clavicle_r, loc=<-9.8658 1.2136 0.4105>, rot=<0.0164 -0.3245 -0.0269 -0.9454>}
 lowerarm_r {parent=upperarm_r, loc=<-26.465 0.8345 -0.259>, rot=<0.0036 0.0161 -0.2298 -0.9731>}
 hand_r {parent=lowerarm_r, loc=<-22.7355 0.3203 0.6106>, rot=<-0.6784 -0.0223 0.0735 -0.7307>}
 thumb_01_r {parent=hand_r, loc=<-1.3118 0.3191 2.459>, rot=<0.6546 -0.4042 0.0664 -0.6354>}
 thumb_02_r {parent=thumb_01_r, loc=<-5.9336 0.0448 -0.0401>, rot=<-0.0037 0.0213 0.298 -0.9543>}
 thumb_03_r {parent=thumb_02_r, loc=<-2.2223 -0.3682 0.0907>, rot=<0.0118 -0.013 0.0066 -0.9998>}
 index_metacarpal_r {parent=hand_r, loc=<-1.6535 -0.0183 2.96>, rot=<0.0158 -0.0124 -0.0173 -0.9996>}
 index_01_r {parent=index_metacarpal_r, loc=<-8.3008 -0.0357 0.0524>, rot=<-0.0044 0.0485 0.0924 -0.9945>}
 index_02_r {parent=index_01_r, loc=<-3.3587 0.3004 -0.0452>, rot=<-0.002 0.0027 0.3225 -0.9466>}
 index_03_r {parent=index_02_r, loc=<-2.5578 0.123 0.0252>, rot=<-0.002 0.0067 0.2558 -0.9667>}
 middle_metacarpal_r {parent=hand_r, loc=<-2.1856 0.1725 0.5984>, rot=<0.006 0.0168 -0.0329 -0.9993>}
 middle_01_r {parent=middle_metacarpal_r, loc=<-7.3275 0.2447 -0.0129>, rot=<-0.0208 0.0042 0.0858 -0.9961>}
 middle_02_r {parent=middle_01_r, loc=<-4.0123 -0.1401 -0.1175>, rot=<-0.0218 0.0258 0.3512 -0.9357>}
 middle_03_r {parent=middle_02_r, loc=<-2.6077 -0.0534 -0.1373>, rot=<0.0082 0.0188 0.1708 -0.9851>}
 ring_metacarpal_r {parent=hand_r, loc=<-2.2074 0.3625 -1.3638>, rot=<-0.0284 0.0722 -0.0458 -0.9959>}
 ring_01_r {parent=ring_metacarpal_r, loc=<-7.0295 0.6197 0.1446>, rot=<-0.0489 -0.0252 0.1892 -0.9804>}
 ring_02_r {parent=ring_01_r, loc=<-3.6174 -0.4647 0.1027>, rot=<0 0 0.2806 -0.9598>}
 ring_03_r {parent=ring_02_r, loc=<-2.4152 -0.114 -0.0199>, rot=<0 0 0.1189 -0.9929>}
 pinky_metacarpal_r {parent=hand_r, loc=<-1.7616 0.9753 -3.3814>, rot=<-0.0397 0.0814 -0.069 -0.9935>}
 pinky_01_r {parent=pinky_metacarpal_r, loc=<-7.1709 0.4831 0.1094>, rot=<-0.0654 -0.0035 0.2242 -0.9723>}
 pinky_02_r {parent=pinky_01_r, loc=<-2.8274 -0.0893 0.1043>, rot=<0.0037 0.0011 0.2854 -0.9584>}
 pinky_03_r {parent=pinky_02_r, loc=<-2.2046 -0.0307 0.0078>, rot=<-0.0002 0.0098 0.1445 -0.9895>}
 deform_wrist_dn_base_r {parent=hand_r, loc=<-0.0001 1.5328 0>, rot=<0 0 0 -1>}
 deform_wrist_dn_r {parent=deform_wrist_dn_base_r, loc=<0 0 0>, rot=<0 0 0 1>}
 deform_wrist_up_base_r {parent=hand_r, loc=<-0.0002 -2.3007 -0.0001>, rot=<0 0 0 -1>}
 deform_wrist_up_r {parent=deform_wrist_up_base_r, loc=<0 0 0>, rot=<0 0 0 1>}
 lowerarm_twist_02_r {parent=lowerarm_r, loc=<-10.6707 0 0>, rot=<0 0 0 -1>}
 lowerarm_twist_01_r {parent=lowerarm_r, loc=<-18.316 0 0>, rot=<0 0 0 -1>}
 deform_elbow_in_base_r {parent=lowerarm_r, loc=<-0.1834 -3.6062 0.2572>, rot=<0 0 0 -1>}
 deform_elbow_in_r {parent=deform_elbow_in_base_r, loc=<0 0 0>, rot=<0 0 0 1>}
 deform_elbow_out_base_r {parent=lowerarm_r, loc=<0 4 -0.0001>, rot=<0 0 0 -1>}
 deform_elbow_out_r {parent=deform_elbow_out_base_r, loc=<0 0 0>, rot=<0 0 0 1>}
 upperarm_twist_01_r {parent=upperarm_r, loc=<-7.2135 0.1134 0>, rot=<0 0 0 -1>}
 deform_upperarm_bk_base_r {parent=upperarm_twist_01_r, loc=<6.4053 4.3746 -0.314>, rot=<0 0 0 -1>}
 deform_upperarm_bk_r {parent=deform_upperarm_bk_base_r, loc=<0 0 0>, rot=<0 0 0 1>}
 deform_upperarm_fr_base_r {parent=upperarm_twist_01_r, loc=<5.6312 -5.092 -0.3354>, rot=<0 0 0 -1>}
 deform_upperarm_fr_r {parent=deform_upperarm_fr_base_r, loc=<0 0 0>, rot=<0 0 0 1>}
 deform_upperarm_up_base_r {parent=upperarm_twist_01_r, loc=<8.891 -0.4825 -3.9517>, rot=<0 0 0 -1>}
 deform_upperarm_up_r {parent=deform_upperarm_up_base_r, loc=<0 0 0>, rot=<0 0 0 1>}
 upperarm_twist_02_r {parent=upperarm_r, loc=<-19.7998 0.1134 0>, rot=<0 0 0 -1>}
 Collar_front_R {parent=clavicle_r, loc=<-0.451 -7.2089 -2.3255>, rot=<0.5316 0.6132 -0.1253 -0.5707>}
 Collar_side_R {parent=clavicle_r, loc=<-3.5948 -2.2344 -4.7519>, rot=<0.5952 0.5787 0.2059 -0.5181>}
 Collar_back_R {parent=clavicle_r, loc=<-2.1907 3.2033 -6.4485>, rot=<0.7053 0.2566 0.5541 -0.36>}
 attach_backpack {parent=spine_05, loc=<0.0249 -13.7125 -0.0012>, rot=<-0.0095 -0.7067 0.0093 -0.7074>}
 Backpack_BR {parent=attach_backpack, type=Null, loc=<0 0 0>, rot=<0 0 0 1>}
 attach_cape {parent=spine_05, loc=<9.0765 -11.9267 0.0053>, rot=<0.0615 -0.704 -0.0615 -0.7048>}
 cape {parent=attach_cape, type=Null, loc=<0 0 0>, rot=<0 0 0 1>}
 Collar_back_M {parent=spine_05, loc=<12.673 -5.0319 -0.0168>, rot=<0.9943 -0.1017 0.0206 0.0234>}
 deform_pec_aim_base_l {parent=spine_04, loc=<3.2456 9.3309 -1.2237>, rot=<0.0603 -0.4033 0.3318 -0.8507>}
 deform_pec_aim_l {parent=deform_pec_aim_base_l, loc=<0 0 0>, rot=<0 0 0 1>}
 deform_pec_base_l {parent=deform_pec_aim_l, loc=<9.3665 0 0>, rot=<0 0 0 1>}
 deform_pec_l {parent=deform_pec_base_l, loc=<0 0 0>, rot=<0 0 0 1>}
 deform_pec_aim_base_r {parent=spine_04, loc=<3.2452 9.3309 1.2237>, rot=<-0.4033 -0.0603 0.8507 0.3318>}
 deform_pec_aim_r {parent=deform_pec_aim_base_r, loc=<0 0 0>, rot=<0 0 0 1>}
 deform_pec_base_r {parent=deform_pec_aim_r, loc=<-9.3667 -0.0003 -0.0004>, rot=<0 0 0 1>}
 deform_pec_r {parent=deform_pec_base_r, loc=<0 0 0>, rot=<0 0 0 1>}
 Skirt_Master {parent=spine_01, loc=<0 0 0>, rot=<0 0 0 1>}
 dyn_Skirt_Master {parent=Skirt_Master, loc=<0 0 0>, rot=<0 0 0 1>}
 base_R_skirt_b {parent=dyn_Skirt_Master, loc=<-4.7607 0.1416 14.446>, rot=<0.0478 0.0038 0.9983 -0.0318>}
 dyn_R_skirt_b_1 {parent=base_R_skirt_b, loc=<0 0 0>, rot=<0 0 0 1>}
 dyn_R_skirt_b_2 {parent=dyn_R_skirt_b_1, loc=<25.4054 0 1.0687>, rot=<0 0.0137 0 -0.9999>}
 dyn_R_skirt_b_3 {parent=dyn_R_skirt_b_2, loc=<15.2446 0 0>, rot=<0 -0.003 0 -1>}
 dyn_R_skirt_b_4 {parent=dyn_R_skirt_b_3, loc=<12.19 0 0>, rot=<0 0.0021 0 -1>}
 dyn_R_skirt_b_5 {parent=dyn_R_skirt_b_4, loc=<12.926 0 0>, rot=<0 -0.0091 0 -1>}
 dyn_R_skirt_b_6 {parent=dyn_R_skirt_b_5, loc=<11.3264 0 0>, rot=<0 0 0 1>}
 base_R_skirt_c {parent=dyn_Skirt_Master, loc=<-3.9289 -10.6021 10.6318>, rot=<0.0615 -0.6851 0.7259 0.0012>}
 dyn_R_skirt_c_1 {parent=base_R_skirt_c, loc=<0 0 0>, rot=<0 0 0 1>}
 dyn_R_skirt_c_2 {parent=dyn_R_skirt_c_1, loc=<25.4054 0 1.0687>, rot=<0 0.0137 0 -0.9999>}
 dyn_R_skirt_c_3 {parent=dyn_R_skirt_c_2, loc=<15.2446 0 0>, rot=<0 -0.003 0 -1>}
 dyn_R_skirt_c_4 {parent=dyn_R_skirt_c_3, loc=<12.19 0 0>, rot=<0 0.0021 0 -1>}
 dyn_R_skirt_c_5 {parent=dyn_R_skirt_c_4, loc=<12.926 0 0>, rot=<0 -0.0091 0 -1>}
 dyn_R_skirt_c_6 {parent=dyn_R_skirt_c_5, loc=<11.3264 0 0>, rot=<0 0 0 1>}
 base_C_skirt_b {parent=dyn_Skirt_Master, loc=<-3.9083 -12.1771 -0.0001>, rot=<0.035 -0.7062 0.7062 -0.035>}
 dyn_C_skirt_b_1 {parent=base_C_skirt_b, loc=<0 0 0>, rot=<0 0 0 1>}
 dyn_C_skirt_b_2 {parent=dyn_C_skirt_b_1, loc=<25.4054 0 1.0687>, rot=<0 0.0137 0 -0.9999>}
 dyn_C_skirt_b_3 {parent=dyn_C_skirt_b_2, loc=<15.2446 0 0>, rot=<0 -0.003 0 -1>}
 dyn_C_skirt_b_4 {parent=dyn_C_skirt_b_3, loc=<12.19 0 0>, rot=<0 0.0021 0 -1>}
 dyn_C_skirt_b_5 {parent=dyn_C_skirt_b_4, loc=<12.926 0 0>, rot=<0 -0.0091 0 -1>}
 dyn_C_skirt_b_6 {parent=dyn_C_skirt_b_5, loc=<11.3264 0 0>, rot=<0 0 0 1>}
 base_L_skirt_c {parent=dyn_Skirt_Master, loc=<-4.4992 -10.3709 -10.129>, rot=<-0.0125 -0.6759 0.7349 -0.0546>}
 dyn_L_skirt_c_1 {parent=base_L_skirt_c, loc=<0 0 0>, rot=<0.0514 0 0 -0.9987>}
 dyn_L_skirt_c_2 {parent=dyn_L_skirt_c_1, loc=<25.4054 0 1.0687>, rot=<0 0.0137 0 -0.9999>}
 dyn_L_skirt_c_3 {parent=dyn_L_skirt_c_2, loc=<15.2446 0 0>, rot=<0 -0.003 0 -1>}
 dyn_L_skirt_c_4 {parent=dyn_L_skirt_c_3, loc=<12.19 0 0>, rot=<0 0.0021 0 -1>}
 dyn_L_skirt_c_5 {parent=dyn_L_skirt_c_4, loc=<12.926 0 0>, rot=<0 -0.0091 0 -1>}
 dyn_L_skirt_c_6 {parent=dyn_L_skirt_c_5, loc=<11.3264 0 0>, rot=<0 0 0 1>}
 base_L_skirt_b {parent=dyn_Skirt_Master, loc=<-3.7649 0.0738 -13.9775>, rot=<0.032 -0.9984 0.0008 -0.0456>}
 dyn_L_skirt_b_1 {parent=base_L_skirt_b, loc=<0 0 0>, rot=<0 0 0 1>}
 dyn_L_skirt_b_2 {parent=dyn_L_skirt_b_1, loc=<25.4054 0 1.0687>, rot=<0 0.0137 0 -0.9999>}
 dyn_L_skirt_b_3 {parent=dyn_L_skirt_b_2, loc=<15.2446 0 0>, rot=<0 -0.003 0 -1>}
 dyn_L_skirt_b_4 {parent=dyn_L_skirt_b_3, loc=<12.19 0 0>, rot=<0 0.0021 0 -1>}
 dyn_L_skirt_b_5 {parent=dyn_L_skirt_b_4, loc=<12.926 0 0>, rot=<0 -0.0091 0 -1>}
 dyn_L_skirt_b_6 {parent=dyn_L_skirt_b_5, loc=<11.3264 0 0>, rot=<0 0 0 1>}
 base_R_skirt_a {parent=dyn_Skirt_Master, loc=<-4.1798 8.2566 10.632>, rot=<-0.0276 -0.7066 -0.7061 0.037>}
 dyn_R_skirt_a_1 {parent=base_R_skirt_a, loc=<0 0 0>, rot=<0 0 0 1>}
 dyn_R_skirt_a_2 {parent=dyn_R_skirt_a_1, loc=<25.4054 0 1.0687>, rot=<0 0.0137 0 -0.9999>}
 dyn_R_skirt_a_3 {parent=dyn_R_skirt_a_2, loc=<15.2446 0 0>, rot=<0 -0.003 0 -1>}
 dyn_R_skirt_a_4 {parent=dyn_R_skirt_a_3, loc=<12.19 0 0>, rot=<0 0.0021 0 -1>}
 dyn_R_skirt_a_5 {parent=dyn_R_skirt_a_4, loc=<12.926 0 0>, rot=<0 -0.0091 0 -1>}
 dyn_R_skirt_a_6 {parent=dyn_R_skirt_a_5, loc=<11.3264 0 0>, rot=<0 0 0 1>}
 base_L_skirt_a {parent=dyn_Skirt_Master, loc=<-4.1801 8.2567 -10.632>, rot=<-0.037 0.7061 0.7066 0.0276>}
 dyn_L_skirt_a_1 {parent=base_L_skirt_a, loc=<0 0 0>, rot=<0 0 0 1>}
 dyn_L_skirt_a_2 {parent=dyn_L_skirt_a_1, loc=<25.4054 0 1.0687>, rot=<0 0.0137 0 -0.9999>}
 dyn_L_skirt_a_3 {parent=dyn_L_skirt_a_2, loc=<15.2446 0 0>, rot=<0 -0.003 0 -1>}
 dyn_L_skirt_a_4 {parent=dyn_L_skirt_a_3, loc=<12.19 0 0>, rot=<0 0.0021 0 -1>}
 dyn_L_skirt_a_5 {parent=dyn_L_skirt_a_4, loc=<12.926 0 0>, rot=<0 -0.0091 0 -1>}
 dyn_L_skirt_a_6 {parent=dyn_L_skirt_a_5, loc=<11.3264 0 0>, rot=<0 0 0 1>}
 base_C_skirt_a {parent=dyn_Skirt_Master, loc=<-4.2261 11.7108 0.0001>, rot=<-0.0047 0.7071 0.7071 -0.0047>}
 dyn_C_skirt_a_1 {parent=base_C_skirt_a, loc=<0 0 0>, rot=<0 0 0 1>}
 dyn_C_skirt_a_2 {parent=dyn_C_skirt_a_1, loc=<25.4054 0 1.0687>, rot=<0 0.0137 0 -0.9999>}
 dyn_C_skirt_a_3 {parent=dyn_C_skirt_a_2, loc=<15.2446 0 0>, rot=<0 -0.003 0 -1>}
 dyn_C_skirt_a_4 {parent=dyn_C_skirt_a_3, loc=<12.19 0 0>, rot=<0 0.0021 0 -1>}
 dyn_C_skirt_a_5 {parent=dyn_C_skirt_a_4, loc=<12.926 0 0>, rot=<0 -0.0091 0 -1>}
 dyn_C_skirt_a_6 {parent=dyn_C_skirt_a_5, loc=<11.3264 0 0>, rot=<0 0 0 1>}
 thigh_l {parent=pelvis, loc=<-1.0855 -0.2109 -8.7053>, rot=<0.0009 0.0355 0.0155 -0.9993>}
 calf_l {parent=thigh_l, loc=<-42.5284 -0.0019 -0.7297>, rot=<0.0006 -0.0161 -0.0391 -0.9991>}
 foot_l {parent=calf_l, loc=<-46.7401 -2.5014 -0.592>, rot=<0.0427 -0.0238 0.0353 -0.9982>}
 ball_l {parent=foot_l, loc=<-3.5658 10.8066 -0.1544>, rot=<0.0001 -0.0001 -0.7063 -0.7079>}
 Mover_foot_fwd_l {parent=foot_l, loc=<0 0 0>, rot=<0 0 0 1>}
 Mover_foot_bck_l {parent=foot_l, loc=<0 0 0>, rot=<0 0 0 1>}
 Mover_foot_side_l {parent=foot_l, loc=<0 0 0>, rot=<0 0 0 1>}
 Mover_foot_side_mid_front_l {parent=foot_l, loc=<0 0 0>, rot=<0 0 0 1>}
 Mover_foot_side_mid_back_l {parent=foot_l, loc=<0 0 0>, rot=<0 0 0 1>}
 calf_twist_02_l {parent=calf_l, loc=<-19.1546 -1.7943 -0.2638>, rot=<0.0237 0.0095 -0.0035 -0.9997>}
 calf_twist_01_l {parent=calf_l, loc=<-35.8483 -1.9197 -0.5794>, rot=<0.0237 0.0095 -0.0035 -0.9997>}
 Mover_calf_twist_01_fwd_l {parent=calf_twist_01_l, loc=<0 0 0>, rot=<0 0 0 1>}
 Mover_calf_twist_01_bck_l {parent=calf_twist_01_l, loc=<0 0 0>, rot=<0 0 0 1>}
 Mover_calf_twist_01_side_l {parent=calf_twist_01_l, loc=<0 0 0>, rot=<0 0 0 1>}
 Mover_calf_twist_01_side_mid_front_l {parent=calf_twist_01_l, loc=<0 0 0>, rot=<0 0 0 1>}
 Mover_calf_twist_01_side_mid_back_l {parent=calf_twist_01_l, loc=<0 0 0>, rot=<0 0 0 1>}
 deform_knee_in_base_l {parent=calf_l, loc=<0.8031 -5.0231 0.2456>, rot=<0 0 0 -1>}
 deform_knee_in_l {parent=deform_knee_in_base_l, loc=<0 0 0>, rot=<0 0 0 1>}
 deform_knee_out_base_l {parent=calf_l, loc=<0 4 0>, rot=<0 0 0 -1>}
 deform_knee_out_l {parent=deform_knee_out_base_l, loc=<0 0 0>, rot=<0 0 0 1>}
 Mover_calf_fwd_l {parent=calf_l, loc=<0 0 0>, rot=<0 0 0 1>}
 Mover_calf_bck_l {parent=calf_l, loc=<0 0 0>, rot=<0 0 0 1>}
 Mover_calf_side_l {parent=calf_l, loc=<0 0 0>, rot=<0 0 0 1>}
 Mover_calf_side_mid_front_l {parent=calf_l, loc=<0 0 0>, rot=<0 0 0 1>}
 Mover_calf_side_mid_back_l {parent=calf_l, loc=<0 0 0>, rot=<0 0 0 1>}
 thigh_twist_01_l {parent=thigh_l, loc=<-27.952 0 0>, rot=<0.0013 0 0 -1>}
 deform_groin_base_l {parent=thigh_l, loc=<0.4801 5.6776 -0.2486>, rot=<0 0 0 -1>}
 deform_groin_l {parent=deform_groin_base_l, loc=<0 0 0>, rot=<0 0 0 1>}
 Mover_thigh__fwd_l {parent=thigh_l, loc=<0 0 0>, rot=<0 0 0 1>}
 Mover_thigh_bck_l {parent=thigh_l, loc=<0 0 0>, rot=<0 0 0 1>}
 Mover_thigh_side_l {parent=thigh_l, loc=<0 0 0>, rot=<0 0 0 1>}
 Mover_thigh_mid_front_l {parent=thigh_l, loc=<0 0 0>, rot=<0 0 0 1>}
 Mover_thigh_mid_back_l {parent=thigh_l, loc=<0 0 0>, rot=<0 0 0 1>}
 thigh_r {parent=pelvis, loc=<-1.0855 -0.2109 8.7053>, rot=<0.0355 -0.0009 0.9993 0.0155>}
 calf_r {parent=thigh_r, loc=<42.5284 0.0019 0.7297>, rot=<0.0006 -0.0161 -0.0391 -0.9991>}
 foot_r {parent=calf_r, loc=<46.7401 2.5014 0.592>, rot=<0.0427 -0.0238 0.0353 -0.9982>}
 ball_r {parent=foot_r, loc=<3.5658 -10.8066 0.1544>, rot=<0.0001 -0.0001 -0.7063 -0.7079>}
 Mover_foot_fwd_r {parent=foot_r, loc=<0 0 0>, rot=<0 0 0 1>}
 Mover_foot_bck_r {parent=foot_r, loc=<0 0 0>, rot=<0 0 0 1>}
 Mover_foot_side_r {parent=foot_r, loc=<0 0 0>, rot=<0 0 0 1>}
 Mover_foot_side_mid_front_r {parent=foot_r, loc=<0 0 0>, rot=<0 0 0 1>}
 Mover_foot_side_mid_back_r {parent=foot_r, loc=<0 0 0>, rot=<0 0 0 1>}
 calf_twist_02_r {parent=calf_r, loc=<19.1546 1.7943 0.2638>, rot=<0.0237 0.0095 -0.0035 -0.9997>}
 calf_twist_01_r {parent=calf_r, loc=<35.8483 1.9197 0.5794>, rot=<0.0237 0.0095 -0.0035 -0.9997>}
 Mover_calf_twist_01_fwd_r {parent=calf_twist_01_r, loc=<0 0 0>, rot=<0 0 0 1>}
 Mover_calf_twist_01_bck_r {parent=calf_twist_01_r, loc=<0 0 0>, rot=<0 0 0 1>}
 Mover_calf_twist_01_side_r {parent=calf_twist_01_r, loc=<0 0 0>, rot=<0 0 0 1>}
 Mover_calf_twist_01_side_mid_front_r {parent=calf_twist_01_r, loc=<0 0 0>, rot=<0 0 0 1>}
 Mover_calf_twist_01_side_mid_backt_r {parent=calf_twist_01_r, loc=<0 0 0>, rot=<0 0 0 1>}
 deform_knee_in_base_r {parent=calf_r, loc=<-0.8031 5.0231 -0.2457>, rot=<0 0 0 -1>}
 deform_knee_in_r {parent=deform_knee_in_base_r, loc=<0 0 0>, rot=<0 0 0 1>}
 deform_knee_out_base_r {parent=calf_r, loc=<0 -4 0>, rot=<0 0 0 -1>}
 deform_knee_out_r {parent=deform_knee_out_base_r, loc=<0 0 0>, rot=<0 0 0 1>}
 Mover_calf_fwd_r {parent=calf_r, loc=<0 0 0>, rot=<0 0 0 1>}
 Mover_calf_bck_r {parent=calf_r, loc=<0 0 0>, rot=<0 0 0 1>}
 Mover_calf__side_r {parent=calf_r, loc=<0 0 0>, rot=<0 0 0 1>}
 Mover_calf__side_mid_front_r {parent=calf_r, loc=<0 0 0>, rot=<0 0 0 1>}
 Mover_calf__side_mid_back_r {parent=calf_r, loc=<0 0 0>, rot=<0 0 0 1>}
 thigh_twist_01_r {parent=thigh_r, loc=<27.952 0 0>, rot=<0.0013 0 0 -1>}
 deform_groin_base_r {parent=thigh_r, loc=<-0.4801 -5.6776 0.2486>, rot=<0 0 0 -1>}
 deform_groin_r {parent=deform_groin_base_r, loc=<0 0 0>, rot=<0 0 0 1>}
 Mover_thigh_fwd_r {parent=thigh_r, loc=<0 0 0>, rot=<0 0 0 1>}
 Mover_thigh_bck_r {parent=thigh_r, loc=<0 0 0>, rot=<0 0 0 1>}
 Mover_thigh_side_r {parent=thigh_r, loc=<0 0 0>, rot=<0 0 0 1>}
 Mover_thigh_side_mid_front_r {parent=thigh_r, loc=<0 0 0>, rot=<0 0 0 1>}
 Mover_thigh_side_mid_back_r {parent=thigh_r, loc=<0 0 0>, rot=<0 0 0 1>}
 deform_glute_base_l {parent=pelvis, loc=<-7.3466 -9.0213 -9.1245>, rot=<0.0172 0.031 -0.4477 -0.8935>}
 deform_glute_l {parent=deform_glute_base_l, loc=<0 0 0>, rot=<0 0 0 1>}
 deform_glute_base_r {parent=pelvis, loc=<-7.3464 -9.0213 9.1245>, rot=<0.031 -0.0172 0.8935 -0.4477>}
 deform_glute_r {parent=deform_glute_base_r, loc=<0 0 0>, rot=<0 0 0 1>}
 ground_joint {parent=root, loc=<-0.965 -0.5652 -8.2542>, rot=<0.0345 -0.0518 0.0533 -0.9966>}
]
variables [
 Clavicle Rot_L {type=FVector, default=(X=0,Y=0,Z=0)}
 Clavicle Rot_R {type=FVector, default=(X=0,Y=0,Z=0)}
 Head Back C {type=double, default=0}
 "Head Side L " {type=double, default=0}
 Head Side R {type=double, default=0}
 Head Twist L {type=double, default=0}
 Head Twist R {type=double, default=0}
 "Is Baseball Sliding " {type=bool, default=False}
 Is DBNO Carried {type=bool, default=False}
 "Is Knee Sliding " {type=bool, default=False}
 L calfBend Alpha {type=double, default=0}
 L thighBack Alpha {type=double, default=0}
 L thighForward Alpha {type=double, default=0}
 L thighSide Alpha {type=double, default=0}
 L thighVeryForward {type=double, default=0}
 Leg ROM_Result Alpha {type=double, default=0}
 Movement Alpha {type=double, default=0}
 Pub DBNO {type=bool, default=False}
 R calfBend Alpha {type=double, default=0}
 R thighBack Alpha {type=double, default=0}
 R thighForward Alpha {type=double, default=0}
 R thighSide Alpha {type=double, default=0}
 R thighVeryForward {type=double, default=0}
 "Right Leg " {type=double, default=0}
 "left Leg " {type=double, default=0}
 pelvisLowToGround_INVERSE {type=double, default=0}
 pelvisLowToGround_translateRotate {type=double, default=0}
]
