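# Physics asset for the F_MED_FruitFire_NudgeHonk female cosmetic body, built to simulate its long skirt.
# Eight six-bone capsule chains hang from base bones under the skirt master bones, with limited swing and zero twist; the "a" chains are stiffer on one swing axis.
# Pelvis, leg and foot capsules plus a large ground box provide collision shapes.
# Lower left and right "a" bones carry extra offset capsules, and their last two pairs are cross-linked by free-angle constraints.
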PhysicsAsset /BRCosmetics/Characters/Player/Female/Medium/Bodies/F_MED_FruitFire_NudgeHonk/Meshes/F_MED_FruitFire_NudgeHonk_Physics1
bodies {
 Skirt_Master [{shapeType=Capsule, length=0.505, radius=0.505, shapeName=Skirt_Master_capsule, rotation=<rot 0 -90 0>}]
 base_C_skirt_a [{shapeType=Capsule, length=0.505, radius=0.505, shapeName=base_C_skirt_a_capsule, rotation=<rot 0 -90 0>}]
 base_C_skirt_b [{shapeType=Capsule, length=0.505, radius=0.505, shapeName=base_C_skirt_b_capsule, rotation=<rot 0 -90 0>}]
 base_L_skirt_a [{shapeType=Capsule, length=0.505, radius=0.505, shapeName=base_L_skirt_a_capsule, rotation=<rot 0 -90 0>}]
 base_L_skirt_b [{shapeType=Capsule, length=0.505, radius=0.505, shapeName=base_L_skirt_b_capsule, rotation=<rot 0 0 90>}]
 base_L_skirt_c [{shapeType=Capsule, length=0.505, radius=0.505, shapeName=base_L_skirt_c_capsule, rotation=<rot 0 -90 0>}]
 base_R_skirt_a [{shapeType=Capsule, length=0.505, radius=0.505, shapeName=base_R_skirt_a_capsule, rotation=<rot 0 -90 0>}]
 base_R_skirt_b [{shapeType=Capsule, length=0.505, radius=0.505, shapeName=base_R_skirt_b_capsule, rotation=<rot 0 0 90>}]
 base_R_skirt_c [{shapeType=Capsule, length=0.505, radius=0.505, shapeName=base_R_skirt_c_capsule, rotation=<rot 0 0 90>}]
 calf_l [{shapeType=Capsule, length=44, radius=12, shapeName=calf_l_capsule, center=<-17.7888 -1.3645 1.5671>, rotation=<rot 0.0324 -89.5255 -90.3377>}]
 calf_r [{shapeType=Capsule, length=44, radius=12, shapeName=calf_r_capsule, center=<16.9045 0.51 -2.9251>, rotation=<rot 88.865 -0.0004 0.0002>}]
 dyn_C_skirt_a_1 [{shapeType=Capsule, length=24, radius=12, shapeName=dyn_C_skirt_a_1_capsule, center=<15 0 9>, rotation=<rot 0 0 90>}]
 dyn_C_skirt_a_2 [{shapeType=Capsule, length=24, radius=12, shapeName=dyn_C_skirt_a_2_capsule, center=<15 0 9>, rotation=<rot 0 0 90>}]
 dyn_C_skirt_a_3 [{shapeType=Capsule, length=24, radius=12, shapeName=dyn_C_skirt_a_3_capsule, center=<15 0 8>, rotation=<rot 0 0 90>}]
 dyn_C_skirt_a_4 [{shapeType=Capsule, length=24, radius=12, shapeName=dyn_C_skirt_a_4_capsule, center=<15 0 7>, rotation=<rot 0 0 90>}]
 dyn_C_skirt_a_5 [{shapeType=Capsule, length=24, radius=12, shapeName=dyn_C_skirt_a_5_capsule, center=<15 0 7>, rotation=<rot 0 0 90>}]
 dyn_C_skirt_a_6 [{shapeType=Capsule, length=24, radius=12, shapeName=dyn_C_skirt_a_6_capsule, center=<15 0 7>, rotation=<rot 0 0 90>}]
 dyn_C_skirt_b_1 [{shapeType=Capsule, length=24, radius=12, shapeName=dyn_C_skirt_b_1_capsule, center=<15 0 8>, rotation=<rot 0 0 90>}]
 dyn_C_skirt_b_2 [{shapeType=Capsule, length=24, radius=12, shapeName=dyn_C_skirt_b_2_capsule, center=<15 0 8>, rotation=<rot 0 0 90>}]
 dyn_C_skirt_b_3 [{shapeType=Capsule, length=24, radius=12, shapeName=dyn_C_skirt_b_3_capsule, center=<15 0 8>, rotation=<rot 0 0 90>}]
 dyn_C_skirt_b_4 [{shapeType=Capsule, length=24, radius=12, shapeName=dyn_C_skirt_b_4_capsule, center=<15 0 8>, rotation=<rot 0 0 90>}]
 dyn_C_skirt_b_5 [{shapeType=Capsule, length=24, radius=12, shapeName=dyn_C_skirt_b_5_capsule, center=<15 0 8>, rotation=<rot 0 0 90>}]
 dyn_C_skirt_b_6 [{shapeType=Capsule, length=24, radius=12, shapeName=dyn_C_skirt_b_6_capsule, center=<15 0 8>, rotation=<rot 0 0 90>}]
 dyn_L_skirt_a_1 [{shapeType=Capsule, length=24, radius=12, shapeName=dyn_L_skirt_a_1_capsule, center=<15 0 12>, rotation=<rot 0.0268 0 66.884>}]
 dyn_L_skirt_a_2 [
  {shapeType=Capsule, length=24, radius=12, shapeName=dyn_L_skirt_a_2_capsule, center=<15 0 12>, rotation=<rot 0.0268 -0.0001 66.884>}
  {shapeType=Capsule, length=24, radius=12, shapeName=dyn_R_skirt_a_2_capsule, center=<15 10 12>, rotation=<rot 0.0012 0 87.3334>}
  {shapeType=Capsule, length=24, radius=12, shapeName=dyn_R_skirt_a_2_capsule, center=<15 -20 -12>, rotation=<rot 0.0013 0 9.3654>}
 ]
 dyn_L_skirt_a_3 [
  {shapeType=Capsule, length=24, radius=12, shapeName=dyn_L_skirt_a_3_capsule, center=<15 0 12>, rotation=<rot 0.0268 -0.0001 66.884>}
  {shapeType=Capsule, length=24, radius=12, shapeName=dyn_R_skirt_a_2_capsule, center=<15 10 12>, rotation=<rot 0.0012 0 87.3334>}
  {shapeType=Capsule, length=24, radius=12, shapeName=dyn_R_skirt_a_2_capsule, center=<15 -20 -12>, rotation=<rot 0.0012 0 9.3654>}
 ]
 dyn_L_skirt_a_4 [
  {shapeType=Capsule, length=24, radius=12, shapeName=dyn_L_skirt_a_4_capsule, center=<15 0 12>, rotation=<rot 0.0268 -0.0001 66.884>}
  {shapeType=Capsule, length=24, radius=12, shapeName=dyn_R_skirt_a_2_capsule, center=<15 10 12>, rotation=<rot 0.0012 0 87.3334>}
  {shapeType=Capsule, length=24, radius=12, shapeName=dyn_R_skirt_a_2_capsule, center=<15 -20 -12>, rotation=<rot 0.0012 0 9.3654>}
 ]
 dyn_L_skirt_a_5 [
  {shapeType=Capsule, length=24, radius=12, shapeName=dyn_L_skirt_a_5_capsule, center=<15 0 12>, rotation=<rot 0.0268 0 66.884>}
  {shapeType=Capsule, length=24, radius=12, shapeName=dyn_R_skirt_a_2_capsule, center=<15 10 12>, rotation=<rot 0.2097 4.4817 87.3416>}
  {shapeType=Capsule, length=24, radius=12, shapeName=dyn_R_skirt_a_2_capsule, center=<15 -20 -12>, rotation=<rot 0.0012 0 9.3654>}
 ]
 dyn_L_skirt_a_6 [
  {shapeType=Capsule, length=24, radius=12, shapeName=dyn_L_skirt_a_6_capsule, center=<15 0 12>, rotation=<rot 0.0268 0 66.884>}
  {shapeType=Capsule, length=24, radius=12, shapeName=dyn_R_skirt_a_2_capsule, center=<15 10 12>, rotation=<rot 0.1688 3.6014 87.3387>}
  {shapeType=Capsule, length=24, radius=12, shapeName=dyn_R_skirt_a_2_capsule, center=<15 -20 -12>, rotation=<rot 0.0012 0 9.3654>}
 ]
 dyn_L_skirt_b_1 [{shapeType=Capsule, length=48, radius=12, shapeName=dyn_L_skirt_b_1_capsule, center=<15 0 12>, rotation=<rot 0 0 90>}]
 dyn_L_skirt_b_2 [{shapeType=Capsule, length=48, radius=12, shapeName=dyn_L_skirt_b_2_capsule, center=<15 0 12>, rotation=<rot 3.7874 0.0012 89.9999>}]
 dyn_L_skirt_b_3 [{shapeType=Capsule, length=48, radius=12, shapeName=dyn_L_skirt_b_3_capsule, center=<15 0 12>, rotation=<rot 0 0 90>}]
 dyn_L_skirt_b_4 [{shapeType=Capsule, length=48, radius=12, shapeName=dyn_L_skirt_b_4_capsule, center=<15 0 12>, rotation=<rot 0 0 90>}]
 dyn_L_skirt_b_5 [{shapeType=Capsule, length=48, radius=12, shapeName=dyn_L_skirt_b_5_capsule, center=<15 0 12>, rotation=<rot 0 0 90>}]
 dyn_L_skirt_b_6 [{shapeType=Capsule, length=48, radius=12, shapeName=dyn_L_skirt_b_6_capsule, center=<15 0 12>, rotation=<rot 0 0 90>}]
 dyn_L_skirt_c_1 [{shapeType=Capsule, length=24, radius=12, shapeName=dyn_L_skirt_c_1_capsule, center=<15 0 10>, rotation=<rot -0.0003 0 116.0961>}]
 dyn_L_skirt_c_2 [{shapeType=Capsule, length=24, radius=12, shapeName=dyn_L_skirt_c_2_capsule, center=<15 0 10>, rotation=<rot -0.0003 0 116.0961>}]
 dyn_L_skirt_c_3 [{shapeType=Capsule, length=24, radius=12, shapeName=dyn_L_skirt_c_3_capsule, center=<15 0 10>, rotation=<rot -0.0003 0 116.0961>}]
 dyn_L_skirt_c_4 [{shapeType=Capsule, length=24, radius=12, shapeName=dyn_L_skirt_c_4_capsule, center=<15 0 10>, rotation=<rot -0.0003 0 116.0961>}]
 dyn_L_skirt_c_5 [{shapeType=Capsule, length=24, radius=12, shapeName=dyn_L_skirt_c_5_capsule, center=<15 0 10>, rotation=<rot -0.0003 0 116.0961>}]
 dyn_L_skirt_c_6 [{shapeType=Capsule, length=24, radius=12, shapeName=dyn_L_skirt_c_6_capsule, center=<15 0 10>, rotation=<rot -0.0003 0 116.0961>}]
 dyn_R_skirt_a_1 [{shapeType=Capsule, length=24, radius=12, shapeName=dyn_R_skirt_a_1_capsule, center=<15 0 12>, rotation=<rot 0.0001 0 110>}]
 dyn_R_skirt_a_2 [
  {shapeType=Capsule, length=24, radius=12, shapeName=dyn_R_skirt_a_2_capsule, center=<15 0 12>, rotation=<rot 0.0005 0 121.7681>}
  {shapeType=Capsule, length=24, radius=12, shapeName=dyn_R_skirt_a_2_capsule, center=<15 -12 12>, rotation=<rot 0.0011 0 88.6128>}
  {shapeType=Capsule, length=24, radius=12, shapeName=dyn_R_skirt_a_2_capsule, center=<15 20 -12>, rotation=<rot 0.0012 0 150.6322>}
 ]
 dyn_R_skirt_a_3 [
  {shapeType=Capsule, length=24, radius=12, shapeName=dyn_R_skirt_a_3_capsule, center=<15 0 12>, rotation=<rot 0.0006 0 121.7681>}
  {shapeType=Capsule, length=24, radius=12, shapeName=dyn_R_skirt_a_2_capsule, center=<15 20 -12>, rotation=<rot 0.0012 0 150.6322>}
  {shapeType=Capsule, length=24, radius=12, shapeName=dyn_R_skirt_a_2_capsule, center=<15 -12 12>, rotation=<rot 0.0011 0 88.6128>}
 ]
 dyn_R_skirt_a_4 [
  {shapeType=Capsule, length=24, radius=12, shapeName=dyn_R_skirt_a_4_capsule, center=<15 0 12>, rotation=<rot 0.0006 0 121.7681>}
  {shapeType=Capsule, length=24, radius=12, shapeName=dyn_R_skirt_a_2_capsule, center=<15 20 -12>, rotation=<rot 0.0012 0 150.6322>}
  {shapeType=Capsule, length=24, radius=12, shapeName=dyn_R_skirt_a_2_capsule, center=<15 -12 12>, rotation=<rot 0.0011 0 88.6128>}
 ]
 dyn_R_skirt_a_5 [
  {shapeType=Capsule, length=24, radius=12, shapeName=dyn_R_skirt_a_5_capsule, center=<15 0 12>, rotation=<rot 0.0006 0 121.7681>}
  {shapeType=Capsule, length=24, radius=12, shapeName=dyn_R_skirt_a_2_capsule, center=<15 20 -12>, rotation=<rot 0.0012 0 150.6322>}
  {shapeType=Capsule, length=24, radius=12, shapeName=dyn_R_skirt_a_2_capsule, center=<15 -12 12>, rotation=<rot 0.0011 0 88.6128>}
 ]
 dyn_R_skirt_a_6 [
  {shapeType=Capsule, length=24, radius=12, shapeName=dyn_R_skirt_a_6_capsule, center=<15 0 12>, rotation=<rot 0.0004 0 117.3049>}
  {shapeType=Capsule, length=24, radius=12, shapeName=dyn_R_skirt_a_2_capsule, center=<15 20 -12>, rotation=<rot 0.0012 0 150.6322>}
  {shapeType=Capsule, length=24, radius=12, shapeName=dyn_R_skirt_a_2_capsule, center=<15 -12 12>, rotation=<rot 0.0011 0 88.6128>}
 ]
 dyn_R_skirt_b_1 [{shapeType=Capsule, length=48, radius=12, shapeName=dyn_R_skirt_b_1_capsule, center=<15 0 12>, rotation=<rot 0 0 90>}]
 dyn_R_skirt_b_2 [{shapeType=Capsule, length=48, radius=12, shapeName=dyn_R_skirt_b_2_capsule, center=<15 0 12>, rotation=<rot 0 0 90>}]
 dyn_R_skirt_b_3 [{shapeType=Capsule, length=48, radius=12, shapeName=dyn_R_skirt_b_3_capsule, center=<15 0 12>, rotation=<rot 0 0 90>}]
 dyn_R_skirt_b_4 [{shapeType=Capsule, length=48, radius=12, shapeName=dyn_R_skirt_b_4_capsule, center=<15 0 12>, rotation=<rot 0 0 90>}]
 dyn_R_skirt_b_5 [{shapeType=Capsule, length=48, radius=12, shapeName=dyn_R_skirt_b_5_capsule, center=<15 0 12>, rotation=<rot 0 0 90>}]
 dyn_R_skirt_b_6 [{shapeType=Capsule, length=48, radius=12, shapeName=dyn_R_skirt_b_6_capsule, center=<15 0 12>, rotation=<rot 0 0 90>}]
 dyn_R_skirt_c_1 [{shapeType=Capsule, length=24, radius=12, shapeName=dyn_R_skirt_c_1_capsule, center=<15 0 10>, rotation=<rot -0.0001 0 58.0499>}]
 dyn_R_skirt_c_2 [{shapeType=Capsule, length=24, radius=12, shapeName=dyn_R_skirt_c_2_capsule, center=<15 0 10>, rotation=<rot -0.0001 0 58.0499>}]
 dyn_R_skirt_c_3 [{shapeType=Capsule, length=24, radius=12, shapeName=dyn_R_skirt_c_3_capsule, center=<15 0 10>, rotation=<rot -0.0001 0 58.0499>}]
 dyn_R_skirt_c_4 [{shapeType=Capsule, length=24, radius=12, shapeName=dyn_R_skirt_c_4_capsule, center=<15 0 10>, rotation=<rot -0.0001 0 58.0499>}]
 dyn_R_skirt_c_5 [{shapeType=Capsule, length=24, radius=12, shapeName=dyn_R_skirt_c_5_capsule, center=<15 0 10>, rotation=<rot -0.0001 0 58.0499>}]
 dyn_R_skirt_c_6 [{shapeType=Capsule, length=24, radius=12, shapeName=dyn_R_skirt_c_6_capsule, center=<15 0 10>, rotation=<rot -0.0001 0 58.0499>}]
 dyn_Skirt_Master [{shapeType=Capsule, length=0.505, radius=0.505, shapeName=dyn_Skirt_Master_capsule, rotation=<rot 0 -90 0>}]
 foot_l [{
   shapeType=Capsule
   length=16.3453
   radius=8.7217
   shapeName=foot_l_capsule
   center=<-4.9985 7.0623 -0.5957>
   rotation=<rot 1.9921 25.3535 -93.8296>
  }]
 foot_r [{
   shapeType=Capsule
   length=16.4173
   radius=8.9707
   shapeName=foot_r_capsule
   center=<5.258 -8.0542 -0.2878>
   rotation=<rot 0.4704 25.773 -90.2549>
  }]
 ground_joint [{shapeType=Box, extentX=1009.5523, extentY=1009.5523, extentZ=59.5521, shapeName=root_box, center=<0 0 -20>}]
 pelvis [{shapeType=Capsule, length=13.7366, radius=12.7149, shapeName=pelvis_capsule, center=<-9.866 -1.4277 0.0376>}]
 spine_01 [{shapeType=Capsule, length=0.505, radius=0.505, shapeName=spine_01_capsule, rotation=<rot 0 -90 0>}]
 thigh_l [{shapeType=Capsule, length=37, radius=9, shapeName=thigh_l_capsule, center=<-24 0 1>, rotation=<rot 0.0003 -91.0046 -90.004>}]
 thigh_r [{shapeType=Capsule, length=37, radius=9, shapeName=thigh_r_capsule, center=<24 0 -1>, rotation=<rot 0.1525 -91.1054 -88.9543>}]
}
constraints [
 {bone1Name=dyn_L_skirt_a_1, bone2Name=base_L_skirt_a, swing1LimitDegrees=25, swing2LimitDegrees=25, twistLimitDegrees=0}
 {bone1Name=dyn_L_skirt_a_2, bone2Name=dyn_L_skirt_a_1, swing1LimitDegrees=5, swing2LimitDegrees=25, twistLimitDegrees=0}
 {bone1Name=dyn_L_skirt_a_3, bone2Name=dyn_L_skirt_a_2, swing1LimitDegrees=5, swing2LimitDegrees=25, twistLimitDegrees=0}
 {bone1Name=dyn_L_skirt_a_4, bone2Name=dyn_L_skirt_a_3, swing1LimitDegrees=5, swing2LimitDegrees=25, twistLimitDegrees=0}
 {bone1Name=dyn_L_skirt_a_5, bone2Name=dyn_L_skirt_a_4, swing1LimitDegrees=5, swing2LimitDegrees=25, twistLimitDegrees=0}
 {bone1Name=dyn_L_skirt_a_6, bone2Name=dyn_L_skirt_a_5, swing1LimitDegrees=5, swing2LimitDegrees=25, twistLimitDegrees=0}
 {
  bone1Name=thigh_l
  bone2Name=pelvis
  swing1LimitDegrees=45
  swing1Motion=Free
  swing2LimitDegrees=45
  swing2Motion=Free
  twistLimitDegrees=45
  twistMotion=Free
 }
 {
  bone1Name=calf_l
  bone2Name=thigh_l
  swing1LimitDegrees=45
  swing1Motion=Free
  swing2LimitDegrees=45
  swing2Motion=Free
  twistLimitDegrees=45
  twistMotion=Free
 }
 {
  bone1Name=foot_l
  bone2Name=calf_l
  swing1LimitDegrees=45
  swing1Motion=Free
  swing2LimitDegrees=45
  swing2Motion=Free
  twistLimitDegrees=45
  twistMotion=Free
 }
 {
  bone1Name=thigh_r
  bone2Name=pelvis
  swing1LimitDegrees=45
  swing1Motion=Free
  swing2LimitDegrees=45
  swing2Motion=Free
  twistLimitDegrees=45
  twistMotion=Free
 }
 {
  bone1Name=calf_r
  bone2Name=thigh_r
  swing1LimitDegrees=45
  swing1Motion=Free
  swing2LimitDegrees=45
  swing2Motion=Free
  twistLimitDegrees=45
  twistMotion=Free
 }
 {
  bone1Name=foot_r
  bone2Name=calf_r
  swing1LimitDegrees=45
  swing1Motion=Free
  swing2LimitDegrees=45
  swing2Motion=Free
  twistLimitDegrees=45
  twistMotion=Free
 }
 {bone1Name=dyn_R_skirt_a_1, bone2Name=base_R_skirt_a, swing1LimitDegrees=25, swing2LimitDegrees=25, twistLimitDegrees=0}
 {bone1Name=dyn_R_skirt_a_2, bone2Name=dyn_R_skirt_a_1, swing1LimitDegrees=5, swing2LimitDegrees=25, twistLimitDegrees=0}
 {bone1Name=dyn_R_skirt_a_3, bone2Name=dyn_R_skirt_a_2, swing1LimitDegrees=5, swing2LimitDegrees=25, twistLimitDegrees=0}
 {bone1Name=dyn_R_skirt_a_4, bone2Name=dyn_R_skirt_a_3, swing1LimitDegrees=5, swing2LimitDegrees=25, twistLimitDegrees=0}
 {bone1Name=dyn_R_skirt_a_5, bone2Name=dyn_R_skirt_a_4, swing1LimitDegrees=5, swing2LimitDegrees=25, twistLimitDegrees=0}
 {bone1Name=dyn_R_skirt_a_6, bone2Name=dyn_R_skirt_a_5, swing1LimitDegrees=5, swing2LimitDegrees=25, twistLimitDegrees=0}
 {bone1Name=dyn_L_skirt_c_1, bone2Name=base_L_skirt_c, swing1LimitDegrees=25, swing2LimitDegrees=25, twistLimitDegrees=0}
 {bone1Name=dyn_L_skirt_c_2, bone2Name=dyn_L_skirt_c_1, swing1LimitDegrees=25, swing2LimitDegrees=25, twistLimitDegrees=0}
 {bone1Name=dyn_L_skirt_c_3, bone2Name=dyn_L_skirt_c_2, swing1LimitDegrees=25, swing2LimitDegrees=25, twistLimitDegrees=0}
 {bone1Name=dyn_L_skirt_c_4, bone2Name=dyn_L_skirt_c_3, swing1LimitDegrees=25, swing2LimitDegrees=25, twistLimitDegrees=0}
 {bone1Name=dyn_L_skirt_c_5, bone2Name=dyn_L_skirt_c_4, swing1LimitDegrees=25, swing2LimitDegrees=25, twistLimitDegrees=0}
 {bone1Name=dyn_L_skirt_c_6, bone2Name=dyn_L_skirt_c_5, swing1LimitDegrees=25, swing2LimitDegrees=25, twistLimitDegrees=0}
 {bone1Name=dyn_R_skirt_c_1, bone2Name=base_R_skirt_c, swing1LimitDegrees=25, swing2LimitDegrees=25, twistLimitDegrees=0}
 {bone1Name=dyn_R_skirt_c_2, bone2Name=dyn_R_skirt_c_1, swing1LimitDegrees=25, swing2LimitDegrees=25, twistLimitDegrees=0}
 {bone1Name=dyn_R_skirt_c_3, bone2Name=dyn_R_skirt_c_2, swing1LimitDegrees=25, swing2LimitDegrees=25, twistLimitDegrees=0}
 {bone1Name=dyn_R_skirt_c_4, bone2Name=dyn_R_skirt_c_3, swing1LimitDegrees=25, swing2LimitDegrees=25, twistLimitDegrees=0}
 {bone1Name=dyn_R_skirt_c_5, bone2Name=dyn_R_skirt_c_4, swing1LimitDegrees=25, swing2LimitDegrees=25, twistLimitDegrees=0}
 {bone1Name=dyn_R_skirt_c_6, bone2Name=dyn_R_skirt_c_5, swing1LimitDegrees=25, swing2LimitDegrees=25, twistLimitDegrees=0}
 {bone1Name=dyn_L_skirt_b_1, bone2Name=base_L_skirt_b, swing1LimitDegrees=25, swing2LimitDegrees=25, twistLimitDegrees=0}
 {bone1Name=dyn_L_skirt_b_2, bone2Name=dyn_L_skirt_b_1, swing1LimitDegrees=25, swing2LimitDegrees=25, twistLimitDegrees=0}
 {bone1Name=dyn_L_skirt_b_3, bone2Name=dyn_L_skirt_b_2, swing1LimitDegrees=25, swing2LimitDegrees=25, twistLimitDegrees=0}
 {bone1Name=dyn_L_skirt_b_4, bone2Name=dyn_L_skirt_b_3, swing1LimitDegrees=25, swing2LimitDegrees=25, twistLimitDegrees=0}
 {bone1Name=dyn_L_skirt_b_5, bone2Name=dyn_L_skirt_b_4, swing1LimitDegrees=25, swing2LimitDegrees=25, twistLimitDegrees=0}
 {bone1Name=dyn_L_skirt_b_6, bone2Name=dyn_L_skirt_b_5, swing1LimitDegrees=25, swing2LimitDegrees=25, twistLimitDegrees=0}
 {bone1Name=dyn_R_skirt_b_6, bone2Name=dyn_R_skirt_b_5, swing1LimitDegrees=25, swing2LimitDegrees=25, twistLimitDegrees=0}
 {bone1Name=dyn_R_skirt_b_1, bone2Name=base_R_skirt_b, swing1LimitDegrees=25, swing2LimitDegrees=25, twistLimitDegrees=0}
 {bone1Name=dyn_R_skirt_b_2, bone2Name=dyn_R_skirt_b_1, swing1LimitDegrees=25, swing2LimitDegrees=25, twistLimitDegrees=0}
 {bone1Name=dyn_R_skirt_b_3, bone2Name=dyn_R_skirt_b_2, swing1LimitDegrees=25, swing2LimitDegrees=25, twistLimitDegrees=0}
 {bone1Name=dyn_R_skirt_b_4, bone2Name=dyn_R_skirt_b_3, swing1LimitDegrees=25, swing2LimitDegrees=25, twistLimitDegrees=0}
 {bone1Name=dyn_R_skirt_b_5, bone2Name=dyn_R_skirt_b_4, swing1LimitDegrees=25, swing2LimitDegrees=25, twistLimitDegrees=0}
 {
  bone1Name=Skirt_Master
  bone2Name=pelvis
  swing1LimitDegrees=45
  swing1Motion=Free
  swing2LimitDegrees=45
  swing2Motion=Free
  twistLimitDegrees=45
  twistMotion=Free
 }
 {
  bone1Name=dyn_R_skirt_a_5
  bone2Name=dyn_L_skirt_a_5
  swing1LimitDegrees=45
  swing1Motion=Free
  swing2LimitDegrees=45
  swing2Motion=Free
  twistLimitDegrees=45
  twistMotion=Free
 }
 {
  bone1Name=dyn_R_skirt_a_6
  bone2Name=dyn_L_skirt_a_6
  swing1LimitDegrees=45
  swing1Motion=Free
  swing2LimitDegrees=45
  swing2Motion=Free
  twistLimitDegrees=45
  twistMotion=Free
 }
 {
  bone1Name=dyn_L_skirt_a_6
  bone2Name=dyn_R_skirt_a_6
  swing1LimitDegrees=45
  swing1Motion=Free
  swing2LimitDegrees=45
  swing2Motion=Free
  twistLimitDegrees=45
  twistMotion=Free
 }
 {
  bone1Name=base_C_skirt_a
  bone2Name=Skirt_Master
  swing1LimitDegrees=45
  swing1Motion=Free
  swing2LimitDegrees=45
  swing2Motion=Free
  twistLimitDegrees=45
  twistMotion=Free
 }
 {bone1Name=dyn_C_skirt_a_1, bone2Name=base_C_skirt_a, swing1LimitDegrees=25, swing2LimitDegrees=25, twistLimitDegrees=0}
 {bone1Name=dyn_C_skirt_a_2, bone2Name=dyn_C_skirt_a_1, swing1LimitDegrees=5, swing2LimitDegrees=25, twistLimitDegrees=0}
 {bone1Name=dyn_C_skirt_a_3, bone2Name=dyn_C_skirt_a_2, swing1LimitDegrees=5, swing2LimitDegrees=25, twistLimitDegrees=0}
 {bone1Name=dyn_C_skirt_a_4, bone2Name=dyn_C_skirt_a_3, swing1LimitDegrees=5, swing2LimitDegrees=25, twistLimitDegrees=0}
 {bone1Name=dyn_C_skirt_a_5, bone2Name=dyn_C_skirt_a_4, swing1LimitDegrees=5, swing2LimitDegrees=25, twistLimitDegrees=0}
 {bone1Name=dyn_C_skirt_a_6, bone2Name=dyn_C_skirt_a_5, swing1LimitDegrees=5, swing2LimitDegrees=25, twistLimitDegrees=0}
 {
  bone1Name=base_C_skirt_b
  bone2Name=Skirt_Master
  swing1LimitDegrees=45
  swing1Motion=Free
  swing2LimitDegrees=45
  swing2Motion=Free
  twistLimitDegrees=45
  twistMotion=Free
 }
 {bone1Name=dyn_C_skirt_b_1, bone2Name=base_C_skirt_b, swing1LimitDegrees=25, swing2LimitDegrees=25, twistLimitDegrees=0}
 {bone1Name=dyn_C_skirt_b_2, bone2Name=dyn_C_skirt_b_1, swing1LimitDegrees=25, swing2LimitDegrees=25, twistLimitDegrees=0}
 {bone1Name=dyn_C_skirt_b_3, bone2Name=dyn_C_skirt_b_2, swing1LimitDegrees=25, swing2LimitDegrees=25, twistLimitDegrees=0}
 {bone1Name=dyn_C_skirt_b_4, bone2Name=dyn_C_skirt_b_3, swing1LimitDegrees=25, swing2LimitDegrees=25, twistLimitDegrees=0}
 {bone1Name=dyn_C_skirt_b_5, bone2Name=dyn_C_skirt_b_4, swing1LimitDegrees=25, swing2LimitDegrees=25, twistLimitDegrees=0}
 {bone1Name=dyn_C_skirt_b_6, bone2Name=dyn_C_skirt_b_5, swing1LimitDegrees=25, swing2LimitDegrees=25, twistLimitDegrees=0}
 {
  bone1Name=spine_01
  bone2Name=pelvis
  swing1LimitDegrees=45
  swing1Motion=Free
  swing2LimitDegrees=45
  swing2Motion=Free
  twistLimitDegrees=45
  twistMotion=Free
 }
 {
  bone1Name=Skirt_Master
  bone2Name=spine_01
  swing1LimitDegrees=45
  swing1Motion=Free
  swing2LimitDegrees=45
  swing2Motion=Free
  twistLimitDegrees=45
  twistMotion=Free
 }
 {
  bone1Name=dyn_Skirt_Master
  bone2Name=Skirt_Master
  swing1LimitDegrees=45
  swing1Motion=Free
  swing2LimitDegrees=45
  swing2Motion=Free
  twistLimitDegrees=45
  twistMotion=Free
 }
 {
  bone1Name=base_C_skirt_a
  bone2Name=dyn_Skirt_Master
  swing1LimitDegrees=45
  swing1Motion=Free
  swing2LimitDegrees=45
  swing2Motion=Free
  twistLimitDegrees=45
  twistMotion=Free
 }
 {
  bone1Name=base_C_skirt_b
  bone2Name=dyn_Skirt_Master
  swing1LimitDegrees=45
  swing1Motion=Free
  swing2LimitDegrees=45
  swing2Motion=Free
  twistLimitDegrees=45
  twistMotion=Free
 }
 {
  bone1Name=base_R_skirt_b
  bone2Name=dyn_Skirt_Master
  swing1LimitDegrees=45
  swing1Motion=Free
  swing2LimitDegrees=45
  swing2Motion=Free
  twistLimitDegrees=45
  twistMotion=Free
 }
 {
  bone1Name=base_R_skirt_c
  bone2Name=dyn_Skirt_Master
  swing1LimitDegrees=45
  swing1Motion=Free
  swing2LimitDegrees=45
  swing2Motion=Free
  twistLimitDegrees=45
  twistMotion=Free
 }
 {
  bone1Name=base_L_skirt_c
  bone2Name=dyn_Skirt_Master
  swing1LimitDegrees=45
  swing1Motion=Free
  swing2LimitDegrees=45
  swing2Motion=Free
  twistLimitDegrees=45
  twistMotion=Free
 }
 {
  bone1Name=base_L_skirt_b
  bone2Name=dyn_Skirt_Master
  swing1LimitDegrees=45
  swing1Motion=Free
  swing2LimitDegrees=45
  swing2Motion=Free
  twistLimitDegrees=45
  twistMotion=Free
 }
 {
  bone1Name=base_R_skirt_a
  bone2Name=dyn_Skirt_Master
  swing1LimitDegrees=45
  swing1Motion=Free
  swing2LimitDegrees=45
  swing2Motion=Free
  twistLimitDegrees=45
  twistMotion=Free
 }
 {
  bone1Name=base_L_skirt_a
  bone2Name=dyn_Skirt_Master
  swing1LimitDegrees=45
  swing1Motion=Free
  swing2LimitDegrees=45
  swing2Motion=Free
  twistLimitDegrees=45
  twistMotion=Free
 }
]
