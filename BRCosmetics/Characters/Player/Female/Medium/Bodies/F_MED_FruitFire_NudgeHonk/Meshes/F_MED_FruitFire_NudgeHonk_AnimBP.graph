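# Post-process style AnimBP for the FruitFire_NudgeHonk female body, adding procedural and physics motion, especially to the skirt.
# The copied parent pose runs through Control Rig, fixed skirt bone offsets and pelvis-space RigidBody, which is bypassed when DBNO-carried, then a DefaultSlot layer on Skirt_Master.
# Events update gravity, movement, weapon and slope variables.
# Crouched targeting sets an alternative skin weight profile on Primary, but ending targeting clears Secondary.
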AnimBlueprint /BRCosmetics/Characters/Player/Female/Medium/Bodies/F_MED_FruitFire_NudgeHonk/Meshes/F_MED_FruitFire_NudgeHonk_AnimBP
target_skeleton=&/F_MED_FruitFire_NudgeHonk_Skeleton
graphs [
 AnimGraph {nodes [
   n1 = Get GravityOverride()
   RigidBody!(ComponentPose, Alpha=0.5, OverrideWorldGravity:Vector=@n1) -> (Pose) {
    bClampLinearTranslationLimitToRefPose=true
    bOverrideWorldGravity=true
    bTransferBoneVelocities=true
    overridePhysicsAsset=&/F_MED_FruitFire_NudgeHonk_Physics1
    simulationSpace=BaseBoneSpace
    alphaBoolBlend {blendOption=Linear}
    baseBoneRef {boneName=dyn_Skirt_Master}
    componentAppliedLinearAccClamp=<3000 3000 3000>
    overrideWorldGravity {z=-980}
    simSpaceSettings {
     maxAngularAcceleration=40
     maxAngularVelocity=8
     maxLinearAcceleration=3500
     maxLinearVelocity=700
     velocityScaleZ=0.25
     worldAlpha=1
     externalLinearDragV=<3 3 3>
    }
   }
   n2 = Use cached pose 'Post DBNO check'()
   Output Pose(Result=Layered blend per bone(BasePose=Use cached pose 'Post DBNO check'(), BlendPoses_0=Slot 'DefaultSlot'(Source=@n2)) {layerSetup [{branchFilters [{boneName=Skirt_Master}]}]}) {layerGroup=None}
   n3 = Control Rig!(
    Source=Copy Pose From Mesh() {bUseAttachedParent=true}
    Alpha=0.87
    Pub DBNO:bool=Get bIsDBNO()
    "Is Knee Sliding ":bool=Get bIsKneeSliding()
    "Is Baseball Sliding ":bool=Get bIsBaseballSliding()
    Is DBNO Carried:bool=Get bIsDBNOCarried()
    Movement Alpha:double=Get MovementAlpha() -> (MovementAlpha:double)
   ) -> (Pose) {
    alphaBoolBlend {blendOption=Linear}
    controlRigAssetReference {blueprintRigClass=&/F_MED_FruitFire_NudgeHonk_CtrlRig.F_MED_FruitFire_NudgeHonk_CtrlRig_C}
   }
   Save cached pose 'Base ControlRig'(Pose=@n3) {cacheName=Base ControlRig}
   n4 = RigidBody(ComponentPose=Local To Component(LocalPose=Use cached pose 'Post Slope Posing'()), Alpha=0.55, OverrideWorldGravity=@n1) {
    bClampLinearTranslationLimitToRefPose=true
    bOverrideWorldGravity=true
    overridePhysicsAsset=&F_MED_FruitFire_NudgeHonk_Physics1
    simulationSpace=BaseBoneSpace
    alphaBoolBlend {blendOption=Linear}
    baseBoneRef {boneName=pelvis}
    componentAppliedLinearAccClamp=<3000 3000 3000>
    componentLinearAccScale=<0.4 0.3 0.3>
    componentLinearVelScale=<0.4 0.3 0.3>
    overrideWorldGravity {z=-980}
    simSpaceSettings {
     maxAngularAcceleration=1000
     maxAngularVelocity=3
     maxLinearAcceleration=700
     maxLinearVelocity=500
     worldAlpha=1
     externalLinearDragV=<1 1 1>
    }
   }
   Save cached pose 'PostDyn'(Pose=Component To Local(ComponentPose=@n4)) {cacheName=PostDyn}
   Save cached pose 'Post DBNO check'(Pose=Blend Poses by bool(BlendPose_0=Use cached pose 'Base ControlRig'(), BlendPose_1=Use cached pose 'PostDyn'(), bActiveValue=Get bIsDBNOCarried())) {cacheName=Post DBNO check}
   n5 = "Transform (Modify) Bone - Bone: ground_joint"!(
    ComponentPose=Local To Component(LocalPose=Use cached pose 'Base ControlRig'())
    bAlphaBoolEnabled:bool=True
    Translation="0,0,0"
    Rotation="0,0,0"
    Scale="1,1,1"
   ) -> (Pose) {
    alphaInputType=Bool
    rotationMode=BMM_Replace
    rotationSpace=BCS_WorldSpace
    translationMode=BMM_Replace
    translationSpace=BCS_BoneSpace
    alphaBoolBlend {blendOption=Linear}
    boneToModify {boneName=ground_joint}
   }
   n6 = "Transform (Modify) Bone - Bone: ground_joint"(ComponentPose=@n5, bAlphaBoolEnabled=True, Translation="0,0,25", Rotation="0,0,0", Scale="1,1,1") {
    alphaInputType=Bool
    rotationSpace=BCS_WorldSpace
    translationMode=BMM_Additive
    translationSpace=BCS_BoneSpace
    alphaBoolBlend {blendInTime=0.1, blendOption=Linear, blendOutTime=0.1}
    boneToModify {boneName=ground_joint}
    translation {z=25}
   }
   n7 = "Transform (Modify) Bone - Bone: dyn_R_skirt_a_1"(ComponentPose=@n6, Alpha=1, Translation="0,0,0", Rotation="5,0,0", Scale="1,1,1") {
    rotationMode=BMM_Additive
    rotationSpace=BCS_BoneSpace
    scaleMode=BMM_Replace
    scaleSpace=BCS_BoneSpace
    translationMode=BMM_Additive
    translationSpace=BCS_BoneSpace
    alphaBoolBlend {blendOption=Linear}
    alphaScaleBiasClamp {interpSpeedDecreasing=4}
    boneToModify {boneName=dyn_R_skirt_a_1}
    rotation {pitch=5}
   }
   n8 = "Transform (Modify) Bone - Bone: dyn_L_skirt_a_1"(ComponentPose=@n7, Alpha=1, Translation="0,0,0", Rotation="5,0,0", Scale="1,1,1") {
    rotationMode=BMM_Additive
    rotationSpace=BCS_BoneSpace
    scaleMode=BMM_Replace
    scaleSpace=BCS_BoneSpace
    translationMode=BMM_Additive
    translationSpace=BCS_BoneSpace
    alphaBoolBlend {blendOption=Linear}
    alphaScaleBiasClamp {interpSpeedDecreasing=4}
    boneToModify {boneName=dyn_L_skirt_a_1}
    rotation {pitch=5}
   }
   n9 = "Transform (Modify) Bone - Bone: dyn_C_skirt_a_1"(ComponentPose=@n8, Alpha=1, Translation="0,0,0", Rotation="5,0,0", Scale="1,1,1") {
    rotationMode=BMM_Additive
    rotationSpace=BCS_BoneSpace
    scaleMode=BMM_Replace
    scaleSpace=BCS_BoneSpace
    translationMode=BMM_Additive
    translationSpace=BCS_BoneSpace
    alphaBoolBlend {blendOption=Linear}
    alphaScaleBiasClamp {interpSpeedDecreasing=4}
    boneToModify {boneName=dyn_C_skirt_a_1}
    rotation {pitch=5}
   }
   n10 = "Transform (Modify) Bone - Bone: Skirt_Master"(ComponentPose=@n9, Alpha=1, Translation="10,0,0", Rotation="0,10,0", Scale="1,1,1") {
    rotationMode=BMM_Additive
    rotationSpace=BCS_BoneSpace
    scaleMode=BMM_Replace
    scaleSpace=BCS_BoneSpace
    translationMode=BMM_Additive
    translationSpace=BCS_BoneSpace
    alphaBoolBlend {blendOption=Linear}
    alphaScaleBiasClamp {interpSpeedDecreasing=4}
    boneToModify {boneName=Skirt_Master}
    rotation {yaw=10}
    translation {x=10}
   }
   n11 = "Transform (Modify) Bone - Bone: Skirt_Master"(ComponentPose=@n10, Alpha=1, Translation="10,0,0", Rotation="-5,10,0", Scale="1,1,1") {
    rotationMode=BMM_Additive
    rotationSpace=BCS_BoneSpace
    scaleMode=BMM_Replace
    scaleSpace=BCS_BoneSpace
    translationMode=BMM_Additive
    translationSpace=BCS_BoneSpace
    alphaBoolBlend {blendOption=Linear}
    alphaScaleBiasClamp {interpSpeedDecreasing=4}
    boneToModify {boneName=Skirt_Master}
    rotation {pitch=-5, yaw=10}
    translation {x=10}
   }
   n12 = "Transform (Modify) Bone - Bone: dyn_R_skirt_a_1"(ComponentPose=@n11, Alpha=1, Translation="0,0,0", Rotation="10,0,0", Scale="1,1,1") {
    rotationMode=BMM_Additive
    rotationSpace=BCS_BoneSpace
    scaleMode=BMM_Replace
    scaleSpace=BCS_BoneSpace
    translationMode=BMM_Additive
    translationSpace=BCS_BoneSpace
    alphaBoolBlend {blendOption=Linear}
    alphaScaleBiasClamp {interpSpeedDecreasing=4}
    boneToModify {boneName=dyn_R_skirt_a_1}
    rotation {pitch=10}
   }
   n13 = "Transform (Modify) Bone - Bone: dyn_L_skirt_a_1"(ComponentPose=@n12, Alpha=1, Translation="0,0,3", Rotation="15,0,0", Scale="1,1,1") {
    rotationMode=BMM_Additive
    rotationSpace=BCS_BoneSpace
    scaleMode=BMM_Replace
    scaleSpace=BCS_BoneSpace
    translationMode=BMM_Additive
    translationSpace=BCS_BoneSpace
    alphaBoolBlend {blendOption=Linear}
    alphaScaleBiasClamp {interpSpeedDecreasing=4}
    boneToModify {boneName=dyn_L_skirt_a_1}
    rotation {pitch=15}
    translation {z=3}
   }
   n14 = "Transform (Modify) Bone - Bone: dyn_C_skirt_a_1"(ComponentPose=@n13, Alpha=1, Translation="0,0,0", Rotation="10,0,0", Scale="1,1,1") {
    rotationMode=BMM_Additive
    rotationSpace=BCS_BoneSpace
    scaleMode=BMM_Replace
    scaleSpace=BCS_BoneSpace
    translationMode=BMM_Additive
    translationSpace=BCS_BoneSpace
    alphaBoolBlend {blendOption=Linear}
    alphaScaleBiasClamp {interpSpeedDecreasing=4}
    boneToModify {boneName=dyn_C_skirt_a_1}
    rotation {pitch=10}
   }
   Save cached pose 'Post Slope Posing'(Pose=Component To Local(ComponentPose=@n14)) {cacheName=Post Slope Posing}
  ]}
 EventGraph {nodes [
   n1 = Event BlueprintUpdateAnimation()
   > n2 = Sequence!(execute) -> (then_0, then_1, then_2:Exec, then_3:Exec, then_4:Exec)
   n3 = PlayerMovement(execute=@n2.then_1, AnimInstance=Self-Reference(), XDotProductCoefficient=25, YDotProductCoefficient=25, ZDotProductCoefficient=25)
   > Set MovementAlpha(MovementAlpha:double="Absolute (Float)"(A=float * float(A=@n3.YMovement, B=0.045) {is=title K2Node_PromotableOperator})) -> (Output_Get:double)
   Set Is Knee Sliding Alpha(execute=@n2.then_2, Is Knee Sliding Alpha:double="To Float (Boolean)"(InBool=Get bIsKneeSliding())) -> (Output_Get:double)
   > Set Is Baseball Sliding(Is Baseball Sliding:double="To Float (Boolean)"(InBool=Get bIsBaseballSliding())) -> (Output_Get:double)
   Handle Slope Warp!(execute=@n2.then_3, self, DeltaTime:double=@n1.DeltaTimeX) -> (then)
   n4 = Cast To FortPlayerAnimInstance(execute=@n2.then_4, Object=GetAnimInstance(self=GetOwnerSkeletalMeshComponent()))
   > Set Is Aiming Pistol(Is Aiming Pistol:double=0) -> (Output_Get:double)
   > Set Is Pickaxe(Is Pickaxe:double=0) -> (Output_Get:double)
   > Set Is Aiming Rifle(Is Aiming Rifle:double=0) -> (Output_Get:double)
   > Set Is Pickaxe Boo(Is Pickaxe Boo:bool=false) -> (Output_Get:bool)
   > n5 = Switch on EFortWeaponCoreAnimation!(execute, Selection:EFortWeaponCoreAnimation Enum=Get WeaponCoreAnim(self=@n4)) -> (
    Melee:Exec
    Pistol:Exec
    Shotgun:Exec
    PaperBlueprint:Exec
    Rifle:Exec
    MeleeOneHand:Exec
    MachinePistol:Exec
    RocketLauncher:Exec
    GrenadeLauncher:Exec
    GoingCommando:Exec
    AssaultRifle:Exec
    TacticalShotgun:Exec
    SniperRifle:Exec
    TrapPlacement:Exec
    ShoulderLauncher:Exec
    AbilityDecoTool:Exec
    Crossbow:Exec
    C4:Exec
    RemoteControl:Exec
    DualWield:Exec
    AR_BullPup:Exec
    AR_ForwardGrip:Exec
    MedPackPaddles:Exec
    SMG_P90:Exec
    AR_DrumGun:Exec
    Consumable_Small:Exec
    Consumable_Large:Exec
    Balloon:Exec
    MountedTurret:Exec
    CreativeTool:Exec
    ExplosiveBow:Exec
    AshtonIndigo:Exec
    AshtonChicago:Exec
    MeleeDualWield:Exec
    Unarmed:Exec
   )
   Branch(execute=[@n5.AssaultRifle, @n5.Crossbow, @n5.GrenadeLauncher, @n5.Rifle, @n5.RocketLauncher, @n5.Shotgun, @n5.SniperRifle, @n5.TacticalShotgun])
   > Set Is Aiming Rifle(Is Aiming Rifle=1)
   Branch(execute=[@n5.MachinePistol, @n5.Pistol])
   > Set Is Aiming Pistol(Is Aiming Pistol=1)
   Branch(execute=@n5.Melee)
   > Set Is Pickaxe(Is Pickaxe=1)
   > Set Is Pickaxe Boo(Is Pickaxe Boo=true)
   n6 = Gravity Override State Function DEPRECATED(
    execute=@n2.then_0
    CharacterPartAnimInstance=Self-Reference()
    GameTimeInSeconds=GetGameTimeInSeconds()
    GravityOverrideParams=Get Gravity Override Params()
   )
   > Set GravityOverride(GravityOverride=@n6)
   > Set Sim Space Settings(Sim Space Settings:Sim Space Settings Structure=Break RBANSetting Struct!(RBANSettingStruct:"RBANSetting Struct Structure (by ref)"=@n6.RBAN SettingsOut) -> (
     ComponentLinearAccScale_28_#1:Vector=~
     ComponentLinearVelScale_31_#2:Vector=~
     SimSpaceSettings_34_#3:Sim Space Settings Structure
    ) {is=title K2Node_BreakStruct}.SimSpaceSettings_34_#3) -> (Output_Get:Sim Space Settings Structure)
   Event OnCharacterCustomizationLoaded()
   > Set Body Skel Part(Body Skel Part:Skeletal Mesh Component Object Reference=GetPartSkeletalMeshComponent()) -> (Output_Get:Skeletal Mesh Component Object Reference)
   Event OnBeginTargeting() {is=title K2Node_Event}
   > Branch(Condition=Get bIsCrouching())
   > SetSkinWeightProfile!(
    execute
    self:Skinned Mesh Component Object Reference=Get Body Skel Part() -> (Body Skel Part:Skeletal Mesh Component Object Reference)
    InProfileName:Name=Alternative
    InLayer:ESkinWeightProfileLayer Enum=Primary
   ) -> (then, ReturnValue:bool)
   Event OnEndTargeting() {is=title K2Node_Event}
   > ClearSkinWeightProfile!(execute, self:Skinned Mesh Component Object Reference=Get Body Skel Part(), InLayer:ESkinWeightProfileLayer Enum=Secondary) -> (then)
  ]}
 Handle Slope Warp {nodes [
   n1 = Handle Slope Warp!() -> (then, DeltaTime:double=~) {is=title K2Node_FunctionEntry}
   > n2 = Sequence!(execute) -> (then_0, then_1)
   n3 = Get bIsOnGround() -> (bIsOnGround:bool)
   > n4 = Branch(Condition=OR Boolean!(A=NOT Boolean(A=@n3), B=Get bIsPlayingEmote(), C:bool=Get bIsInVehicle() -> (bIsInVehicle:bool)) -> (ReturnValue))
   > Set Target Info Input Root Slope Translation(Target Info Input Root Slope Translation:Vector="0,0,-100") -> (Output_Get:Vector)
   Set Slope Info Input Slope Rotation(
    execute=@n4.else
    Slope Info Input Slope Rotation:Rotator=RInterpTo!(
     Current:Rotator=Get Slope Info Input Slope Rotation() -> (Slope Info Input Slope Rotation:Rotator)
     Target:Rotator=Get!() -> (
      SlopeInfoInput_RootSlopeTranslation:Vector=~
      SlopeInfoInput_LeftFootSlopeTranslation:Vector=~
      SlopeInfoInput_RightFootSlopeTranslation:Vector=~
      SlopeInfoInput_SlopeRotation:Rotator=~
     ).SlopeInfoInput_SlopeRotation
     DeltaTime:float=@n1
     InterpSpeed:float=7
    ) -> (ReturnValue:Rotator)
   ) -> (Output_Get:Rotator)
   > Set Target Info Input Root Slope Translation(Target Info Input Root Slope Translation="0,0,5")
   Set Slope Info Input Root Slope Translation(
    execute=@n2.then_1
    Slope Info Input Root Slope Translation:Vector=VInterpTo!(
     Current:Vector=Get Slope Info Input Root Slope Translation() -> (Slope Info Input Root Slope Translation:Vector)
     Target:Vector=Get Target Info Input Root Slope Translation() -> (Target Info Input Root Slope Translation:Vector)
     DeltaTime:float=@n1
     InterpSpeed:float=10
    ) -> (ReturnValue:Vector)
   ) -> (Output_Get:Vector)
  ]}
]
variables [
 Body Skel Part {type=SkeletalMeshComponent}
 Gravity Override Params {
  type=GravityOverrideParamsStruct
  default {
   onGround_Standing {windAmplitude=<-5 -70 30>, windFrequency=<-3 -1 -1>, windOffset=<-5 -50 30>}
   skydive_DivingDown {gravityOverride=<-2000 300 980>, windAmplitude=<0 0 0>, windFrequency=<0 0 0>}
   skydive_FloatingDown {gravityOverride=<0 -980 -980>}
   skydive_GoingUp {gravityOverride=<0 0 -980>}
  }
 }
 GravityOverride {type=Vector}
 Is Aiming Pistol {type=number}
 Is Aiming Rifle {type=number}
 Is Baseball Sliding {type=number}
 Is Knee Sliding Alpha {type=number}
 Is Pickaxe {type=number}
 Is Pickaxe Boo {type=boolean}
 MovementAlpha {type=number}
 Sim Space Settings {
  type=SimSpaceSettings
  default {
   dampingAlpha=1
   maxAngularAcceleration=10000
   maxAngularVelocity=10000
   maxLinearAcceleration=10000
   maxLinearVelocity=10000
   velocityScaleZ=1
   worldAlpha=0
   externalLinearDragV=<0 0 0>
  }
 }
 Slope Info Input Root Slope Translation {type=Vector}
 Slope Info Input Slope Rotation {type=Rotator}
 Target Info Input Root Slope Translation {type=Vector}
]
