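# Physics asset of cape colliders for the F_MED_FruitFire_NudgeHonk female medium player body cosmetic.
# Three six-bone dyn skirt chains (center back, left, right) carry capsules and link bone to bone, each rooted to spine_01 by constraints with free swing and twist.
# Body capsules cover head, clavicles, spine, pelvis and feet, with doubled capsules on middle spine bones.
# Arm, thigh, calf and neck bodies have no shapes.
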PhysicsAsset /BRCosmetics/Characters/Player/Female/Medium/Bodies/F_MED_FruitFire_NudgeHonk/Meshes/F_MED_FruitFire_NudgeHonk_CapeColliders
bodies {
 calf_l []
 calf_r []
 clavicle_l [{shapeType=Capsule, length=6.5623, radius=7.5194, center=<3.9566 -1.0862 -2.1127>, rotation=<rot -77.5566 177.3915 7.7314>}]
 clavicle_r [{shapeType=Capsule, length=6.5623, radius=7.5194, center=<-3.9566 1.0862 2.1127>, rotation=<rot -77.5566 177.3915 -172.2686>}]
 dyn_C_skirt_b_1 [{shapeType=Capsule, length=4.5334, radius=2.3966, shapeName=dyn_C_skirt_b_1_capsule, center=<4 0 0>, rotation=<rot 90 0 0>}]
 dyn_C_skirt_b_2 [{shapeType=Capsule, length=6.2656, radius=10.8477, shapeName=dyn_C_skirt_b_2_capsule, center=<6.2344 0 -3.9035>, rotation=<rot 90 0 0>}]
 dyn_C_skirt_b_3 [{shapeType=Capsule, length=5.1039, radius=10.9685, shapeName=dyn_C_skirt_b_3_capsule, center=<6.2344 0 -3.9035>, rotation=<rot 90 0 0>}]
 dyn_C_skirt_b_4 [{shapeType=Capsule, length=6.597, radius=12.6047, shapeName=dyn_C_skirt_b_4_capsule, center=<6.2344 0 -3.9035>, rotation=<rot 90 0 0>}]
 dyn_C_skirt_b_5 [{shapeType=Capsule, length=5.8087, radius=9.2101, shapeName=dyn_C_skirt_b_5_capsule, center=<6.2344 0 -3.9035>, rotation=<rot 90 0 0>}]
 dyn_C_skirt_b_6 [{shapeType=Capsule, length=8.5842, radius=12.4628, shapeName=dyn_C_skirt_b_6_capsule, center=<6.2344 0 -3.9035>, rotation=<rot 90 0 0>}]
 dyn_L_skirt_c_1 [{shapeType=Capsule, length=6.3754, radius=5.3491, shapeName=dyn_L_skirt_c_1_capsule, center=<4 0 0>, rotation=<rot 90 0 0>}]
 dyn_L_skirt_c_2 [{shapeType=Capsule, length=1.9994, radius=9.2509, shapeName=dyn_L_skirt_c_2_capsule, center=<4 0 -4.2664>, rotation=<rot 90 0 0>}]
 dyn_L_skirt_c_3 [{shapeType=Capsule, length=1.6526, radius=9.6289, shapeName=dyn_L_skirt_c_3_capsule, center=<4 0 -4.2664>, rotation=<rot 90 0 0>}]
 dyn_L_skirt_c_4 [{shapeType=Capsule, length=1.9889, radius=10.5862, shapeName=dyn_L_skirt_c_4_capsule, center=<4 0 -4.2664>, rotation=<rot 90 0 0>}]
 dyn_L_skirt_c_5 [{shapeType=Capsule, length=2.2481, radius=10.4647, shapeName=dyn_L_skirt_c_5_capsule, center=<4 0 -4.2664>, rotation=<rot 90 0 0>}]
 dyn_L_skirt_c_6 [{shapeType=Capsule, length=6.258, radius=13.2285, shapeName=dyn_L_skirt_c_6_capsule, center=<4 0 -4.2664>, rotation=<rot 90 0 0>}]
 dyn_R_skirt_c_1 [{shapeType=Capsule, length=8.8316, radius=6.3156, shapeName=dyn_R_skirt_c_1_capsule, center=<4 0 0>, rotation=<rot 90 0 0>}]
 dyn_R_skirt_c_2 [{shapeType=Capsule, length=1.5612, radius=13.0802, shapeName=dyn_R_skirt_c_2_capsule, center=<4 0 -2.5304>, rotation=<rot 90 0 0>}]
 dyn_R_skirt_c_3 [{shapeType=Capsule, length=0.999, radius=10.8239, shapeName=dyn_R_skirt_c_3_capsule, center=<4 0 -2.5304>, rotation=<rot 90 0 0>}]
 dyn_R_skirt_c_4 [{shapeType=Capsule, length=0.0507, radius=11.056, shapeName=dyn_R_skirt_c_4_capsule, center=<4 0 -2.5304>, rotation=<rot 90 0 0>}]
 dyn_R_skirt_c_5 [{shapeType=Capsule, length=2.2332, radius=12.1439, shapeName=dyn_R_skirt_c_5_capsule, center=<4 0 -2.5304>, rotation=<rot 90 0 0>}]
 dyn_R_skirt_c_6 [{shapeType=Capsule, length=4.9569, radius=14.7512, shapeName=dyn_R_skirt_c_6_capsule, center=<4 0 -2.5304>, rotation=<rot 90 0 0>}]
 foot_l [{shapeType=Capsule, length=11.0765, radius=8.0344, center=<-4.1811 6.0027 -0.338>, rotation=<rot 0.0388 1.4701 -89.3982>}]
 foot_r [{shapeType=Capsule, length=11.0765, radius=8.0344, center=<4.1811 -6.0027 0.338>, rotation=<rot 0.0388 1.4701 90.6018>}]
 head [{shapeType=Capsule, length=0, radius=18, center=<9.0697 0.0746 0>, rotation=<rot 0 -90 0>}]
 lowerarm_l []
 lowerarm_r []
 neck_02 []
 pelvis [{shapeType=Capsule, length=9.6042, radius=11.1441, center=<-2.4471 -2.5289 -0.1981>}]
 spine_01 [{shapeType=Capsule, length=4.3888, radius=14.3638, center=<-8.9021 -0.8973 0>}]
 spine_02 [
  {shapeType=Capsule, length=2.5219, radius=9.7317, center=<-2.6885 -1.2641 0>}
  {shapeType=Capsule, length=20.6875, radius=9.3125, center=<-1.8317 1.9362 0>, rotation=<rot 90 0 0>}
 ]
 spine_03 [
  {shapeType=Capsule, length=6.1863, radius=8.4854, center=<0.0574 0.4416 0>}
  {shapeType=Capsule, length=20.7188, radius=9.2812, center=<-1.2163 0.549 0>, rotation=<rot 90 180 -175.0732>}
 ]
 spine_04 [
  {shapeType=Capsule, length=9.0076, radius=10.3916, center=<5.6234 -0.6779 0>}
  {shapeType=Capsule, length=10.0312, radius=9.2812, center=<-1.2326 1.308 0>, rotation=<rot 90 0 0>}
 ]
 spine_05 [{shapeType=Capsule, length=9.678, radius=12.3571, center=<0.6969 1.91 0>}]
 thigh_l []
 thigh_r []
 upperarm_l []
 upperarm_r []
}
constraints [
 {
  bone1Name=dyn_L_skirt_c_4
  bone2Name=dyn_L_skirt_c_3
  swing1LimitDegrees=45
  swing1Motion=Free
  swing2LimitDegrees=45
  swing2Motion=Free
  twistLimitDegrees=45
  twistMotion=Free
 }
 {
  bone1Name=dyn_L_skirt_c_6
  bone2Name=dyn_L_skirt_c_5
  swing1LimitDegrees=45
  swing1Motion=Free
  swing2LimitDegrees=45
  swing2Motion=Free
  twistLimitDegrees=45
  twistMotion=Free
 }
 {
  bone1Name=dyn_L_skirt_c_5
  bone2Name=dyn_L_skirt_c_4
  swing1LimitDegrees=45
  swing1Motion=Free
  swing2LimitDegrees=45
  swing2Motion=Free
  twistLimitDegrees=45
  twistMotion=Free
 }
 {
  bone1Name=dyn_L_skirt_c_3
  bone2Name=dyn_L_skirt_c_2
  swing1LimitDegrees=45
  swing1Motion=Free
  swing2LimitDegrees=45
  swing2Motion=Free
  twistLimitDegrees=45
  twistMotion=Free
 }
 {
  bone1Name=dyn_L_skirt_c_2
  bone2Name=dyn_L_skirt_c_1
  swing1LimitDegrees=45
  swing1Motion=Free
  swing2LimitDegrees=45
  swing2Motion=Free
  twistLimitDegrees=45
  twistMotion=Free
 }
 {
  bone1Name=dyn_L_skirt_c_1
  bone2Name=spine_01
  swing1LimitDegrees=45
  swing1Motion=Free
  swing2LimitDegrees=45
  swing2Motion=Free
  twistLimitDegrees=45
  twistMotion=Free
 }
 {
  bone1Name=dyn_C_skirt_b_6
  bone2Name=dyn_C_skirt_b_5
  swing1LimitDegrees=45
  swing1Motion=Free
  swing2LimitDegrees=45
  swing2Motion=Free
  twistLimitDegrees=45
  twistMotion=Free
 }
 {
  bone1Name=dyn_C_skirt_b_5
  bone2Name=dyn_C_skirt_b_4
  swing1LimitDegrees=45
  swing1Motion=Free
  swing2LimitDegrees=45
  swing2Motion=Free
  twistLimitDegrees=45
  twistMotion=Free
 }
 {
  bone1Name=dyn_C_skirt_b_4
  bone2Name=dyn_C_skirt_b_3
  swing1LimitDegrees=45
  swing1Motion=Free
  swing2LimitDegrees=45
  swing2Motion=Free
  twistLimitDegrees=45
  twistMotion=Free
 }
 {
  bone1Name=dyn_C_skirt_b_3
  bone2Name=dyn_C_skirt_b_2
  swing1LimitDegrees=45
  swing1Motion=Free
  swing2LimitDegrees=45
  swing2Motion=Free
  twistLimitDegrees=45
  twistMotion=Free
 }
 {
  bone1Name=dyn_C_skirt_b_2
  bone2Name=dyn_C_skirt_b_1
  swing1LimitDegrees=45
  swing1Motion=Free
  swing2LimitDegrees=45
  swing2Motion=Free
  twistLimitDegrees=45
  twistMotion=Free
 }
 {
  bone1Name=dyn_C_skirt_b_1
  bone2Name=spine_01
  swing1LimitDegrees=45
  swing1Motion=Free
  swing2LimitDegrees=45
  swing2Motion=Free
  twistLimitDegrees=45
  twistMotion=Free
 }
 {
  bone1Name=dyn_R_skirt_c_6
  bone2Name=dyn_R_skirt_c_5
  swing1LimitDegrees=45
  swing1Motion=Free
  swing2LimitDegrees=45
  swing2Motion=Free
  twistLimitDegrees=45
  twistMotion=Free
 }
 {
  bone1Name=dyn_R_skirt_c_5
  bone2Name=dyn_R_skirt_c_4
  swing1LimitDegrees=45
  swing1Motion=Free
  swing2LimitDegrees=45
  swing2Motion=Free
  twistLimitDegrees=45
  twistMotion=Free
 }
 {
  bone1Name=dyn_R_skirt_c_4
  bone2Name=dyn_R_skirt_c_3
  swing1LimitDegrees=45
  swing1Motion=Free
  swing2LimitDegrees=45
  swing2Motion=Free
  twistLimitDegrees=45
  twistMotion=Free
 }
 {
  bone1Name=dyn_R_skirt_c_3
  bone2Name=dyn_R_skirt_c_2
  swing1LimitDegrees=45
  swing1Motion=Free
  swing2LimitDegrees=45
  swing2Motion=Free
  twistLimitDegrees=45
  twistMotion=Free
 }
 {
  bone1Name=dyn_R_skirt_c_2
  bone2Name=dyn_R_skirt_c_1
  swing1LimitDegrees=45
  swing1Motion=Free
  swing2LimitDegrees=45
  swing2Motion=Free
  twistLimitDegrees=45
  twistMotion=Free
 }
 {
  bone1Name=dyn_R_skirt_c_1
  bone2Name=spine_01
  swing1LimitDegrees=45
  swing1Motion=Free
  swing2LimitDegrees=45
  swing2Motion=Free
  twistLimitDegrees=45
  twistMotion=Free
 }
]
