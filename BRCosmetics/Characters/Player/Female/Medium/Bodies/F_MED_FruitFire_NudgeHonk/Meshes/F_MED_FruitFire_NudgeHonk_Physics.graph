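# Physics asset for the F_MED_FruitFire_NudgeHonk female player body cosmetic, covering its skirt and legs with capsule bodies.
# Six skirt chains (left/right a, b, c) each run six dyn segments from a small base body, with twist locked and per-chain swing limits.
# Free-rotation constraints cross-link lower segments, mainly left to right on the a chains, with some pairs repeated.
# Thigh, calf and foot capsules form two separately constrained leg chains.
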PhysicsAsset /BRCosmetics/Characters/Player/Female/Medium/Bodies/F_MED_FruitFire_NudgeHonk/Meshes/F_MED_FruitFire_NudgeHonk_Physics
bodies {
 base_L_skirt_a [{shapeType=Capsule, length=0.505, radius=0.505, shapeName=base_L_skirt_a_capsule, rotation=<rot 0 -90 0>}]
 base_L_skirt_b [{shapeType=Capsule, length=0.505, radius=0.505, shapeName=base_L_skirt_b_capsule, rotation=<rot 0 -90 0>}]
 base_L_skirt_c [{shapeType=Capsule, length=0.505, radius=0.505, shapeName=base_L_skirt_c_capsule, rotation=<rot 0 -90 0>}]
 base_R_skirt_a [{shapeType=Capsule, length=0.505, radius=0.505, shapeName=base_R_skirt_a_capsule, rotation=<rot 0 -90 0>}]
 base_R_skirt_b [{shapeType=Capsule, length=0.505, radius=0.505, shapeName=base_R_skirt_b_capsule, rotation=<rot 0 -90 0>}]
 base_R_skirt_c [{shapeType=Capsule, length=0.505, radius=0.505, shapeName=base_R_skirt_c_capsule, rotation=<rot 0 -90 0>}]
 calf_l [{shapeType=Capsule, length=32.0345, radius=12.7497, center=<-21.4473 -0.5027 3.2184>, rotation=<rot 0 89.6715 -87.6413>}]
 calf_r [{shapeType=Capsule, length=29.3586, radius=12.0531, center=<21.1267 0.1032 -4.2626>, rotation=<rot 0.1756 87.415 -87.2391>}]
 dyn_L_skirt_a_1 [{shapeType=Capsule, length=8.5433, radius=13.4567, shapeName=dyn_L_skirt_a_1_capsule, center=<15.2648 0 9.8927>, rotation=<rot 90 90 0>}]
 dyn_L_skirt_a_2 [{shapeType=Capsule, length=24.5366, radius=13.4567, shapeName=dyn_L_skirt_a_2_capsule, center=<12.3397 3.129 9.411>, rotation=<rot 90 -90 0>}]
 dyn_L_skirt_a_3 [{shapeType=Capsule, length=24.5366, radius=13.4567, shapeName=dyn_L_skirt_a_3_capsule, center=<10.6525 3.129 9.0245>, rotation=<rot 90 -90 0>}]
 dyn_L_skirt_a_4 [{shapeType=Capsule, length=24.5366, radius=13.4567, shapeName=dyn_L_skirt_a_4_capsule, center=<11.2741 3.129 8.2165>, rotation=<rot 90 -90 0>}]
 dyn_L_skirt_a_5 [{shapeType=Capsule, length=24.5366, radius=13.4567, shapeName=dyn_L_skirt_a_5_capsule, center=<9.5626 3.129 8.2165>, rotation=<rot 90 -90 0>}]
 dyn_L_skirt_a_6 [{shapeType=Capsule, length=24.5366, radius=13.4567, shapeName=dyn_L_skirt_a_6_capsule, center=<9.5573 3.129 8.2165>, rotation=<rot 90 -90 0>}]
 dyn_L_skirt_b_1 [{shapeType=Capsule, length=12, radius=10, shapeName=dyn_L_skirt_b_1_capsule, center=<4 0 11.9986>, rotation=<rot 90 0 0>}]
 dyn_L_skirt_b_2 [{shapeType=Capsule, length=26.5901, radius=10, shapeName=dyn_L_skirt_b_2_capsule, center=<9.7111 0 9.0699>, rotation=<rot 90 -90 0>}]
 dyn_L_skirt_b_3 [{shapeType=Capsule, length=26.5901, radius=10, shapeName=dyn_L_skirt_b_3_capsule, center=<9.7111 0 9.0213>, rotation=<rot 90 -90 0>}]
 dyn_L_skirt_b_4 [{
   shapeType=Capsule
   length=26.5901
   radius=10
   shapeName=dyn_L_skirt_b_4_capsule
   center=<13.5594 0 9.0791>
   rotation=<rot 85.8392 -179.9999 90>
  }]
 dyn_L_skirt_b_5 [{
   shapeType=Capsule
   length=26.5901
   radius=10
   shapeName=dyn_L_skirt_b_5_capsule
   center=<9.5464 0 9.3625>
   rotation=<rot 83.5025 -179.9999 90>
  }]
 dyn_L_skirt_b_6 [{shapeType=Capsule, length=26.5901, radius=10, shapeName=dyn_L_skirt_b_6_capsule, center=<8.6595 0 7.0838>, rotation=<rot 90 -90 0>}]
 dyn_L_skirt_c_1 [{shapeType=Capsule, length=31.0935, radius=10, shapeName=dyn_L_skirt_c_1_capsule, center=<11.3255 0.1229 9.7047>, rotation=<rot 90 0 0>}]
 dyn_L_skirt_c_2 [{
   shapeType=Capsule
   length=23.0935
   radius=10
   shapeName=dyn_L_skirt_c_2_capsule
   center=<0.0003 0.1229 9.4247>
   rotation=<rot 68.7445 89.8776 179.8679>
  }]
 dyn_L_skirt_c_3 [{
   shapeType=Capsule
   length=23.0935
   radius=10
   shapeName=dyn_L_skirt_c_3_capsule
   center=<0.0003 0.1229 8.8447>
   rotation=<rot 68.7445 89.8776 179.8679>
  }]
 dyn_L_skirt_c_4 [{
   shapeType=Capsule
   length=23.0935
   radius=10
   shapeName=dyn_L_skirt_c_4_capsule
   center=<0.0003 0.1229 8.8447>
   rotation=<rot 68.7445 89.8776 179.8679>
  }]
 dyn_L_skirt_c_5 [{
   shapeType=Capsule
   length=23.0935
   radius=10
   shapeName=dyn_L_skirt_c_5_capsule
   center=<0.002 1.0962 10.8001>
   rotation=<rot 68.7445 89.8777 179.8679>
  }]
 dyn_L_skirt_c_6 [{
   shapeType=Capsule
   length=23.0935
   radius=10
   shapeName=dyn_L_skirt_c_6_capsule
   center=<0.0006 0.2995 9.2987>
   rotation=<rot 68.7445 89.8777 179.8679>
  }]
 dyn_R_skirt_a_1 [{shapeType=Capsule, length=18.1042, radius=12.3074, shapeName=dyn_R_skirt_a_1_capsule, center=<16.0338 -1.2023 6.5784>, rotation=<rot 0 0 90>}]
 dyn_R_skirt_a_2 [{
   shapeType=Capsule
   length=25.011
   radius=11.753
   shapeName=dyn_R_skirt_a_2_capsule
   center=<10.6186 -7.8832 9.3071>
   rotation=<rot 87.4856 -90.009 -0.009>
  }]
 dyn_R_skirt_a_3 [{
   shapeType=Capsule
   length=24.3249
   radius=12.4391
   shapeName=dyn_R_skirt_a_3_capsule
   center=<9.7466 -8.0263 6.3677>
   rotation=<rot 87.4856 -90.0089 -0.0089>
  }]
 dyn_R_skirt_a_4 [{
   shapeType=Capsule
   length=25.1689
   radius=11.595
   shapeName=dyn_R_skirt_a_4_capsule
   center=<11.6367 -8.3735 4.46>
   rotation=<rot 87.4856 -90.0089 -0.0089>
  }]
 dyn_R_skirt_a_5 [{
   shapeType=Capsule
   length=24.0133
   radius=12.7507
   shapeName=dyn_R_skirt_a_5_capsule
   center=<10.7879 -9.0422 7.3385>
   rotation=<rot 87.4856 -90.009 -0.009>
  }]
 dyn_R_skirt_a_6 [{
   shapeType=Capsule
   length=23.5993
   radius=13.1646
   shapeName=dyn_R_skirt_a_6_capsule
   center=<10.3959 -6.4045 7.5911>
   rotation=<rot 88.5743 90.07 -179.9974>
  }]
 dyn_R_skirt_b_1 [{shapeType=Capsule, length=20.8598, radius=12.8972, shapeName=dyn_R_skirt_b_1_capsule, center=<11.3253 0 10.6118>, rotation=<rot 90 0 0>}]
 dyn_R_skirt_b_2 [{shapeType=Capsule, length=30.5727, radius=12.8972, shapeName=dyn_R_skirt_b_2_capsule, center=<0 0 9.7064>, rotation=<rot 90 -90 0>}]
 dyn_R_skirt_b_3 [{shapeType=Capsule, length=30.5727, radius=12.8972, shapeName=dyn_R_skirt_b_3_capsule, center=<0 0 9.2923>, rotation=<rot 90 -90 0>}]
 dyn_R_skirt_b_4 [{shapeType=Capsule, length=30.5727, radius=12.8972, shapeName=dyn_R_skirt_b_4_capsule, center=<0 0 8.8537>, rotation=<rot 90 -90 0>}]
 dyn_R_skirt_b_5 [{shapeType=Capsule, length=30.5727, radius=12.8972, shapeName=dyn_R_skirt_b_5_capsule, center=<0 0 8.5306>, rotation=<rot 90 -90 0>}]
 dyn_R_skirt_b_6 [{shapeType=Capsule, length=30.5727, radius=12.8972, shapeName=dyn_R_skirt_b_6_capsule, center=<0 0 8.1637>, rotation=<rot 90 -90 0>}]
 dyn_R_skirt_c_1 [{shapeType=Capsule, length=42.4085, radius=10, shapeName=dyn_R_skirt_c_1_capsule, center=<11.3257 0.304 8.8792>, rotation=<rot 90 0 0>}]
 dyn_R_skirt_c_2 [{
   shapeType=Capsule
   length=34.4085
   radius=10
   shapeName=dyn_R_skirt_c_2_capsule
   center=<-0.0009 -0.6641 7.5552>
   rotation=<rot 57.8085 -90.0759 0.0002>
  }]
 dyn_R_skirt_c_3 [{
   shapeType=Capsule
   length=34.4085
   radius=10
   shapeName=dyn_R_skirt_c_3_capsule
   center=<-0.0005 -0.3559 7.0657>
   rotation=<rot 57.8085 -90.0759 0.0002>
  }]
 dyn_R_skirt_c_4 [{
   shapeType=Capsule
   length=34.4085
   radius=10
   shapeName=dyn_R_skirt_c_4_capsule
   center=<-0.0006 -0.44 7.1992>
   rotation=<rot 57.8085 -90.0759 0.0002>
  }]
 dyn_R_skirt_c_5 [{
   shapeType=Capsule
   length=34.4085
   radius=10
   shapeName=dyn_R_skirt_c_5_capsule
   center=<-0.0001 -0.0903 6.6438>
   rotation=<rot 57.8085 -90.0759 0.0002>
  }]
 dyn_R_skirt_c_6 [{
   shapeType=Capsule
   length=34.4085
   radius=10
   shapeName=dyn_R_skirt_c_6_capsule
   center=<0 0.0326 6.4486>
   rotation=<rot 57.8085 -90.0759 0.0002>
  }]
 foot_l [{shapeType=Capsule, length=11.3092, radius=8.4835, center=<-4.7006 8.7373 -0.7485>, rotation=<rot -0.1388 14.4567 -88.0993>}]
 foot_r [{shapeType=Capsule, length=12.9473, radius=9.2427, center=<2.5978 -8.558 0.6804>, rotation=<rot -0.2128 15.5351 -87.8405>}]
 thigh_l [{shapeType=Capsule, length=33.3449, radius=11.97, center=<-28.6272 -0.2854 3.7923>, rotation=<rot 0.2277 -90.8292 -91.0105>}]
 thigh_r [{shapeType=Capsule, length=37.1089, radius=11.3131, center=<29.2437 0.6226 -4.1001>, rotation=<rot -0.1872 -88.9415 -91.5788>}]
}
constraints [
 {bone1Name=calf_l, bone2Name=thigh_l, swing1LimitDegrees=45, swing2LimitDegrees=45, twistLimitDegrees=45}
 {bone1Name=foot_l, bone2Name=calf_l, swing1LimitDegrees=45, swing2LimitDegrees=45, twistLimitDegrees=45}
 {bone1Name=calf_r, bone2Name=thigh_r, swing1LimitDegrees=45, swing2LimitDegrees=45, twistLimitDegrees=45}
 {bone1Name=foot_r, bone2Name=calf_r, swing1LimitDegrees=45, swing2LimitDegrees=45, twistLimitDegrees=45}
 {bone1Name=dyn_L_skirt_a_3, bone2Name=dyn_L_skirt_a_2, swing1LimitDegrees=0, swing2LimitDegrees=24, twistLimitDegrees=0, twistMotion=Locked}
 {bone1Name=dyn_L_skirt_a_4, bone2Name=dyn_L_skirt_a_3, swing1LimitDegrees=0, swing2LimitDegrees=24, twistLimitDegrees=0, twistMotion=Locked}
 {bone1Name=dyn_L_skirt_a_5, bone2Name=dyn_L_skirt_a_4, swing1LimitDegrees=0, swing2LimitDegrees=24, twistLimitDegrees=0, twistMotion=Locked}
 {bone1Name=dyn_L_skirt_b_3, bone2Name=dyn_L_skirt_b_2, swing1LimitDegrees=15, swing2LimitDegrees=4, twistLimitDegrees=1, twistMotion=Locked}
 {bone1Name=dyn_L_skirt_b_4, bone2Name=dyn_L_skirt_b_3, swing1LimitDegrees=15, swing2LimitDegrees=4, twistLimitDegrees=1, twistMotion=Locked}
 {bone1Name=dyn_L_skirt_b_5, bone2Name=dyn_L_skirt_b_4, swing1LimitDegrees=15, swing2LimitDegrees=4, twistLimitDegrees=1, twistMotion=Locked}
 {bone1Name=dyn_L_skirt_c_3, bone2Name=dyn_L_skirt_c_2, swing1LimitDegrees=1, swing2LimitDegrees=15, twistLimitDegrees=1, twistMotion=Locked}
 {bone1Name=dyn_L_skirt_c_4, bone2Name=dyn_L_skirt_c_3, swing1LimitDegrees=1, swing2LimitDegrees=15, twistLimitDegrees=1, twistMotion=Locked}
 {bone1Name=dyn_L_skirt_c_5, bone2Name=dyn_L_skirt_c_4, swing1LimitDegrees=1, swing2LimitDegrees=15, twistLimitDegrees=1, twistMotion=Locked}
 {bone1Name=dyn_R_skirt_a_3, bone2Name=dyn_R_skirt_a_2, swing1LimitDegrees=15, swing2LimitDegrees=4, twistLimitDegrees=1, twistMotion=Locked}
 {bone1Name=dyn_R_skirt_a_4, bone2Name=dyn_R_skirt_a_3, swing1LimitDegrees=15, swing2LimitDegrees=4, twistLimitDegrees=1, twistMotion=Locked}
 {bone1Name=dyn_R_skirt_a_5, bone2Name=dyn_R_skirt_a_4, swing1LimitDegrees=15, swing2LimitDegrees=4, twistLimitDegrees=1, twistMotion=Locked}
 {bone1Name=dyn_R_skirt_b_3, bone2Name=dyn_R_skirt_b_2, swing1LimitDegrees=15, swing2LimitDegrees=4, twistLimitDegrees=1, twistMotion=Locked}
 {bone1Name=dyn_R_skirt_b_4, bone2Name=dyn_R_skirt_b_3, swing1LimitDegrees=15, swing2LimitDegrees=4, twistLimitDegrees=1, twistMotion=Locked}
 {bone1Name=dyn_R_skirt_b_5, bone2Name=dyn_R_skirt_b_4, swing1LimitDegrees=15, swing2LimitDegrees=4, twistLimitDegrees=1, twistMotion=Locked}
 {bone1Name=dyn_R_skirt_c_3, bone2Name=dyn_R_skirt_c_2, swing1LimitDegrees=2, swing2LimitDegrees=25, twistLimitDegrees=1, twistMotion=Locked}
 {bone1Name=dyn_R_skirt_c_4, bone2Name=dyn_R_skirt_c_3, swing1LimitDegrees=2, swing2LimitDegrees=25, twistLimitDegrees=1, twistMotion=Locked}
 {bone1Name=dyn_R_skirt_c_5, bone2Name=dyn_R_skirt_c_4, swing1LimitDegrees=2, swing2LimitDegrees=25, twistLimitDegrees=1, twistMotion=Locked}
 {bone1Name=dyn_L_skirt_a_2, bone2Name=dyn_L_skirt_a_1, swing1LimitDegrees=0, swing2LimitDegrees=24, twistLimitDegrees=0, twistMotion=Locked}
 {bone1Name=dyn_L_skirt_b_2, bone2Name=dyn_L_skirt_b_1, swing1LimitDegrees=15, swing2LimitDegrees=4, twistLimitDegrees=45, twistMotion=Locked}
 {bone1Name=dyn_L_skirt_c_2, bone2Name=dyn_L_skirt_c_1, swing1LimitDegrees=1, swing2LimitDegrees=15, twistLimitDegrees=45, twistMotion=Locked}
 {bone1Name=dyn_R_skirt_a_2, bone2Name=dyn_R_skirt_a_1, swing1LimitDegrees=15, swing2LimitDegrees=4, twistLimitDegrees=45, twistMotion=Locked}
 {bone1Name=dyn_R_skirt_b_2, bone2Name=dyn_R_skirt_b_1, swing1LimitDegrees=15, swing2LimitDegrees=4, twistLimitDegrees=45, twistMotion=Locked}
 {bone1Name=dyn_R_skirt_c_2, bone2Name=dyn_R_skirt_c_1, swing1LimitDegrees=2, swing2LimitDegrees=25, twistLimitDegrees=45, twistMotion=Locked}
 {bone1Name=dyn_R_skirt_c_6, bone2Name=dyn_R_skirt_c_5, swing1LimitDegrees=2, swing2LimitDegrees=25, twistLimitDegrees=45, twistMotion=Locked}
 {bone1Name=dyn_R_skirt_b_6, bone2Name=dyn_R_skirt_b_5, swing1LimitDegrees=15, swing2LimitDegrees=4, twistLimitDegrees=45, twistMotion=Locked}
 {bone1Name=dyn_R_skirt_a_6, bone2Name=dyn_R_skirt_a_5, swing1LimitDegrees=15, swing2LimitDegrees=4, twistLimitDegrees=45, twistMotion=Locked}
 {bone1Name=dyn_L_skirt_c_6, bone2Name=dyn_L_skirt_c_5, swing1LimitDegrees=1, swing2LimitDegrees=15, twistLimitDegrees=45, twistMotion=Locked}
 {bone1Name=dyn_L_skirt_b_6, bone2Name=dyn_L_skirt_b_5, swing1LimitDegrees=15, swing2LimitDegrees=4, twistLimitDegrees=45, twistMotion=Locked}
 {bone1Name=dyn_L_skirt_a_6, bone2Name=dyn_L_skirt_a_5, swing1LimitDegrees=0, swing2LimitDegrees=24, twistLimitDegrees=0, twistMotion=Locked}
 {bone1Name=dyn_L_skirt_a_1, bone2Name=base_L_skirt_a, swing1LimitDegrees=5, swing2LimitDegrees=24, twistLimitDegrees=2, twistMotion=Locked}
 {bone1Name=dyn_L_skirt_b_1, bone2Name=base_L_skirt_b, swing1LimitDegrees=45, swing2LimitDegrees=2, twistLimitDegrees=45, twistMotion=Locked}
 {bone1Name=dyn_L_skirt_c_1, bone2Name=base_L_skirt_c, swing1LimitDegrees=1, swing2LimitDegrees=15, twistLimitDegrees=1, twistMotion=Locked}
 {bone1Name=dyn_R_skirt_a_1, bone2Name=base_R_skirt_a, swing1LimitDegrees=2, swing2LimitDegrees=15, twistLimitDegrees=45, twistMotion=Locked}
 {bone1Name=dyn_R_skirt_b_1, bone2Name=base_R_skirt_b, swing1LimitDegrees=45, swing2LimitDegrees=2, twistLimitDegrees=45, twistMotion=Locked}
 {bone1Name=dyn_R_skirt_c_1, bone2Name=base_R_skirt_c, swing1LimitDegrees=2, swing2LimitDegrees=25, twistLimitDegrees=45, twistMotion=Locked}
 {
  bone1Name=dyn_R_skirt_a_5
  bone2Name=dyn_L_skirt_a_5
  swing1LimitDegrees=45
  swing1Motion=Free
  swing2LimitDegrees=45
  swing2Motion=Free
  twistLimitDegrees=45
  twistMotion=Free
 }
 {
  bone1Name=dyn_R_skirt_a_6
  bone2Name=dyn_L_skirt_a_6
  swing1LimitDegrees=45
  swing1Motion=Free
  swing2LimitDegrees=45
  swing2Motion=Free
  twistLimitDegrees=45
  twistMotion=Free
 }
 {
  bone1Name=dyn_L_skirt_a_6
  bone2Name=dyn_R_skirt_a_6
  swing1LimitDegrees=45
  swing1Motion=Free
  swing2LimitDegrees=45
  swing2Motion=Free
  twistLimitDegrees=45
  twistMotion=Free
 }
 {
  bone1Name=dyn_R_skirt_a_6
  bone2Name=dyn_L_skirt_a_6
  swing1LimitDegrees=45
  swing1Motion=Free
  swing2LimitDegrees=45
  swing2Motion=Free
  twistLimitDegrees=45
  twistMotion=Free
 }
 {
  bone1Name=dyn_R_skirt_a_5
  bone2Name=dyn_R_skirt_b_5
  swing1LimitDegrees=45
  swing1Motion=Free
  swing2LimitDegrees=45
  swing2Motion=Free
  twistLimitDegrees=45
  twistMotion=Free
 }
]
