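# Post-simulation Control Rig for the FruitFire_NudgeHonk female body that corrects skirt bones after simulation.
# A base robe offset feeds curve-driven deforms, including Kalman-smoothed thigh rotation deltas from the previous frame.
# Center skirt chains are then constrained between left and right chains.
# Spherical pose readers on thighs and calves add crouch-dependent offsets, with extra poses for DBNO, sliding and pirouette.
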ControlRigBlueprint /BRCosmetics/Characters/Player/Female/Medium/Bodies/F_MED_FruitFire_NudgeHonk/Meshes/F_MED_FruitFire_NudgeHonk_PostSim_CtrlRig
graphs [
 New Function_ContainedGraph {nodes [
   RigVMFunctionEntryNode()
   n1 = &RigUnit_GetTransform(Item=(Type=Bone,Name="thigh_r"), Space=LocalSpace, bInitial=False)
   n2 = /Script/RigVM.RigVMFunction_MathQuaternionToAxisAndAngle(Value=/Script/RigVM.RigVMFunction_MathQuaternionMul(
     B=/Script/RigVM.RigVMFunction_MathQuaternionInverse(Value=@n1.Transform.Rotation) {is=struct RigVMUnitNode, template="Inverse::Execute(in Value,out Result)"}.Result
     A=RigVMVariableNode() {variable=prev thigh r quatrnion}.Value
    ) {is=struct RigVMUnitNode, template="Multiply::Execute(in A,in B,out Result)"}.Result) {is=struct RigVMUnitNode, template="ToAxisAndAngle::Execute(in Value,out Axis,out Angle)"}
   n3 = &RigVMFunction_KalmanFloat(
    BufferSize=4
    Value=&RigVMFunction_MathDoubleRemap(
     SourceMaximum=15
     SourceMinimum=0
     TargetMaximum=1
     TargetMinimum=0
     Value=/Script/RigVM.RigVMFunction_MathDoubleDeg(Value=@n2.Angle) {is=struct RigVMUnitNode, template="Degrees::Execute(in Value,out Result)"}.Result
    ).Result
   ) {template="KalmanFilter::Execute(in Value,in BufferSize,out Result)"}
   RigUnit_GetTransform = &RigUnit_GetTransform(Item=(Type=Bone,Name="thigh_l"), Space=LocalSpace, bInitial=False)
   Inverse = &RigVMFunction_MathQuaternionInverse(Value=@RigUnit_GetTransform.Transform.Rotation) {template="Inverse::Execute(in Value,out Result)"}
   RigVMFunction_MathQuaternionToAxisAndAngle = &RigVMFunction_MathQuaternionToAxisAndAngle(Value=&RigVMFunction_MathQuaternionMul(B=@Inverse.Result, A=RigVMVariableNode() {variable=prev thigh l quatrnion}.Value) {template="Multiply::Execute(in A,in B,out Result)"}.Result) {template="ToAxisAndAngle::Execute(in Value,out Axis,out Angle)"}
   Degrees = &RigVMFunction_MathDoubleDeg(Value=@RigVMFunction_MathQuaternionToAxisAndAngle.Angle) {template="Degrees::Execute(in Value,out Result)"}
   KalmanFilter = &RigVMFunction_KalmanFloat(
    BufferSize=4
    Value=&RigVMFunction_MathDoubleRemap(SourceMaximum=-15, SourceMinimum=0, TargetMaximum=-1, TargetMinimum=0, Value=@Degrees.Result).Result
   ) {template="KalmanFilter::Execute(in Value,in BufferSize,out Result)"}
   > Modify Transforms with Mirror_2 = RigVMFunctionReferenceNode(
    Base Modify Data="((Item=(Type=Bone,Name=\"base_L_skirt_a\"),Transform=(Rotation=(X=-0.0076,Y=-0.0868,Z=-0.0868,W=0.9924),Translation=(X=0,Y=-5,Z=5),Scale3D=(X=1,Y=1,Z=1))),(Item=(Type=Bone,Name=\"dyn_L_skirt_b_1\"),Transform=(Rotation=(X=-0.0114,Y=-0.0864,Z=-0.13,W=0.9877),Tra ...(truncated)"
    Mirror Modify Data="((Item=(Type=Bone,Name=\"base_R_skirt_a\"),Transform=(Rotation=(X=0.0114,Y=-0.0864,Z=0.13,W=0.9877),Translation=(X=0,Y=2,Z=8),Scale3D=(X=1,Y=1,Z=1))),(Item=(Type=Bone,Name=\"dyn_R_skirt_b_1\"),Transform=(Rotation=(X=0,Y=0,Z=0.2588,W=0.9659),Translation=(X=0,Y= ...(truncated)"
    "Mirror Weight "=@n3.Result
    Weight=@KalmanFilter.Result
   )
   Add = &RigVMFunction_MathDoubleAdd(B=@n3.Result, A=@KalmanFilter.Result)
   Clamp = &RigVMFunction_MathDoubleClamp(Maximum=0, Value=&RigVMFunction_MathDoubleMul(B=0.5, A=@Add.Result).Result)
   > &RigUnit_ModifyTransforms(
    ItemToModify=((Item=(Type=Bone,Name="dyn_L_skirt_a_1"),Transform=(Rotation=(X=0,Y=0,Z=0.0872,W=0.9962),Translation=(X=0,Y=0,Z=0),Scale3D=(X=1,Y=1,Z=1))))
    WeightMinimum=0
    Weight=@Clamp.Result
   )
   > RigVMVariableNode(Value=@RigUnit_GetTransform.Transform.Rotation) {variable=prev thigh l quatrnion}
   > RigVMVariableNode(Value=@n1.Transform.Rotation) {variable=prev thigh r quatrnion}
   AlphaInterp = &RigVMFunction_AlphaInterp(
    InterpSpeedDecreasing=10
    InterpSpeedIncreasing=10
    Scale=1
    bClampResult=False
    Value=/Script/ControlRig.RigUnit_GetCurveValue(Curve=CR Rear Knee Bend L) {is=struct RigVMUnitNode}.Value
   )
   Remap = &RigVMFunction_MathDoubleRemap(SourceMaximum=770, SourceMinimum=220, TargetMaximum=1, TargetMinimum=0, Value=RigVMVariableNode() {variable=Player Speed}.Value)
   Multiply = &RigVMFunction_MathDoubleMul(A=@AlphaInterp.Result, B=@Remap.Result)
   n4 = &RigVMFunction_AlphaInterp(
    InterpSpeedDecreasing=10
    InterpSpeedIncreasing=10
    Scale=1
    bClampResult=False
    Value=&RigUnit_GetCurveValue(Curve="CR  Rear Knee Bend R").Value
   )
   > Modify Transforms with Mirror_5 = RigVMFunctionReferenceNode(
    Base Modify Data=((Item=(Type=Bone,Name="dyn_L_skirt_c_1"),Transform=(Rotation=(X=-0.0057,Y=-0.1304,Z=-0.0432,W=0.9905),Translation=(X=0,Y=0,Z=0),Scale3D=(X=1,Y=1,Z=1))))
    Mirror Modify Data=((Item=(Type=Bone,Name="dyn_R_skirt_c_1"),Transform=(Rotation=(X=0,Y=0,Z=0,W=1),Translation=(X=0,Y=0,Z=0),Scale3D=(X=1,Y=1,Z=1))))
    Weight=@Multiply.Result
    "Mirror Weight "=&RigVMFunction_MathDoubleMul(B=@n4.Result, A=@Remap.Result).Result
   )
   > Modify Transforms with Mirror_6 = RigVMFunctionReferenceNode(
    Base Modify Data=((Item=(Type=Bone,Name="dyn_L_skirt_c_1"),Transform=(Rotation=(X=0.4022,Y=-0.1853,Z=-0.0084,W=0.8966),Translation=(X=-5,Y=-3,Z=0),Scale3D=(X=1,Y=1,Z=1))))
    Mirror Modify Data=((Item=(Type=Bone,Name="dyn_R_skirt_c_1"),Transform=(Rotation=(X=0.0114,Y=0.13,Z=-0.0864,W=0.9877),Translation=(X=10,Y=-10,Z=0),Scale3D=(X=1,Y=1,Z=1))))
    Weight=&RigUnit_GetCurveValue(Curve=CR Foot Near Pelvis L).Value
    "Mirror Weight "=&RigUnit_GetCurveValue(Curve=CR Foot Near Pelvis R).Value
   )
   > &RigUnit_ModifyTransforms(
    ItemToModify="((Item=(Type=Bone,Name=\"base_R_skirt_c\"),Transform=(Rotation=(X=0,Y=0.0893,Z=0,W=0.996),Translation=(X=0,Y=0,Z=0),Scale3D=(X=1,Y=1,Z=1))),(Item=(Type=Bone,Name=\"base_C_skirt_b\"),Transform=(Rotation=(X=0,Y=0,Z=0,W=1),Translation=(X=0,Y=0,Z=0),Scale3D=(X=1,Y ...(truncated)"
    WeightMinimum=0
    Weight=RigVMVariableNode() {variable=IsCustomLobbyIdle}.Value
   )
   > RigVMFunctionReturnNode()
  ]}
 New Function_ContainedGraph {nodes [
   RigVMFunctionEntryNode()
   > RigVMFunction_Sequence = RigVMAggregateNode()
   For_Each_2 = RigVMDispatchNode(ExecuteContext=@RigVMFunction_Sequence.B, Array=RigVMVariableNode() {variable=Rear Skirt Mid}.Value) {template="DISPATCH_RigVMDispatch_ArrayIterator(in Array,out Element,out Index,out Count,out Ratio)"}
   > &RigUnit_ParentConstraint(
    AdvancedSettings=(InterpolationType=Average,RotationOrderForFilter=XZY)
    Parents=((Item=(Type=Bone,Name="None"),Weight=0.5),(Item=(Type=Bone,Name="None"),Weight=0.5))
    Weight=1
    bMaintainOffset=false
    Parents.0.Item=RigVMDispatchNode(Index=@For_Each_2.Index, Array=RigVMVariableNode() {variable=Rear Skirt L}.Value) {template="DISPATCH_RigVMDispatch_ArrayGetAtIndex(in Array,in Index,out Element)"}.Element
    Parents.1.Item=RigVMDispatchNode(Index=@For_Each_2.Index, Array=RigVMVariableNode() {variable=Rear Skirt R}.Value) {template="DISPATCH_RigVMDispatch_ArrayGetAtIndex(in Array,in Index,out Element)"}.Element
    Child=@For_Each_2.Element
   )
   RigVMFunctionReturnNode(ExecuteContext=@RigVMFunction_Sequence.D)
   Remap = &RigVMFunction_MathDoubleRemap(
    SourceMaximum=1
    SourceMinimum=0
    TargetMaximum=0.7
    TargetMinimum=0.5
    Value=&RigUnit_GetCurveValue(Curve=CR Thigh Raised Front L).Value
   )
   Subtract = &RigVMFunction_MathDoubleSub(A=1, B=@Remap.Result)
   &RigUnit_ParentConstraint(
    AdvancedSettings=(InterpolationType=Average,RotationOrderForFilter=XZY)
    Child=(Type=Bone,Name="base_C_skirt_a")
    Filter=(TranslationFilter=(bX=True,bY=True,bZ=True),RotationFilter=(bX=true,bY=true,bZ=true),ScaleFilter=(bX=True,bY=True,bZ=True))
    Parents=((Item=(Type=Bone,Name="base_L_skirt_a"),Weight=0.5),(Item=(Type=Bone,Name="base_R_skirt_a"),Weight=0.5))
    Weight=1
    ExecutePin=@RigVMFunction_Sequence.C
    Parents.0.Weight=@Remap.Result
    Parents.1.Weight=@Subtract.Result
   )
   > &RigUnit_ParentConstraint(
    AdvancedSettings=(InterpolationType=Average,RotationOrderForFilter=XZY)
    Child=(Type=Bone,Name="dyn_C_skirt_a_1")
    Filter=(TranslationFilter=(bX=True,bY=True,bZ=True),RotationFilter=(bX=true,bY=true,bZ=true),ScaleFilter=(bX=True,bY=True,bZ=True))
    Parents=((Item=(Type=Bone,Name="dyn_L_skirt_a_1"),Weight=0.5),(Item=(Type=Bone,Name="dyn_R_skirt_a_1"),Weight=0.5))
    Weight=1
    Parents.0.Weight=@Remap.Result
    Parents.1.Weight=@Subtract.Result
   )
   > &RigUnit_ParentConstraint(
    AdvancedSettings=(InterpolationType=Average,RotationOrderForFilter=XZY)
    Child=(Type=Bone,Name="dyn_C_skirt_a_2")
    Filter=(TranslationFilter=(bX=True,bY=True,bZ=True),RotationFilter=(bX=true,bY=true,bZ=true),ScaleFilter=(bX=True,bY=True,bZ=True))
    Parents=((Item=(Type=Bone,Name="dyn_L_skirt_a_2"),Weight=0.5),(Item=(Type=Bone,Name="dyn_R_skirt_a_2"),Weight=0.5))
    Weight=1
    Parents.0.Weight=@Remap.Result
    Parents.1.Weight=@Subtract.Result
   )
   Add = &RigVMFunction_MathDoubleAdd(
    A=&RigUnit_GetCurveValue(Curve=CR Thigh Raised Front L).Value
    B=&RigUnit_GetCurveValue(Curve="CR ThighRaised  Front R").Value
   )
   Multiply = &RigVMFunction_MathDoubleMul(B=0.5, A=@Add.Result)
   Clamp = &RigVMFunction_MathDoubleClamp(Value=@Multiply.Result)
   > &RigUnit_ModifyTransforms(
    ItemToModify="((Item=(Type=Bone,Name=\"base_C_skirt_a\"),Transform=(Rotation=(X=0,Y=-0.2164,Z=0,W=0.9763),Translation=(X=0,Y=0,Z=0),Scale3D=(X=1,Y=1,Z=1))),(Item=(Type=Bone,Name=\"base_C_skirt_b\"),Transform=(Rotation=(X=0,Y=0,Z=0,W=1),Translation=(X=0,Y=0,Z=0),Scale3D=(X=1 ...(truncated)"
    WeightMinimum=0
    Weight=@Clamp.Result
   )
   > &RigUnit_ModifyTransforms(
    ItemToModify="((Item=(Type=Bone,Name=\"base_C_skirt_a\"),Transform=(Rotation=(X=-0.0868,Y=0.0076,Z=-0.0868,W=0.9924),Translation=(X=0,Y=0,Z=0),Scale3D=(X=1,Y=1,Z=1))),(Item=(Type=Bone,Name=\"base_C_skirt_b\"),Transform=(Rotation=(X=0,Y=0,Z=0,W=1),Translation=(X=0,Y=0,Z=0),S ...(truncated)"
    WeightMinimum=0
    Weight=&RigUnit_GetCurveValue(Curve=CR Thigh Raised Front L).Value
   )
   > &RigUnit_ModifyTransforms(
    ItemToModify="((Item=(Type=Bone,Name=\"base_C_skirt_a\"),Transform=(Rotation=(X=0.0868,Y=0.0076,Z=0.0868,W=0.9924),Translation=(X=0,Y=0,Z=0),Scale3D=(X=1,Y=1,Z=1))),(Item=(Type=Bone,Name=\"base_C_skirt_b\"),Transform=(Rotation=(X=0,Y=0,Z=0,W=1),Translation=(X=0,Y=0,Z=0),Sca ...(truncated)"
    WeightMinimum=0
    Weight=&RigUnit_GetCurveValue(Curve="CR ThighRaised  Front R").Value
   )
   > &RigUnit_ParentConstraint(
    AdvancedSettings=(InterpolationType=Average,RotationOrderForFilter=XZY)
    Child=(Type=Bone,Name="dyn_C_skirt_a_3")
    Filter=(TranslationFilter=(bX=True,bY=True,bZ=True),RotationFilter=(bX=true,bY=true,bZ=true),ScaleFilter=(bX=True,bY=True,bZ=True))
    Parents=((Item=(Type=Bone,Name="dyn_L_skirt_a_3"),Weight=0.5),(Item=(Type=Bone,Name="dyn_R_skirt_a_3"),Weight=0.5))
    Weight=1
    Parents.1.Weight=@Subtract.Result
    Parents.0.Weight=@Remap.Result
   )
   > &RigUnit_ParentConstraint(
    AdvancedSettings=(InterpolationType=Average,RotationOrderForFilter=XZY)
    Child=(Type=Bone,Name="dyn_C_skirt_a_4")
    Filter=(TranslationFilter=(bX=True,bY=True,bZ=True),RotationFilter=(bX=true,bY=true,bZ=true),ScaleFilter=(bX=True,bY=True,bZ=True))
    Parents=((Item=(Type=Bone,Name="dyn_L_skirt_a_4"),Weight=0.5),(Item=(Type=Bone,Name="dyn_R_skirt_a_4"),Weight=0.5))
    Weight=1
    Parents.1.Weight=@Subtract.Result
    Parents.0.Weight=@Remap.Result
   )
   > &RigUnit_ParentConstraint(
    AdvancedSettings=(InterpolationType=Average,RotationOrderForFilter=XZY)
    Child=(Type=Bone,Name="dyn_C_skirt_a_5")
    Filter=(TranslationFilter=(bX=True,bY=True,bZ=True),RotationFilter=(bX=true,bY=true,bZ=true),ScaleFilter=(bX=True,bY=True,bZ=True))
    Parents=((Item=(Type=Bone,Name="dyn_L_skirt_a_5"),Weight=0.5),(Item=(Type=Bone,Name="dyn_R_skirt_a_5"),Weight=0.5))
    Weight=1
    Parents.1.Weight=@Subtract.Result
    Parents.0.Weight=@Remap.Result
   )
   > &RigUnit_ParentConstraint(
    AdvancedSettings=(InterpolationType=Average,RotationOrderForFilter=XZY)
    Child=(Type=Bone,Name="dyn_C_skirt_a_6")
    Filter=(TranslationFilter=(bX=True,bY=True,bZ=True),RotationFilter=(bX=true,bY=true,bZ=true),ScaleFilter=(bX=True,bY=True,bZ=True))
    Parents=((Item=(Type=Bone,Name="dyn_L_skirt_a_6"),Weight=0.5),(Item=(Type=Bone,Name="dyn_R_skirt_a_6"),Weight=0.5))
    Weight=1
    Parents.1.Weight=@Subtract.Result
    Parents.0.Weight=@Remap.Result
   )
   For_Each_1 = RigVMDispatchNode(ExecuteContext=(), Array=RigVMVariableNode() {variable=Front Skirt Mid}.Value) {template="DISPATCH_RigVMDispatch_ArrayIterator(in Array,out Element,out Index,out Count,out Ratio)"}
   > &RigUnit_ParentConstraint(
    AdvancedSettings=(InterpolationType=Average,RotationOrderForFilter=XZY)
    Filter=(TranslationFilter=(bX=True,bY=True,bZ=True),RotationFilter=(bX=true,bY=true,bZ=true),ScaleFilter=(bX=True,bY=True,bZ=True))
    Parents=((Item=(Type=Bone,Name="None"),Weight=0.5),(Item=(Type=Bone,Name="None"),Weight=0.5))
    Weight=1
    Parents.0.Item=RigVMDispatchNode(Index=@For_Each_1.Index, Array=RigVMVariableNode() {variable=Front Skirt L}.Value) {template="DISPATCH_RigVMDispatch_ArrayGetAtIndex(in Array,in Index,out Element)"}.Element
    Parents.1.Item=RigVMDispatchNode(Index=@For_Each_1.Index, Array=RigVMVariableNode() {variable=Front Skirt R}.Value) {template="DISPATCH_RigVMDispatch_ArrayGetAtIndex(in Array,in Index,out Element)"}.Element
    Child=@For_Each_1.Element
   )
   &RigVMFunction_MathDoubleRemap(
    SourceMaximum=1
    SourceMinimum=0
    TargetMaximum=1
    TargetMinimum=0
    Value=&RigUnit_GetCurveValue(Curve="CR ThighRaised  Front R").Value
   )
   RigVMRerouteNode(Value=())
  ]}
 New Function_ContainedGraph {nodes [
   RigVMFunctionEntryNode()
   > RigVMFunction_Sequence = RigVMAggregateNode()
   For_Each_1 = RigVMDispatchNode(Array=(), ExecuteContext=@RigVMFunction_Sequence.B) {template="DISPATCH_RigVMDispatch_ArrayIterator(in Array,out Element,out Index,out Count,out Ratio)"}
   > &RigUnit_AimConstraintLocalSpaceOffset(
    AdvancedSettings=(DebugSettings=(bEnabled=true,Scale=10,WorldOffset=(Rotation=(X=0,Y=0,Z=0,W=1),Translation=(X=0,Y=0,Z=0),Scale3D=(X=1,Y=1,Z=1))),RotationOrderForFilter=XZY)
    Filter=(bX=True,bY=True,bZ=True)
    Parents=((Item=(Type=Bone,Name="None"),Weight=1))
    UpAxis=(X=0,Y=-1,Z=0)
    Weight=1
    WorldUp=(Target=(X=0,Y=1,Z=0),Kind=Location,Space=(Type=None,Name="None"))
    bMaintainOffset=true
    Parents.0.Item=RigVMDispatchNode(Array=(), Index=@For_Each_1.Index) {template="DISPATCH_RigVMDispatch_ArrayGetAtIndex(in Array,in Index,out Element)"}.Element
    Child=@For_Each_1.Element
    WorldUp.Space=@For_Each_1.Element
   )
   For_Each_2 = RigVMDispatchNode(Array=(), ExecuteContext=@RigVMFunction_Sequence.C) {template="DISPATCH_RigVMDispatch_ArrayIterator(in Array,out Element,out Index,out Count,out Ratio)"}
   > &RigUnit_AimConstraintLocalSpaceOffset(
    AimAxis=(X=0,Y=0,Z=1)
    Filter=(bX=True,bY=True,bZ=True)
    Parents=((Item=(Type=Bone,Name="None"),Weight=1))
    UpAxis=(X=0,Y=1,Z=0)
    Weight=1
    WorldUp=(Target=(X=0,Y=-1,Z=0),Kind=Location,Space=(Type=None,Name="None"))
    Parents.0.Item=RigVMDispatchNode(Array=(), Index=@For_Each_2.Index) {template="DISPATCH_RigVMDispatch_ArrayGetAtIndex(in Array,in Index,out Element)"}.Element
    Child=@For_Each_2.Element
    WorldUp.Space=@For_Each_2.Element
   )
   RigVMFunctionReturnNode(ExecuteContext=@RigVMFunction_Sequence.D)
   For_Each = RigVMDispatchNode(Array=RigVMVariableNode() {variable=Front Skirt Mid}.Value, ExecuteContext=@RigVMFunction_Sequence.A) {template="DISPATCH_RigVMDispatch_ArrayIterator(in Array,out Element,out Index,out Count,out Ratio)"}
   > &RigUnit_ParentConstraint(
    AdvancedSettings=(InterpolationType=Average,RotationOrderForFilter=XZY)
    Parents=((Item=(Type=Bone,Name="None"),Weight=1))
    Weight=1
    Parents.0.Item=RigVMDispatchNode(Array=(), Index=@For_Each.Index) {template="DISPATCH_RigVMDispatch_ArrayGetAtIndex(in Array,in Index,out Element)"}.Element
    Child=@For_Each.Element
   )
  ]}
 New Function_ContainedGraph {nodes [
   RigVMFunctionEntryNode()
   > RigVMFunction_Sequence = RigVMAggregateNode()
   RigVMFunctionReturnNode(ExecuteContext=@RigVMFunction_Sequence.C)
   Maximum = &RigVMFunction_MathDoubleMax(
    A=&RigUnit_GetCurveValue(Curve=CR Thigh Raised Front L).Value
    B=&RigUnit_GetCurveValue(Curve="CR ThighRaised  Front R").Value
   )
   KalmanFilter = &RigVMFunction_KalmanFloat(BufferSize=8, Value=@Maximum.Result) {template="KalmanFilter::Execute(in Value,in BufferSize,out Result)"}
   &RigUnit_ModifyTransforms(
    ItemToModify="((Item=(Type=Bone,Name=\"dyn_front_panel_02\"),Transform=(Rotation=(X=0,Y=0,Z=-0.0175,W=0.9998),Translation=(X=0,Y=0,Z=0),Scale3D=(X=1,Y=1,Z=1))),(Item=(Type=Bone,Name=\"dyn_front_panel_MSTR\"),Transform=(Rotation=(X=0,Y=0,Z=-0.0175,W=0.9998),Translation=(X=0, ...(truncated)"
    WeightMinimum=0
    Weight=@KalmanFilter.Result
    ExecutePin=@RigVMFunction_Sequence.B
   )
   > &RigUnit_ModifyTransforms(
    ItemToModify=((Item=(Type=Bone,Name="dyn_front_panel_02"),Transform=(Rotation=(X=-0.1171,Y=-0.2446,Z=-0.148,W=0.9511),Translation=(X=0,Y=0,Z=0),Scale3D=(X=1,Y=1,Z=1))))
    WeightMinimum=0
    Weight=&RigUnit_GetCurveValue(Curve=CR Thigh L Across R).Value
   )
   > &RigUnit_ModifyTransforms(
    ItemToModify=((Item=(Type=Bone,Name="dyn_front_panel_02"),Transform=(Rotation=(X=0.1171,Y=0.2446,Z=-0.148,W=0.9511),Translation=(X=0,Y=0,Z=0),Scale3D=(X=1,Y=1,Z=1))))
    WeightMinimum=0
    Weight=&RigUnit_GetCurveValue(Curve=CR Thigh R Across L).Value
   )
   > Modify Transforms with Mirror_2 = RigVMFunctionReferenceNode(
    Base Modify Data="((Item=(Type=Bone,Name=\"dyn_front_panel_l_01\"),Transform=(Rotation=(X=0,Y=0,Z=0,W=1),Translation=(X=0,Y=-3,Z=0),Scale3D=(X=1,Y=1,Z=1))),(Item=(Type=Bone,Name=\"dyn_front_panel_l_02\"),Transform=(Rotation=(X=0,Y=0,Z=0,W=1),Translation=(X=0,Y=-2.5,Z=0),Scale3D ...(truncated)"
    Mirror Modify Data="((Item=(Type=Bone,Name=\"dyn_front_panel_r_01\"),Transform=(Rotation=(X=0,Y=0,Z=0,W=1),Translation=(X=0,Y=0,Z=0),Scale3D=(X=1,Y=1,Z=1))),(Item=(Type=Bone,Name=\"dyn_front_panel_r_02\"),Transform=(Rotation=(X=0,Y=0,Z=0,W=1),Translation=(X=0,Y=0,Z=0),Scale3D=(X= ...(truncated)"
    Weight=&RigUnit_GetCurveValue(Curve=CR Thigh Raised Front L).Value
    "Mirror Weight "=&RigUnit_GetCurveValue(Curve="CR ThighRaised  Front R").Value
   )
   RigVMFunction_ControlFlowBranch = &RigVMFunction_ControlFlowBranch(Condition=False, ExecuteContext=@"Modify Transforms with Mirror_2".ExecuteContext)
   Remap = &RigVMFunction_MathDoubleRemap(SourceMaximum=770, SourceMinimum=0, TargetMaximum=1, TargetMinimum=0, Value=0)
   n1 = &RigUnit_ModifyTransforms(
    ItemToModify=((Item=(Type=Bone,Name="dyn_front_panel_MSTR"),Transform=(Rotation=(X=0,Y=0,Z=-0.0436,W=0.999),Translation=(X=0,Y=-10,Z=0),Scale3D=(X=1,Y=1,Z=1))))
    WeightMinimum=0
    Weight=@Remap.Result
    ExecutePin=@RigVMFunction_ControlFlowBranch.False
   )
   n2 = &RigVMFunction_ControlFlowBranch(Condition=False, ExecuteContext=@n1.ExecutePin)
   Absolute = &RigVMFunction_MathDoubleAbs(Value=0)
   &RigUnit_ModifyTransforms(
    ItemToModify=((Item=(Type=Bone,Name="dyn_front_panel_MSTR"),Transform=(Rotation=(X=0,Y=0,Z=-0.3827,W=0.9239),Translation=(X=0,Y=0,Z=0),Scale3D=(X=1,Y=1,Z=1))))
    WeightMinimum=0
    Weight=&RigVMFunction_MathDoubleRemap(SourceMaximum=7, SourceMinimum=0, TargetMaximum=1, TargetMinimum=0, Value=@Absolute.Result).Result
    ExecutePin=@n2.False
   )
   &RigUnit_ModifyTransforms(
    ItemToModify=((Item=(Type=Bone,Name="dyn_front_panel_MSTR"),Transform=(Rotation=(X=0,Y=0,Z=-0.0175,W=0.9998),Translation=(X=0,Y=-0.5,Z=0),Scale3D=(X=1,Y=1,Z=1))))
    Weight=0.5
    WeightMinimum=0
    ExecutePin=@RigVMFunction_Sequence.A
   )
   > Modify Transforms with Mirror_1 = RigVMFunctionReferenceNode(
    Base Modify Data="((Item=(Type=Bone,Name=\"dyn_front_panel_l_01\"),Transform=(Rotation=(X=-0.0436,Y=0,Z=0,W=0.999),Translation=(X=0,Y=-0.5,Z=0),Scale3D=(X=1,Y=1,Z=1))),(Item=(Type=Bone,Name=\"dyn_front_panel_l_02\"),Transform=(Rotation=(X=-0.0436,Y=0,Z=0,W=0.999),Translation=(X ...(truncated)"
    Mirror Modify Data="((Item=(Type=Bone,Name=\"dyn_front_panel_r_01\"),Transform=(Rotation=(X=0,Y=0,Z=0,W=1),Translation=(X=0,Y=0,Z=0),Scale3D=(X=1,Y=1,Z=1))),(Item=(Type=Bone,Name=\"dyn_front_panel_r_02\"),Transform=(Rotation=(X=0,Y=0,Z=0,W=1),Translation=(X=0,Y=0,Z=0),Scale3D=(X= ...(truncated)"
    "Mirror Weight "=1
    Weight=1
   )
  ]}
 RigVMFunctionLibrary {nodes [
   Constraint Front Panel = RigVMCollapseNode(ExecuteContext=())
   Ctrain Mid Sim Chains = RigVMCollapseNode(ExecuteContext=())
   Deform Post Sim Skirt Curves = RigVMCollapseNode(ExecuteContext=())
   Pose Front Panel = RigVMCollapseNode(ExecuteContext=())
  ]}
 RigVMFunction_Sequence_1_ContainedGraph {nodes [
   n1 = RigVMFunctionEntryNode()
   RigVMFunction_Sequence = &RigVMFunction_Sequence(ExecuteContext=@n1.ExecuteContext)
   n2 = &RigVMFunction_Sequence(ExecuteContext=@RigVMFunction_Sequence.B)
   n3 = &RigVMFunction_Sequence(ExecuteContext=@n2.B)
   n4 = &RigVMFunction_Sequence(ExecuteContext=@n3.B)
   RigVMFunctionReturnNode(A=@RigVMFunction_Sequence.A, B=@n2.A, C=@n3.A, D=@n4.A, E=@n4.B)
  ]}
 RigVMFunction_Sequence_1_ContainedGraph {nodes [
   n1 = RigVMFunctionEntryNode()
   RigVMFunction_Sequence = &RigVMFunction_Sequence(ExecuteContext=@n1.ExecuteContext)
   n2 = &RigVMFunction_Sequence(ExecuteContext=@RigVMFunction_Sequence.B)
   n3 = &RigVMFunction_Sequence(ExecuteContext=@n2.B)
   RigVMFunctionReturnNode(A=@RigVMFunction_Sequence.A, B=@n2.A, C=@n3.A, D=@n3.B)
  ]}
 RigVMFunction_Sequence_1_ContainedGraph {nodes [
   n1 = RigVMFunctionEntryNode()
   RigVMFunction_Sequence = &RigVMFunction_Sequence(ExecuteContext=@n1.ExecuteContext)
   n2 = &RigVMFunction_Sequence(ExecuteContext=@RigVMFunction_Sequence.B)
   n3 = &RigVMFunction_Sequence(ExecuteContext=@n2.B)
   RigVMFunctionReturnNode(A=@RigVMFunction_Sequence.A, B=@n2.A, C=@n3.A, D=@n3.B)
  ]}
 RigVMFunction_Sequence_1_ContainedGraph {nodes [
   n1 = RigVMFunctionEntryNode()
   RigVMFunction_Sequence = &RigVMFunction_Sequence(ExecuteContext=@n1.ExecuteContext)
   n2 = &RigVMFunction_Sequence(ExecuteContext=@RigVMFunction_Sequence.B)
   RigVMFunctionReturnNode(A=@RigVMFunction_Sequence.A, B=@n2.A, C=@n2.B)
  ]}
 RigVMModel {nodes [
   &RigUnit_ParentConstraint(
    AdvancedSettings=(InterpolationType=Average,RotationOrderForFilter=XZY)
    Child=(Type=Bone,Name="dyn_C_skirt_a_1")
    ExecutePin=()
    Filter=(TranslationFilter=(bX=True,bY=True,bZ=True),RotationFilter=(bX=true,bY=true,bZ=true),ScaleFilter=(bX=True,bY=True,bZ=True))
    Parents=((Item=(Type=Bone,Name="dyn_L_skirt_a_1"),Weight=0.5),(Item=(Type=Bone,Name="dyn_R_skirt_a_1"),Weight=0.5))
    Weight=1
   )
   > &RigUnit_ParentConstraint(
    AdvancedSettings=(InterpolationType=Average,RotationOrderForFilter=XZY)
    Child=(Type=Bone,Name="dyn_C_skirt_a_2")
    Filter=(TranslationFilter=(bX=True,bY=True,bZ=True),RotationFilter=(bX=true,bY=true,bZ=true),ScaleFilter=(bX=True,bY=True,bZ=True))
    Parents=((Item=(Type=Bone,Name="dyn_L_skirt_a_2"),Weight=0.5),(Item=(Type=Bone,Name="dyn_R_skirt_a_2"),Weight=0.5))
    Weight=1
   )
   &RigUnit_ParentConstraint(
    AdvancedSettings=(InterpolationType=Average,RotationOrderForFilter=XZY)
    Child=(Type=Bone,Name="dyn_L_skirt_a_2")
    ExecutePin=()
    Filter=(TranslationFilter=(bX=True,bY=True,bZ=True),RotationFilter=(bX=true,bY=true,bZ=true),ScaleFilter=(bX=True,bY=True,bZ=True))
    Parents=((Item=(Type=Bone,Name="thigh_l"),Weight=1))
    Weight=1
   )
   > &RigUnit_ParentConstraint(
    AdvancedSettings=(InterpolationType=Average,RotationOrderForFilter=XZY)
    Child=(Type=Bone,Name="dyn_L_skirt_b_2")
    Filter=(TranslationFilter=(bX=True,bY=True,bZ=True),RotationFilter=(bX=true,bY=true,bZ=true),ScaleFilter=(bX=True,bY=True,bZ=True))
    Parents=((Item=(Type=Bone,Name="thigh_l"),Weight=1))
    Weight=1
   )
   > &RigUnit_ParentConstraint(
    AdvancedSettings=(InterpolationType=Average,RotationOrderForFilter=XZY)
    Child=(Type=Bone,Name="dyn_L_skirt_c_2")
    Filter=(TranslationFilter=(bX=True,bY=True,bZ=True),RotationFilter=(bX=true,bY=true,bZ=true),ScaleFilter=(bX=True,bY=True,bZ=True))
    Parents=((Item=(Type=Bone,Name="thigh_l"),Weight=1))
    Weight=1
   )
   RigUnit_BeginExecution = &RigUnit_BeginExecution()
   RigVMFunction_ControlFlowBranch = &RigVMFunction_ControlFlowBranch(
    Condition=RigVMDispatchNode(B=1, A=RigVMVariableNode() {variable=IsCustomLobbyIdle}.Value) {template="DISPATCH_RigVMDispatch_CoreEquals(in A,in B,out Result)"}.Result
    ExecuteContext=@RigUnit_BeginExecution.ExecutePin
   )
   RigVMFunction_Sequence = RigVMAggregateNode(ExecuteContext=@RigVMFunction_ControlFlowBranch.Completed)
   Modify Transforms with Mirror_1 = RigVMFunctionReferenceNode(
    Base Modify Data="((Item=(Type=Bone,Name=\"dyn_L_skirt_a_1\"),Transform=(Rotation=(X=0,Y=0,Z=0,W=1),Translation=(X=0,Y=-1,Z=1),Scale3D=(X=1,Y=1,Z=1))),(Item=(Type=Bone,Name=\"dyn_L_skirt_b_1\"),Transform=(Rotation=(X=0.0015,Y=-0.0349,Z=0.0436,W=0.9984),Translation=(X=0,Y=0,Z=1) ...(truncated)"
    Mirror Modify Data="((Item=(Type=Bone,Name=\"dyn_R_skirt_a_1\"),Transform=(Rotation=(X=0,Y=0,Z=0,W=1),Translation=(X=0,Y=-1,Z=2),Scale3D=(X=1,Y=1,Z=1))),(Item=(Type=Bone,Name=\"dyn_R_skirt_b_1\"),Transform=(Rotation=(X=0,Y=0,Z=0,W=1),Translation=(X=0,Y=0,Z=2),Scale3D=(X=1,Y=1,Z=1 ...(truncated)"
    "Mirror Weight "=1
    ExecuteContext=@RigVMFunction_Sequence.A
    Weight=RigVMVariableNode() {variable=RobeOffsetWeight}.Value
   )
   > Deform Post Sim Skirt Curves = RigVMFunctionReferenceNode()
   > Ctrain Mid Sim Chains = RigVMFunctionReferenceNode()
   &RigUnit_ModifyTransforms(
    ItemToModify="((Item=(Type=Bone,Name=\"dyn_C_skirt_a_1\"),Transform=(Rotation=(X=0.0239,Y=-0.1775,Z=-0.1314,W=0.975),Translation=(X=0,Y=0,Z=0),Scale3D=(X=1,Y=1,Z=1))),(Item=(Type=Bone,Name=\"dyn_C_skirt_b_1\"),Transform=(Rotation=(X=0.006,Y=0.0519,Z=0.1148,W=-0.992),Transla ...(truncated)"
    WeightMinimum=0
    ExecutePin=@RigVMFunction_Sequence.C
    Weight=RigVMVariableNode() {variable=Pirouette}.Value
   )
   n1 = &RigVMFunction_ControlFlowBranch(Condition=RigVMVariableNode() {variable=Pub DBNO}.Value, ExecuteContext=@RigVMFunction_Sequence.B)
   n2 = &RigVMFunction_ControlFlowBranch(ExecuteContext=@n1.False, Condition=RigVMVariableNode() {variable=Pub Is Crouching}.Value)
   n3 = &RigUnit_SphericalPoseReader(
    ActiveRegionScaleFactors=(PositiveWidth=3,NegativeWidth=1,PositiveHeight=1.2,NegativeHeight=1)
    ActiveRegionSize=0.1
    Debug=(bDrawDebug=false,bDraw2D=False,bDrawLocalAxes=False,DebugScale=25,DebugSegments=20,DebugThickness=0.25)
    DriverAxis=(X=-1,Y=0,Z=0)
    DriverItem=(Type=Bone,Name="thigh_l")
    FalloffRegionScaleFactors=(PositiveWidth=1,NegativeWidth=1,PositiveHeight=0.75,NegativeHeight=0.75)
    FalloffSize=0.58
    FlipHeightScaling=False
    FlipWidthScaling=False
    OptionalParentItem=(Type=Bone,Name="None")
    RotationOffset=(X=90,Y=0,Z=-15)
    ExecutePin=@n2.False
   )
   > &RigUnit_ModifyTransforms(
    ItemToModify="((Item=(Type=Bone,Name=\"base_C_skirt_a\"),Transform=(Rotation=(X=0.0151,Y=-0.173,Z=0.0858,W=0.9811),Translation=(X=5,Y=0,Z=0),Scale3D=(X=1,Y=1,Z=1))),(Item=(Type=Bone,Name=\"base_L_skirt_a\"),Transform=(Rotation=(X=-0.0076,Y=-0.1735,Z=-0.043,W=0.9839),Transla ...(truncated)"
    WeightMinimum=0
    Weight=@n3.OutputParam
   )
   > n4 = &RigUnit_SphericalPoseReader(
    ActiveRegionScaleFactors=(PositiveWidth=3,NegativeWidth=1,PositiveHeight=1.2,NegativeHeight=1)
    ActiveRegionSize=0.1
    Debug=(bDrawDebug=false,bDraw2D=False,bDrawLocalAxes=False,DebugScale=25,DebugSegments=20,DebugThickness=0.25)
    DriverAxis=(X=1,Y=0,Z=0)
    DriverItem=(Type=Bone,Name="thigh_r")
    FalloffRegionScaleFactors=(PositiveWidth=1,NegativeWidth=1,PositiveHeight=0.75,NegativeHeight=0.75)
    FalloffSize=0.58
    FlipHeightScaling=true
    FlipWidthScaling=true
    OptionalParentItem=(Type=Bone,Name="None")
    RotationOffset=(X=-90,Y=0,Z=-15)
   )
   > &RigUnit_ModifyTransforms(
    ItemToModify="((Item=(Type=Bone,Name=\"base_C_skirt_a\"),Transform=(Rotation=(X=0,Y=-0.1736,Z=0,W=0.9848),Translation=(X=0,Y=0,Z=0),Scale3D=(X=1,Y=1,Z=1))),(Item=(Type=Bone,Name=\"base_R_skirt_a\"),Transform=(Rotation=(X=0,Y=0,Z=0,W=1),Translation=(X=0,Y=0,Z=0),Scale3D=(X=1 ...(truncated)"
    WeightMinimum=0
    Weight=@n4.OutputParam
   )
   > n5 = &RigUnit_SphericalPoseReader(
    ActiveRegionScaleFactors=(PositiveWidth=3,NegativeWidth=0.5,PositiveHeight=5,NegativeHeight=5)
    ActiveRegionSize=0.1
    Debug=(bDrawDebug=false,bDraw2D=False,bDrawLocalAxes=False,DebugScale=25,DebugSegments=20,DebugThickness=0.25)
    DriverAxis=(X=-1,Y=0,Z=0)
    DriverItem=(Type=Bone,Name="calf_l")
    FalloffRegionScaleFactors=(PositiveWidth=0.2,NegativeWidth=0.92,PositiveHeight=1,NegativeHeight=1)
    FalloffSize=0.7
    FlipHeightScaling=False
    FlipWidthScaling=False
    OptionalParentItem=(Type=Bone,Name="thigh_l")
    RotationOffset=(X=-90,Y=0,Z=27)
   )
   > &RigUnit_ModifyTransforms(
    ItemToModify="((Item=(Type=Bone,Name=\"dyn_L_skirt_b_2\"),Transform=(Rotation=(X=0,Y=0,Z=-0.0872,W=0.9962),Translation=(X=0,Y=0,Z=0),Scale3D=(X=1,Y=1,Z=1))),(Item=(Type=Bone,Name=\"base_L_skirt_b\"),Transform=(Rotation=(X=-0.0151,Y=-0.0858,Z=-0.173,W=0.9811),Translation=(X= ...(truncated)"
    WeightMinimum=0
    Weight=@n5.OutputParam
   )
   > n6 = &RigUnit_SphericalPoseReader(
    ActiveRegionScaleFactors=(PositiveWidth=3,NegativeWidth=0.5,PositiveHeight=5,NegativeHeight=5)
    ActiveRegionSize=0.1
    Debug=(bDrawDebug=false,bDraw2D=False,bDrawLocalAxes=False,DebugScale=25,DebugSegments=20,DebugThickness=0.25)
    DriverAxis=(X=1,Y=0,Z=0)
    DriverItem=(Type=Bone,Name="calf_r")
    FalloffRegionScaleFactors=(PositiveWidth=0.2,NegativeWidth=0.92,PositiveHeight=1,NegativeHeight=1)
    FalloffSize=0.7
    FlipHeightScaling=true
    FlipWidthScaling=true
    OptionalParentItem=(Type=Bone,Name="thigh_r")
    RotationOffset=(X=90,Y=0,Z=27)
   )
   > &RigUnit_ModifyTransforms(
    ItemToModify="((Item=(Type=Bone,Name=\"dyn_R_skirt_b_2\"),Transform=(Rotation=(X=0,Y=0,Z=0.0872,W=0.9962),Translation=(X=0,Y=0,Z=0),Scale3D=(X=1,Y=1,Z=1))),(Item=(Type=Bone,Name=\"base_R_skirt_b\"),Transform=(Rotation=(X=0.0151,Y=-0.0858,Z=0.173,W=0.9811),Translation=(X=0,Y ...(truncated)"
    WeightMinimum=0
    Weight=@n6.OutputParam
   )
   n7 = &RigUnit_SphericalPoseReader(
    ActiveRegionScaleFactors=(PositiveWidth=3,NegativeWidth=1,PositiveHeight=1.2,NegativeHeight=1)
    ActiveRegionSize=0.1
    Debug=(bDrawDebug=false,bDraw2D=False,bDrawLocalAxes=False,DebugScale=25,DebugSegments=20,DebugThickness=0.25)
    DriverAxis=(X=-1,Y=0,Z=0)
    DriverItem=(Type=Bone,Name="thigh_l")
    FalloffRegionScaleFactors=(PositiveWidth=1,NegativeWidth=1,PositiveHeight=0.75,NegativeHeight=0.75)
    FalloffSize=0.58
    FlipHeightScaling=False
    FlipWidthScaling=False
    OptionalParentItem=(Type=Bone,Name="None")
    RotationOffset=(X=90,Y=0,Z=-15)
    ExecutePin=@n2.True
   )
   > &RigUnit_ModifyTransforms(
    ItemToModify="((Item=(Type=Bone,Name=\"base_C_skirt_a\"),Transform=(Rotation=(X=-0.0151,Y=0.173,Z=0.0858,W=0.9811),Translation=(X=0,Y=10,Z=0),Scale3D=(X=1,Y=1,Z=1))),(Item=(Type=Bone,Name=\"base_L_skirt_a\"),Transform=(Rotation=(X=0.0038,Y=0.0871,Z=-0.0435,W=0.9952),Transla ...(truncated)"
    WeightMinimum=0
    Weight=@n7.OutputParam
   )
   > n8 = &RigUnit_SphericalPoseReader(
    ActiveRegionScaleFactors=(PositiveWidth=3,NegativeWidth=1,PositiveHeight=1.2,NegativeHeight=1)
    ActiveRegionSize=0.1
    Debug=(bDrawDebug=false,bDraw2D=False,bDrawLocalAxes=False,DebugScale=25,DebugSegments=20,DebugThickness=0.25)
    DriverAxis=(X=1,Y=0,Z=0)
    DriverItem=(Type=Bone,Name="thigh_r")
    FalloffRegionScaleFactors=(PositiveWidth=1,NegativeWidth=1,PositiveHeight=0.75,NegativeHeight=0.75)
    FalloffSize=0.58
    FlipHeightScaling=true
    FlipWidthScaling=true
    OptionalParentItem=(Type=Bone,Name="None")
    RotationOffset=(X=-90,Y=0,Z=-15)
   )
   > &RigUnit_ModifyTransforms(
    ItemToModify="((Item=(Type=Bone,Name=\"base_C_skirt_a\"),Transform=(Rotation=(X=-0.0302,Y=-0.171,Z=-0.171,W=0.9698),Translation=(X=0,Y=0,Z=0),Scale3D=(X=1,Y=1,Z=1))),(Item=(Type=Bone,Name=\"base_C_skirt_b\"),Transform=(Rotation=(X=0,Y=0,Z=-0.0436,W=0.999),Translation=(X=0,Y ...(truncated)"
    WeightMinimum=0
    Weight=@n8.OutputParam
   )
   > n9 = &RigUnit_SphericalPoseReader(
    ActiveRegionScaleFactors=(PositiveWidth=3,NegativeWidth=0.5,PositiveHeight=5,NegativeHeight=5)
    ActiveRegionSize=0.1
    Debug=(bDrawDebug=false,bDraw2D=False,bDrawLocalAxes=False,DebugScale=25,DebugSegments=20,DebugThickness=0.25)
    DriverAxis=(X=-1,Y=0,Z=0)
    DriverItem=(Type=Bone,Name="calf_l")
    FalloffRegionScaleFactors=(PositiveWidth=0.2,NegativeWidth=0.92,PositiveHeight=1,NegativeHeight=1)
    FalloffSize=0.7
    FlipHeightScaling=False
    FlipWidthScaling=False
    OptionalParentItem=(Type=Bone,Name="thigh_l")
    RotationOffset=(X=-90,Y=0,Z=27)
   )
   > &RigUnit_ModifyTransforms(
    ItemToModify="((Item=(Type=Bone,Name=\"dyn_L_skirt_b_2\"),Transform=(Rotation=(X=0,Y=0,Z=-0.0872,W=0.9962),Translation=(X=0,Y=0,Z=0),Scale3D=(X=1,Y=1,Z=1))),(Item=(Type=Bone,Name=\"base_L_skirt_b\"),Transform=(Rotation=(X=-0.0151,Y=-0.0858,Z=-0.173,W=0.9811),Translation=(X= ...(truncated)"
    WeightMinimum=0
    Weight=@n9.OutputParam
   )
   > n10 = &RigUnit_SphericalPoseReader(
    ActiveRegionScaleFactors=(PositiveWidth=3,NegativeWidth=0.5,PositiveHeight=5,NegativeHeight=5)
    ActiveRegionSize=0.1
    Debug=(bDrawDebug=false,bDraw2D=False,bDrawLocalAxes=False,DebugScale=25,DebugSegments=20,DebugThickness=0.25)
    DriverAxis=(X=1,Y=0,Z=0)
    DriverItem=(Type=Bone,Name="calf_r")
    FalloffRegionScaleFactors=(PositiveWidth=0.2,NegativeWidth=0.92,PositiveHeight=1,NegativeHeight=1)
    FalloffSize=0.7
    FlipHeightScaling=true
    FlipWidthScaling=true
    OptionalParentItem=(Type=Bone,Name="thigh_r")
    RotationOffset=(X=90,Y=0,Z=27)
   )
   > n11 = &RigUnit_ModifyTransforms(
    ItemToModify="((Item=(Type=Bone,Name=\"dyn_R_skirt_b_2\"),Transform=(Rotation=(X=0,Y=0,Z=0.0872,W=0.9962),Translation=(X=0,Y=0,Z=0),Scale3D=(X=1,Y=1,Z=1))),(Item=(Type=Bone,Name=\"base_R_skirt_b\"),Transform=(Rotation=(X=0.0151,Y=-0.0858,Z=0.173,W=0.9811),Translation=(X=0,Y ...(truncated)"
    WeightMinimum=0
    Weight=@n10.OutputParam
   )
   n12 = &RigVMFunction_ControlFlowBranch(ExecuteContext=@n11.ExecutePin, Condition=RigVMVariableNode() {variable="Is Baseball Sliding "}.Value)
   n13 = &RigVMFunction_ControlFlowBranch(ExecuteContext=@n12.Completed, Condition=RigVMVariableNode() {variable="Is Knee Sliding "}.Value)
   &RigUnit_ModifyTransforms(
    ItemToModify=((Item=(Type=Bone,Name="dyn_L_skirt_c_2"),Transform=(Rotation=(X=0,Y=0,Z=0,W=1),Translation=(X=0,Y=0,Z=5.9424),Scale3D=(X=1,Y=1,Z=1))))
    Weight=1
    WeightMinimum=0
    ExecutePin=@n13.True
   )
   &RigUnit_ModifyTransforms(
    ItemToModify="((Item=(Type=Bone,Name=\"dyn_R_skirt_b_2\"),Transform=(Rotation=(X=-0.1572,Y=-0.0054,Z=0.3965,W=-0.9045),Translation=(X=-9.6441,Y=-5.6604,Z=-5.9132),Scale3D=(X=1,Y=1,Z=1))),(Item=(Type=Bone,Name=\"dyn_C_skirt_a_1\"),Transform=(Rotation=(X=0,Y=0,Z=0,W=-1),Trans ...(truncated)"
    Weight=1
    WeightMinimum=0
    ExecutePin=@n12.True
   )
   n14 = &RigUnit_ModifyTransforms(
    ItemToModify="((Item=(Type=Bone,Name=\"base_L_skirt_a\"),Transform=(Rotation=(X=-0.2152,Y=-0.1507,Z=0.3463,W=-0.9006),Translation=(X=0,Y=0,Z=0),Scale3D=(X=1,Y=1,Z=1))),(Item=(Type=Bone,Name=\"base_C_skirt_a\"),Transform=(Rotation=(X=0.0018,Y=-0.0197,Z=0.0903,W=0.9957),Trans ...(truncated)"
    Weight=1
    WeightMinimum=0
    ExecutePin=@n1.True
   )
   n15 = &RigVMFunction_ControlFlowBranch(ExecuteContext=@n14.ExecutePin, Condition=RigVMVariableNode() {variable=Moving 2d}.Value)
   &RigUnit_ModifyTransforms(
    ItemToModify="((Item=(Type=Bone,Name=\"base_L_skirt_a\"),Transform=(Rotation=(X=0,Y=0,Z=0,W=1),Translation=(X=0,Y=0,Z=0),Scale3D=(X=1,Y=1,Z=1))),(Item=(Type=Bone,Name=\"base_C_skirt_a\"),Transform=(Rotation=(X=0,Y=0,Z=0,W=1),Translation=(X=0,Y=0,Z=0),Scale3D=(X=1,Y=1,Z=1))) ...(truncated)"
    Weight=1
    WeightMinimum=0
    ExecutePin=@n15.True
   )
   RigVMVariableNode(Value=1, ExecuteContext=@RigVMFunction_ControlFlowBranch.False) {variable=RobeOffsetWeight}
   RigVMVariableNode(Value=0.5, ExecuteContext=@RigVMFunction_ControlFlowBranch.True) {variable=RobeOffsetWeight}
   n16 = &RigVMFunction_ControlFlowBranch(ExecuteContext=(), Condition=RigVMVariableNode() {variable=Skydive Floating}.Value)
   &RigUnit_ParentConstraint(
    AdvancedSettings=(InterpolationType=Average,RotationOrderForFilter=XZY)
    Child=(Type=Bone,Name="dyn_L_skirt_a_4")
    Filter=(TranslationFilter=(bX=True,bY=True,bZ=True),RotationFilter=(bX=true,bY=true,bZ=true),ScaleFilter=(bX=True,bY=True,bZ=True))
    Parents=((Item=(Type=Bone,Name="thigh_l"),Weight=1))
    Weight=1
    ExecutePin=@n16.Completed
   )
   > &RigUnit_ParentConstraint(
    AdvancedSettings=(InterpolationType=Average,RotationOrderForFilter=XZY)
    Child=(Type=Bone,Name="dyn_L_skirt_b_4")
    Filter=(TranslationFilter=(bX=True,bY=True,bZ=True),RotationFilter=(bX=true,bY=true,bZ=true),ScaleFilter=(bX=True,bY=True,bZ=True))
    Parents=((Item=(Type=Bone,Name="thigh_l"),Weight=1))
    Weight=1
   )
   > &RigUnit_ParentConstraint(
    AdvancedSettings=(InterpolationType=Average,RotationOrderForFilter=XZY)
    Child=(Type=Bone,Name="dyn_L_skirt_c_4")
    Filter=(TranslationFilter=(bX=True,bY=True,bZ=True),RotationFilter=(bX=true,bY=true,bZ=true),ScaleFilter=(bX=True,bY=True,bZ=True))
    Parents=((Item=(Type=Bone,Name="thigh_l"),Weight=1))
    Weight=1
   )
   setFKCtrl = RigVMFunctionReferenceNode(ExecuteContext=(), setFK_Ctrl=false)
   > FN_DeformRig_FNC = RigVMFunctionReferenceNode(
    DeformRig_ Curve="(deformRig_l_upperarm_up_alpha_72_#1=False,deformRig_r_upperarm_up_alpha_73_#2=False,deformRig_l_upperarm_fr_alpha_74_#3=False,deformRig_r_upperarm_fr_alpha_75_CB51B8 ...(truncated)"
    DeformRig_Bool="(upperarm_1_#4=True,elbow_3_#5=True,wrist_7_#6=True,groin_8_#7=True,glute_17_#8=True,knee_9_40C408324CA261 ...(truncated)"
    DeformRig_Debug=(upperarm_debug_18_#4=False,wrist_debug_20_#6=False,lat_debug_26_#9=False,groinGlute_debug_27_#7=False)
    DeformRig_Value="(upperarm_up_18_#10=(Rotation=(X=0,Y=0,Z=0,W=1),Translation=(X=0,Y=0,Z=8),Scale3D=(X=0.7,Y=1,Z=1)),upperarm_fwd_fr_19_#11=(Rotation=(X=0,Y=0,Z=0,W=1),Translation=(X=0,Y=10,Z=0),Scale3D=(X=0.5,Y=1,Z= ...(truncated)"
   )
  ]}
 "RigVMModel Construction Solve " {nodes [
   n1 = &RigUnit_PrepareForExecution()
   &RigUnit_HierarchyImportFromSkeleton(NameSpace=None, bIncludeCurves=False, bIncludeMeshSockets=False, bIncludeVirtualBones=True, ExecutePin=@n1.ExecutePin)
   RigUnit_GetTransform = &RigUnit_GetTransform(Item=(Type=Bone,Name="thigh_l"), Space=LocalSpace, bInitial=False)
   > RigVMVariableNode(Value=@RigUnit_GetTransform.Transform.Rotation) {variable=prev thigh l quatrnion}
   > RigVMVariableNode(Value=&RigUnit_GetTransform(Item=(Type=Bone,Name="thigh_r"), Space=LocalSpace, bInitial=False).Transform.Rotation) {variable=prev thigh r quatrnion}
  ]}
]
hierarchy [
 "CR  Rear Knee Bend R" {type=Curve, loc=<0 0 0>, rot=<0 0 0 1>}
 CR Foot Near Pelvis L {type=Curve, loc=<0 0 0>, rot=<0 0 0 1>}
 CR Foot Near Pelvis R {type=Curve, loc=<0 0 0>, rot=<0 0 0 1>}
 CR Thigh Raised Front L {type=Curve, loc=<0 0 0>, rot=<0 0 0 1>}
 "CR ThighRaised  Front R" {type=Curve, loc=<0 0 0>, rot=<0 0 0 1>}
 CR Rear Knee Bend L {type=Curve, loc=<0 0 0>, rot=<0 0 0 1>}
 CR Thigh L Across R {type=Curve, loc=<0 0 0>, rot=<0 0 0 1>}
]
variables [
 Clavicle Rot_L {type=FVector, default=(X=0,Y=0,Z=0)}
 Clavicle Rot_R {type=FVector, default=(X=0,Y=0,Z=0)}
 DNBO {type=bool, default=False}
 Front Skirt L {
  type="TArray<FRigElementKey>"
  default=((Type=Bone,Name="dyn_L_skirt_a_1"),(Type=Bone,Name="dyn_L_skirt_a_2"),(Type=Bone,Name="dyn_L_skirt_a_3"),(Type=Bone,Name="dyn_L_skirt_a_4"),(Type=Bone,Name="dyn_L_skirt_a_5"),(Type=Bone,Name="dyn_L_skirt_a_6"))
 }
 Front Skirt Mid {
  type="TArray<FRigElementKey>"
  default=((Type=Bone,Name="dyn_C_skirt_a_1"),(Type=Bone,Name="dyn_C_skirt_a_2"),(Type=Bone,Name="dyn_C_skirt_a_3"),(Type=Bone,Name="dyn_C_skirt_a_4"),(Type=Bone,Name="dyn_C_skirt_a_5"),(Type=Bone,Name="dyn_C_skirt_a_6"))
 }
 Front Skirt R {
  type="TArray<FRigElementKey>"
  default=((Type=Bone,Name="dyn_R_skirt_a_1"),(Type=Bone,Name="dyn_R_skirt_a_2"),(Type=Bone,Name="dyn_R_skirt_a_3"),(Type=Bone,Name="dyn_R_skirt_a_4"),(Type=Bone,Name="dyn_R_skirt_a_5"),(Type=Bone,Name="dyn_R_skirt_a_6"))
 }
 Head Back C {type=double, default=0}
 "Head Side L " {type=double, default=0}
 Head Side R {type=double, default=0}
 "Is Baseball Sliding " {type=bool, default=False}
 Is Jumping {type=bool, default=False}
 "Is Knee Sliding " {type=bool, default=False}
 Is On Ground {type=bool, default=False}
 Is crossover {type=double, default=0}
 IsAccelerating2D {type=bool, default=False}
 IsCustomLobbyIdle {type=double, default=0}
 L thighBackward alpha {type=double, default=0}
 L thighForward alpha {type=double, default=0}
 Leg ROM_Result Alpha {type=double, default=0}
 Moving 2d {type=bool, default=False}
 Pirouette {type=double, default=0}
 Player Speed {type=double, default=0}
 Pub DBNO {type=bool, default=False}
 Pub Disablecontrol Rig {type=bool, default=False}
 Pub Is Crouching {type=bool, default=False}
 Pub Is Front End {type=bool, default=False}
 Pub Is Playing Eastern Bloc {type=bool, default=False}
 Pub Is Playing Emote {type=bool, default=False}
 Pub is Baseball Sliding {type=double, default=0}
 Pub is Crouch Strafe Left {type=double, default=0}
 Pub is Crouch Strafe Right {type=double, default=0}
 Pub is Grind Rail {type=double, default=0}
 "Pub is Knee Sliding " {type=double, default=0}
 Pub is Zipline {type=double, default=0}
 R thighBackward alpha {type=double, default=0}
 R thighForward alpha {type=double, default=0}
 Rear Skirt L {
  type="TArray<FRigElementKey>"
  default=((Type=Bone,Name="dyn_L_skirt_c_1"),(Type=Bone,Name="dyn_L_skirt_c_2"),(Type=Bone,Name="dyn_L_skirt_c_3"),(Type=Bone,Name="dyn_L_skirt_c_4"),(Type=Bone,Name="dyn_L_skirt_c_5"),(Type=Bone,Name="dyn_L_skirt_c_6"))
 }
 Rear Skirt Mid {
  type="TArray<FRigElementKey>"
  default=((Type=Bone,Name="dyn_C_skirt_b_1"),(Type=Bone,Name="dyn_C_skirt_b_2"),(Type=Bone,Name="dyn_C_skirt_b_3"),(Type=Bone,Name="dyn_C_skirt_b_4"),(Type=Bone,Name="dyn_C_skirt_b_5"),(Type=Bone,Name="dyn_C_skirt_b_6"))
 }
 Rear Skirt R {
  type="TArray<FRigElementKey>"
  default=((Type=Bone,Name="dyn_R_skirt_c_1"),(Type=Bone,Name="dyn_R_skirt_c_2"),(Type=Bone,Name="dyn_R_skirt_c_3"),(Type=Bone,Name="dyn_R_skirt_c_4"),(Type=Bone,Name="dyn_R_skirt_c_5"),(Type=Bone,Name="dyn_R_skirt_c_6"))
 }
 "Right Leg " {type=double, default=0}
 RobeOffsetWeight {type=double, default=0}
 Skydive Floating {type=bool, default=False}
 l leg backwards SPR float {type=float, default=0}
 l thighLocal ry float {type=double, default=0}
 l thighLocal rz float {type=double, default=0}
 "left Leg " {type=double, default=0}
 prev thigh l quatrnion {type=FQuat, default=(X=0,Y=0,Z=0,W=1)}
 prev thigh r quatrnion {type=FQuat, default=(X=0,Y=0,Z=0,W=1)}
 r leg backwards SPR float {type=float, default=0}
 r thighLocal ry float {type=double, default=0}
 r thighLocal rz float {type=double, default=0}
]
