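# Physics asset for the face accessory mesh of the female medium FruitFire_NudgeHonk player cosmetic, setting up physics for its hair.
# Four capsule bodies cover the head and a three-bone hair chain, Hair_1 to Hair_3.
# Constraints link the chain in sequence from the head, with tighter swing limit values at the root than further along and almost no twist allowance.
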PhysicsAsset /BRCosmetics/Characters/Player/Female/Medium/Bodies/F_MED_FruitFire_NudgeHonk/Meshes/Parts/F_MED_FruitFire_NudgeHonk_FaceAcc_Physics
bodies {
 Hair_1 [{shapeType=Capsule, length=4.794, radius=20.1864, center=<9.8415 0 0>, rotation=<rot 2.6507 -90 92.8128>}]
 Hair_2 [{shapeType=Capsule, length=12.0922, radius=17.4694, center=<1.1565 0 0>, rotation=<rot 0.5413 -90 86.1833>}]
 Hair_3 [{shapeType=Capsule, length=5.9413, radius=14.0926, center=<-2.7674 0 0>, rotation=<rot 11.3896 89.2849 -87.3002>}]
 head [{shapeType=Capsule, length=16.13, radius=13.5288, center=<8.6648 -6.4343 -0.9931>, rotation=<rot -5.9719 -90 -3.4218>}]
}
constraints [
 {bone1Name=Hair_1, bone2Name=head, swing1LimitDegrees=10, swing2LimitDegrees=10, twistLimitDegrees=1}
 {bone1Name=Hair_2, bone2Name=Hair_1, swing1LimitDegrees=25, swing2LimitDegrees=25, twistLimitDegrees=0}
 {bone1Name=Hair_3, bone2Name=Hair_2, swing1LimitDegrees=25, swing2LimitDegrees=25, twistLimitDegrees=0}
]
