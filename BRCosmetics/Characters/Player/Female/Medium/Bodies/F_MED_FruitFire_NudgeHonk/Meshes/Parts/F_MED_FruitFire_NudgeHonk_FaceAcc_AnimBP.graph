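# Post-copy AnimBP for the FruitFire NudgeHonk female face accessory, adding hair physics to the pose copied from the attached parent mesh.
# Output blends on parachute state: open uses RigidBody plus pitch offsets on two hair bones, closed uses RigidBody weighted by Hair Alpha.
# Update zeroes that alpha when falling or jump-sprinting, and stores nativized gravity-override results.
# Two AnimDynamics chains remain unconnected to output.
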AnimBlueprint /BRCosmetics/Characters/Player/Female/Medium/Bodies/F_MED_FruitFire_NudgeHonk/Meshes/Parts/F_MED_FruitFire_NudgeHonk_FaceAcc_AnimBP
target_skeleton=&/F_MED_FruitFire_NudgeHonk_FaceAcc_Skeleton
graphs [
 AnimGraph {nodes [
   n1 = Anim Dynamics - Bone: Hair_1!(ComponentPose, Alpha=0.75, GravityOverride:Vector="0,0,0", ExternalForce="0,0,0") -> (Pose) {
    angularDampingOverride=0.99
    angularSpringConstant=2000
    bAngularSpring=true
    bOverrideAngularDamping=true
    bUseGravityOverride=true
    gravityScale=0.35
    simulationSpace=BoneRelative
    boundBone {boneName=Hair_1}
    componentAppliedLinearAccClamp=<0 0 0>
    physicsBodyDefinitions [{
      boundBone {boneName=Hair_1}
      boxExtents=<20 30 30>
      constraintSetup {angularLimitsMax=<0 21 3>, angularLimitsMin=<0 -21 -3>, angularTarget=<1 0 0>}
      localJointOffset=<8 0 0>
     }]
    relativeSpaceBone {boneName=head}
   }
   n2 = Anim Dynamics - Bone: Hair_2!(ComponentPose=@n1, Alpha=0.75, GravityOverride:Vector="0,0,0", ExternalForce="0,0,0") -> (Pose) {
    angularDampingOverride=1
    angularSpringConstant=2000
    bAngularSpring=true
    bOverrideAngularDamping=true
    bUseGravityOverride=true
    gravityScale=0.35
    simulationSpace=BoneRelative
    alphaScaleBiasClamp {bInterpResult=true, interpSpeedDecreasing=15}
    boundBone {boneName=Hair_2}
    componentAppliedLinearAccClamp=<0 0 0>
    physicsBodyDefinitions [{
      boundBone {boneName=Hair_2}
      boxExtents=<20 30 30>
      constraintSetup {angularLimitsMax=<4.2 12 9>, angularLimitsMin=<-4.2 -12 -9>, angularTarget=<1 0 0>}
      localJointOffset=<6 0 0>
     }]
    relativeSpaceBone {boneName=head}
   }
   Anim Dynamics - Bone: Hair_3!(ComponentPose=@n2, Alpha=0.75, GravityOverride:Vector="0,0,0", ExternalForce="0,0,0") -> (Pose) {
    angularDampingOverride=1
    angularSpringConstant=10000
    bAngularSpring=true
    bOverrideAngularDamping=true
    bUseGravityOverride=true
    gravityScale=0.35
    simulationSpace=BoneRelative
    alphaScaleBiasClamp {bInterpResult=true, interpSpeedDecreasing=15}
    boundBone {boneName=Hair_3}
    componentAppliedLinearAccClamp=<0 0 0>
    physicsBodyDefinitions [{
      boundBone {boneName=Hair_3}
      boxExtents=<20 30 30>
      constraintSetup {angularLimitsMax=<1.8 6 12>, angularLimitsMin=<-1.8 -6 -12>, angularTarget=<1 0 0>}
      localJointOffset=<4 0 0>
     }]
    relativeSpaceBone {boneName=head}
   }
   n3 = Anim Dynamics - Bone: Hair_1(Alpha=1, GravityOverride="0,0,0", ExternalForce="0,0,0") {
    angularDampingOverride=0.99
    angularSpringConstant=2000
    bAngularSpring=true
    bOverrideAngularDamping=true
    bUseGravityOverride=true
    gravityScale=0.35
    simulationSpace=BoneRelative
    boundBone {boneName=Hair_1}
    componentAppliedLinearAccClamp=<0 0 0>
    physicsBodyDefinitions [{
      boundBone {boneName=Hair_1}
      boxExtents=<20 30 30>
      constraintSetup {angularLimitsMax=<0 -10 5>, angularLimitsMin=<0 -45 -5>, angularTarget=<1 0 0>}
      localJointOffset=<8 0 0>
     }]
    relativeSpaceBone {boneName=head}
   }
   n4 = Anim Dynamics - Bone: Hair_2(ComponentPose=@n3, Alpha=1, GravityOverride="0,0,0", ExternalForce="0,0,0") {
    angularDampingOverride=1
    angularSpringConstant=2000
    bAngularSpring=true
    bOverrideAngularDamping=true
    bUseGravityOverride=true
    gravityScale=0.35
    simulationSpace=BoneRelative
    alphaScaleBiasClamp {bInterpResult=true, interpSpeedDecreasing=15}
    boundBone {boneName=Hair_2}
    componentAppliedLinearAccClamp=<0 0 0>
    physicsBodyDefinitions [{
      boundBone {boneName=Hair_2}
      boxExtents=<20 30 30>
      constraintSetup {angularLimitsMax=<7 -10 15>, angularLimitsMin=<-7 -35 -15>, angularTarget=<1 0 0>}
      localJointOffset=<6 0 0>
     }]
    relativeSpaceBone {boneName=head}
   }
   n5 = Anim Dynamics - Bone: Hair_3(ComponentPose=@n4, Alpha=1, GravityOverride="0,0,0", ExternalForce="0,0,0") {
    angularDampingOverride=1
    angularSpringConstant=10000
    bAngularSpring=true
    bOverrideAngularDamping=true
    bUseGravityOverride=true
    gravityScale=0.35
    simulationSpace=BoneRelative
    alphaScaleBiasClamp {bInterpResult=true, interpSpeedDecreasing=15}
    boundBone {boneName=Hair_3}
    componentAppliedLinearAccClamp=<0 0 0>
    physicsBodyDefinitions [{
      boundBone {boneName=Hair_3}
      boxExtents=<20 30 30>
      constraintSetup {angularLimitsMax=<3 -5 20>, angularLimitsMin=<-3 -15 -20>, angularTarget=<1 0 0>}
      localJointOffset=<4 0 0>
     }]
    relativeSpaceBone {boneName=head}
   }
   Component To Local(ComponentPose=@n5)
   n6 = RigidBody(ComponentPose=Local To Component(LocalPose=Copy Pose From Mesh() {bUseAttachedParent=true})) {
    bClampLinearTranslationLimitToRefPose=true
    bOverrideWorldGravity=true
    overridePhysicsAsset=&/F_MED_FruitFire_NudgeHonk_FaceAcc_Physics
    simulationSpace=BaseBoneSpace
    baseBoneRef {boneName=neck_02}
    overrideWorldGravity {z=-980}
    simSpaceSettings {dampingAlpha=0, maxAngularAcceleration=50, maxAngularVelocity=5, maxLinearAcceleration=50, maxLinearVelocity=80, worldAlpha=0.05}
   }
   n7 = "Transform (Modify) Bone - Bone: Hair_2"(ComponentPose=@n6, Alpha=1, Translation="0,0,0", Rotation="10,0,0", Scale="1,1,1") {
    rotationMode=BMM_Additive
    rotationSpace=BCS_BoneSpace
    translationMode=BMM_Additive
    translationSpace=BCS_BoneSpace
    boneToModify {boneName=Hair_2}
    rotation {pitch=10}
   }
   n8 = "Transform (Modify) Bone - Bone: Hair_1"(ComponentPose=@n7, Alpha=1, Translation="0,0,0", Rotation="10,0,0", Scale="1,1,1") {
    rotationMode=BMM_Additive
    rotationSpace=BCS_BoneSpace
    translationMode=BMM_Additive
    translationSpace=BCS_BoneSpace
    boneToModify {boneName=Hair_1}
    rotation {pitch=10}
   }
   n9 = RigidBody!(
    ComponentPose=Local To Component(LocalPose=Copy Pose From Mesh() {bUseAttachedParent=true})
    Alpha=Get Hair Alpha() -> (Hair Alpha:double)
    ExternalForce
    SimSpaceSettings:Sim Space Settings Structure=(WorldAlpha=0.05,VelocityScaleZ=1,DampingAlpha=0,MaxLinearVelocity=80,MaxAngularVelocity=5,MaxLinearAcceleration=50,MaxAngularAcceleration=50,ExternalLinearDragV=(X=0,Y=0,Z=0),ExternalLinearVelocity=(X=0,Y=0,Z=0),ExternalAngularVelocity=(X=0,Y=0,Z=0))
   ) -> (Pose) {
    bClampLinearTranslationLimitToRefPose=true
    bOverrideWorldGravity=true
    overridePhysicsAsset=&F_MED_FruitFire_NudgeHonk_FaceAcc_Physics
    simulationSpace=BaseBoneSpace
    baseBoneRef {boneName=neck_02}
    componentAppliedLinearAccClamp=<0 0 0>
    overrideWorldGravity {z=-980}
    simSpaceSettings {dampingAlpha=0, maxAngularAcceleration=50, maxAngularVelocity=5, maxLinearAcceleration=50, maxLinearVelocity=80, worldAlpha=0.05}
   }
   Output Pose(Result=Blend Poses by bool(BlendPose_0=Component To Local(ComponentPose=@n8), BlendPose_1=Component To Local(ComponentPose=@n9), bActiveValue=Get bIsParachuteOpened())) {layerGroup=None}
  ]}
 EventGraph {nodes [
   Event BlueprintUpdateAnimation()
   > n1 = "Gravity Override State Function (Parameter Asset) (Nativized)"(
    CharacterPartAnimInstance=Self-Reference()
    GameTimeInSeconds=GetGameTimeInSeconds()
    Parameters=&/M_MED_FruitFie_NudgeHonk_FaceAccess_Parameters
   )
   > Set GravityOverride(GravityOverride=@n1)
   > Set RBAN Settings Out(RBAN Settings Out=@n1.OutRigidBodyAnimNodeParameters)
   n2 = AND Boolean(A=Get bIsJumping(), B=Get bIsTacticalSprinting())
   n3 = Get bIsFalling()
   > Set Hair Alpha(Hair Alpha:double=SelectFloat(A=0, B=1, bPickA=OR Boolean(A=@n2, B=@n3))) -> (Output_Get:double)
  ]}
]
variables [
 GravityOverride {type=Vector}
 Hair Alpha {type=number}
 RBAN Settings Out {type=FortRigidBodyAnimNodeParameters}
]
